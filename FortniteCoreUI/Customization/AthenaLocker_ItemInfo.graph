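# Locker item info panel widget (subclass of AthenaLockerItemInfo) showing the selected cosmetic's name, type, rarity tag, description and restriction warning.
# Item events feed UpdateRarityWidgets, which tints the rarity tag's dynamic material from rarity or series colors.
# Restriction updates set callout text, color and warning-or-lock icon, with a looping pulse animation.
# PreConstruct applies compact mode, alignment, wrapping and description visibility options.
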WidgetBlueprint /FortniteCoreUI/Customization/AthenaLocker_ItemInfo
animations [PulseAnim {
  binding_count=2
  length_seconds=1.5
  object=&self:PulseAnim
  sequence {bindings [
    Pulse {
     id=#1
     possessed_class=Image
     tracks [
      {
       type=MovieSceneFloatTrack
       display_name=Render Opacity
       sections [{type=MovieSceneFloatSection, channels [None {type=MovieSceneScriptingFloatChannel, keys [[0,0],[2,0.4],[20,0],[30,0]]}]}]
      }
      {
       type=MovieScene2DTransformTrack
       sections [{
         type=MovieScene2DTransformSection
         channels [
          Translation.X {type=MovieSceneScriptingFloatChannel, const [2,2]}
          Translation.Y {type=MovieSceneScriptingFloatChannel, keys [[0,-2],[20,-8]]}
          Angle {type=MovieSceneScriptingFloatChannel}
          Scale.X {type=MovieSceneScriptingFloatChannel, keys [[0,1],[20,2.5]]}
          Scale.Y {type=MovieSceneScriptingFloatChannel, keys [[0,1],[20,2.5]]}
          Shear.X {type=MovieSceneScriptingFloatChannel}
          Shear.Y {type=MovieSceneScriptingFloatChannel}
         ]
        }]
      }
     ]
    }
    Pulse_1 {
     id=#2
     possessed_class=Image
     tracks [
      {
       type=MovieSceneFloatTrack
       display_name=Render Opacity
       sections [{type=MovieSceneFloatSection, channels [None {type=MovieSceneScriptingFloatChannel, keys [[0,0],[2,0.4],[20,0],[30,0]]}]}]
      }
      {
       type=MovieScene2DTransformTrack
       sections [{
         type=MovieScene2DTransformSection
         channels [
          Translation.X {type=MovieSceneScriptingFloatChannel}
          Translation.Y {type=MovieSceneScriptingFloatChannel}
          Angle {type=MovieSceneScriptingFloatChannel}
          Scale.X {type=MovieSceneScriptingFloatChannel, keys [[0,1],[20,2.5]]}
          Scale.Y {type=MovieSceneScriptingFloatChannel, keys [[0,1],[20,2.5]]}
          Shear.X {type=MovieSceneScriptingFloatChannel}
          Shear.Y {type=MovieSceneScriptingFloatChannel}
         ]
        }]
      }
     ]
    }
   ]}
 }]
graphs [
 EventGraph {nodes [
   n1 = Event OnItemDefinitionSet!() -> (OutputDelegate, then, ItemDefinition:Fort Item Definition Object Reference) {is=title K2Node_Event}
   > n2 = Is Valid(InputObject=@n1.ItemDefinition)
   n3 = GetRarityColorPalette!(execute=[@n4.then, @n2.Is Not Valid], Rarity:EFortRarity Enum=Common) -> (then, ReturnValue:Fort Color Palette Structure)
   > UpdateRarityWidgets!(
    execute
    self
    InColor:Fort Color Palette Structure=@n3
    InRarity:EFortRarity Enum=Common
    InSeries:Fort Item Series Definition Object Reference
    HideRarity:bool=true
   ) -> (then)
   UpdateRarityWidgets(
    execute=@n2.Is Valid
    InColor=GetRarityOrSeriesColors!(self:Fort Item Definition Object Reference=@n1.ItemDefinition) -> (ReturnValue:Fort Color Palette Structure)
    InRarity=GetRarity!(self:Fort Item Definition Object Reference=@n1.ItemDefinition) -> (ReturnValue:EFortRarity Enum=Common)
    InSeries=GetSeries!(ItemDefinition:Item Definition Base Object Reference=@n1.ItemDefinition) -> (ReturnValue:Fort Item Series Definition Object Reference)
    HideRarity=false
   )
   n4 = Event OnNoItemSet() {is=title K2Node_Event}
   Event PreConstruct()
   > SetVisibility(
    self=[
     Get OverrideDescriptionVisibility() -> (OverrideDescriptionVisibility:Overlay Object Reference)
     Get OverrideWarningVisibility() -> (OverrideWarningVisibility:Overlay Object Reference)
    ]
    InVisibility=Select(Option 0=Collapsed, Option 1=SelfHitTestInvisible, Index=Get ShowDescription() -> (ShowDescription:bool))
   )
   > SetHorizontalAlignment(
    self=SlotAsOverlaySlot(Widget=Get ContentOverlay() -> (ContentOverlay:Overlay Object Reference))
    InHorizontalAlignment=Select(
     Option 0:EHorizontalAlignment Enum=HAlign_Fill
     Option 1:EHorizontalAlignment Enum=HAlign_Left
     Index=Get CompactMode() -> (CompactMode:bool)
    ) -> (ReturnValue:EHorizontalAlignment Enum)
   )
   > SetVisibility(
    self=Get VerticalSpacer() -> (VerticalSpacer:Spacer Object Reference)
    InVisibility=Select(Option 0:ESlateVisibility Enum=HitTestInvisible, Option 1:ESlateVisibility Enum=Collapsed, Index=Get CompactMode()) -> (ReturnValue:ESlateVisibility Enum)
   )
   > SetMinDesiredWidth!(
    execute
    self:Size Box Object Reference=Get SizeBox_0() -> (SizeBox_0:Size Box Object Reference)
    InMinDesiredWidth:float=Select(Option 0:float=800, Option 1:float=0, Index=Get CompactMode()) -> (ReturnValue:float)
   ) -> (then)
   > SetHorizontalAlignment(
    self=[
     SlotAsOverlaySlot(Widget=Get Rarity+Type() -> (Rarity+Type:Horizontal Box Object Reference))
     SlotAsOverlaySlot(Widget=Get Text_ItemName() -> (Text_ItemName:Common Text Object Reference))
     SlotAsOverlaySlot(Widget=Get Text_ItemType() -> (Text_ItemType:Common Text Object Reference))
    ]
    InHorizontalAlignment=Get ContentAlignment() -> (ContentAlignment:EHorizontalAlignment Enum=HAlign_Fill)
   )
   n5 = "Equal (Enum)"(A:EHorizontalAlignment Enum=Get ContentAlignment() -> (ContentAlignment=HAlign_Fill), B:EHorizontalAlignment Enum=HAlign_Left)
   > Set Justification!(
    execute
    Justification:ETextJustify Enum=Select(Option 0:ETextJustify Enum=InvariantRight, Option 1:ETextJustify Enum=InvariantLeft, Index=@n5) -> (ReturnValue:ETextJustify Enum)
    self:Text Layout Widget Object Reference=Get RichText_Description() -> (RichText_Description:Common Rich Text Block Object Reference)
   ) -> (then, Output_Get:ETextJustify Enum=Left)
   > Set FlowDirectionPreference!(
    execute
    FlowDirectionPreference:EFlowDirectionPreference Enum=Select(Option 0:EFlowDirectionPreference Enum=LeftToRight, Option 1:EFlowDirectionPreference Enum=RightToLeft, Index=@n5) -> (ReturnValue:EFlowDirectionPreference Enum)
    self:Widget Object Reference=Get Rarity+Type()
   ) -> (then, Output_Get:EFlowDirectionPreference Enum=Inherit)
   > SetAutoWrapText!(execute, self:Rich Text Block Object Reference=Get RichText_Description(), InAutoTextWrap:bool=Get Wrap Text() -> (Wrap Text:bool)) -> (then)
   n6 = Event OnUpdateAccessRestrictionWarning!() -> (
    OutputDelegate
    then
    bShouldShow:bool
    WarningText:"Text (by ref)"
    WarningType:EAthenaLockerInfoRestrictionWarning Enum=UnsatisfiedExclusiveItem
   ) {is=title K2Node_Event}
   > SetActiveWidgetIndex(
    self=Get IconSwitcher() -> (IconSwitcher:Widget Switcher Object Reference)
    Index=Select!(
     UnsatisfiedExclusiveItem:int=0
     LockedEmote:int=1
     CosmeticRestriction:int=0
     AllItemsAreArchived:int=0
     Unknown:int=0
     Index:EAthenaLockerInfoRestrictionWarning Enum=@n6.WarningType
    ) -> (ReturnValue:int)
   )
   > SetVisibility(
    self=Get ExclusiveCalloutBox() -> (ExclusiveCalloutBox:Horizontal Box Object Reference)
    InVisibility=Select!(Option 0:ESlateVisibility Enum=Collapsed, Option 1:ESlateVisibility Enum=SelfHitTestInvisible, Index:bool=@n6.bShouldShow) -> (ReturnValue:ESlateVisibility Enum)
   )
   > PlayAnimation(InAnimation=Get PulseAnim() -> (PulseAnim:Widget Animation Object Reference), NumLoopsToPlay=0)
   n7 = Get ExclusiveCalloutText() -> (ExclusiveCalloutText:Common Rich Text Block Object Reference)
   > SetText(self=@n7, InText=@n6.WarningText)
   > SetDefaultColorAndOpacity!(
    execute
    self:Rich Text Block Object Reference=@n7
    InColorAndOpacity:Slate Color Structure=Make Slate Color(SpecifiedColor=Select!(
      UnsatisfiedExclusiveItem:Linear Color Structure=(R=0.9823,G=1,B=0.4233,A=1)
      LockedEmote:Linear Color Structure=(R=0.2159,G=0.6584,B=1,A=1)
      CosmeticRestriction:Linear Color Structure=(R=0.9823,G=1,B=0.4233,A=1)
      AllItemsAreArchived:Linear Color Structure=(R=0.9823,G=1,B=0.4233,A=1)
      Unknown:Linear Color Structure=(R=0.9823,G=1,B=0.4233,A=1)
      Index:EAthenaLockerInfoRestrictionWarning Enum=@n6.WarningType
     ) -> (ReturnValue:Linear Color Structure))
   ) -> (then)
   n8 = Event OnItemSet!() -> (OutputDelegate, then, Item:Fort Item VM Object Reference) {is=title K2Node_Event}
   > OnItemDefinitionSet!(
    execute
    self
    ItemDefinition:Fort Item Definition Object Reference=Get ItemDefinition!(self:Fort Item VM Object Reference=@n8.Item) -> (ItemDefinition:Fort Item Definition Object Reference)
   ) -> (then)
   n9 = Get ItemDetails!(self:Fort Item VM Object Reference=@n8.Item) -> (ItemDetails:Fort Item Details VM Object Reference)
   > SetVisibility(
    self=Get Border_ItemRarity() -> (Border_ItemRarity:Border Object Reference)
    InVisibility=Select!(
     Option 0:ESlateVisibility Enum=Collapsed
     Option 1:ESlateVisibility Enum=SelfHitTestInvisible
     Index:bool=Get bShouldShowRarity!(self:Fort Item Details VM Object Reference=@n9) -> (bShouldShowRarity:bool)
    ) -> (ReturnValue:ESlateVisibility Enum)
   )
  ]}
 UpdateRarityWidgets {nodes [
   n1 = UpdateRarityWidgets!() -> (
    then
    InColor:Fort Color Palette Structure
    InRarity:EFortRarity Enum=Common
    InSeries:Fort Item Series Definition Object Reference
    HideRarity:bool=~
   ) {is=title K2Node_FunctionEntry}
   > n2 = GetDynamicMaterial(self:Border Object Reference=Get Border_ItemRarity())
   n3 = Break Fort Color Palette!(FortColorPalette:"Fort Color Palette Structure (by ref)"=@n1) -> (
    Color1:Linear Color Structure
    Color2:Linear Color Structure
    Color3:Linear Color Structure
    Color4:Linear Color Structure
    Color5:Linear Color Structure
   ) {is=title K2Node_BreakStruct}
   > SetVectorParameterValue(self=@n2, ParameterName=BackingColor, Value=@n3.Color4)
   > SetVectorParameterValue(
    self=@n2
    ParameterName=BorderColor
    Value=Select(
     Option 0:Linear Color Structure=@n3
     Option 1:Linear Color Structure=@n3.Color2
     Index=IsNullItem!(ItemObject:Object Reference=@n1.InSeries) -> (Is Null Item:bool)
    ) -> (ReturnValue:Linear Color Structure)
   )
   > SetContentColorAndOpacity!(
    execute
    self:Border Object Reference=Get Border_RarityTint() -> (Border_RarityTint:Border Object Reference)
    InContentColorAndOpacity:Linear Color Structure=Select(Option 0=@n3, Option 1=@n3.Color2, Index=IsNullItem(ItemObject=@n1.InSeries))
   ) -> (then)
   > SetVisibility(
    self=Get Border_ItemRarity()
    InVisibility=Select(Option 0:ESlateVisibility Enum=HitTestInvisible, Option 1:ESlateVisibility Enum=Collapsed, Index=Get HideRarity() -> (HideRarity:bool)) -> (ReturnValue:ESlateVisibility Enum)
   )
  ]}
]
variables [
 CompactMode {type=boolean}
 ContentAlignment {type=EHorizontalAlignment, default=HAlign_Right}
 ShowDescription {type=boolean, default=true}
 Wrap Text {type=boolean}
]
widgets {
 info {inheritedWidgetCount=4, parentClass=/Script/FortniteUI.AthenaLockerItemInfo, rootWidgetClass=&SizeBox, widgetCount=28}
 widgets [
  SizeBox_0 {widgetClassPath=&SizeBox, bIsVariable=true, widget=SizeBox_0, properties {bOverride_MinDesiredWidth=true, minDesiredWidth=800}}
  Overlay_0 {widgetClassPath=&Overlay, parent=SizeBox_0, widget=Overlay_0, slot {class=SizeBoxSlot}}
  Image_93 {
   widgetClassPath=&Image
   bIsVariable=true
   parent=Overlay_0
   widget=Image_93
   properties {
    renderOpacity=0
    visibility=Collapsed
    brush {
     bIsDynamicallyLoaded=false
     drawAs=Image
     imageType=NoImage
     mirroring=NoMirror
     resourceName=None
     resourceObject=/FortniteCoreUI/Athena/HUD/Art/Tutorial/FTUE_Gradient_Radial_Quarter
     tiling=NoTile
     imageSize=<256 256>
     margin=<ltrb 0 0 0 0>
     outlineSettings {
      bUseBrushTransparency=false
      roundingType=HalfHeightRadius
      width=0
      color {colorUseRule=UseColor_Specified, specifiedColor=<rgba 0 0 0 0>}
      cornerRadii=<0 0 0 0>
     }
     tintColor {colorUseRule=UseColor_Specified, specifiedColor=<rgba 0.0075 0.0545 0.3916 1>}
     uVRegion {bIsValid=false, max=<0 0>, min=<0 0>}
    }
    renderTransform {angle=0, scale=<5 2>, shear=<0 0>, translation=<0 0>}
    renderTransformPivot=<1 1>
   }
   slot {class=OverlaySlot, properties {horizontalAlignment=HAlign_Right, verticalAlignment=VAlign_Bottom, padding=<ltrb 0 0 -128 0>}}
  }
  HorizontalBox_0 {
   widgetClassPath=&HorizontalBox
   parent=Overlay_0
   widget=HorizontalBox_0
   slot {class=OverlaySlot, properties {horizontalAlignment=HAlign_Right, verticalAlignment=VAlign_Bottom}}
  }
  ContentOverlay {
   widgetClassPath=&Overlay
   bIsVariable=true
   parent=HorizontalBox_0
   widget=ContentOverlay
   slot {class=HorizontalBoxSlot, properties {size {sizeRule=Fill, value=1}}}
  }
  Rarity+Type {
   widgetClassPath=&HorizontalBox
   bIsVariable=true
   parent=ContentOverlay
   widget=Rarity+Type
   slot {class=OverlaySlot, properties {horizontalAlignment=HAlign_Right}}
  }
  Spacer_321 {widgetClassPath=&Spacer, parent=Rarity+Type, widget=Spacer_321, properties {size=<0 32>}, slot {class=HorizontalBoxSlot}}
  Border_ItemType {
   widgetClassPath=&Border
   bIsVariable=true
   parent=Rarity+Type
   widget=Border_ItemType
   properties {
    background {
     bIsDynamicallyLoaded=false
     drawAs=NoDrawType
     imageType=NoImage
     mirroring=NoMirror
     resourceName=None
     resourceObject=null
     tiling=NoTile
     imageSize=<32 32>
     margin=<ltrb 0 0 0 0>
     outlineSettings {
      bUseBrushTransparency=false
      roundingType=HalfHeightRadius
      width=0
      color {colorUseRule=UseColor_Specified, specifiedColor=<rgba 0 0 0 0>}
      cornerRadii=<0 0 0 0>
     }
     tintColor {colorUseRule=UseColor_Specified, specifiedColor=<rgba 1 1 1 1>}
     uVRegion {bIsValid=false, max=<0 0>, min=<0 0>}
    }
    contentColorAndOpacity=<rgba 0.0331 0.7011 1 1>
    padding=<ltrb 8 0 4 0>
   }
   slot {class=HorizontalBoxSlot, properties {horizontalAlignment=HAlign_Right, verticalAlignment=VAlign_Center, size {sizeRule=Fill, value=1}}}
  }
  Text_ItemType {
   widgetClassPath=&CommonTextBlock
   bInherited=true
   bIsVariable=true
   parent=Border_ItemType
   widget=Text_ItemType
   properties {
    justification=Right
    scrollStyle=&TextScrollStyle-Base_C
    style=/FortUILibrary/Styles/Typography/_TextStyle_TagNormal._TextStyle_TagNormal_C
    text=Item Type
    textTransformPolicy=ToUpper
    font {
     bForceMonospaced=false
     bMaterialIsStencil=false
     fontMaterial=null
     fontObject=&BurbankSmall
     letterSpacing=0
     monospacedWidth=1
     size=15
     skewAmount=0
     typefaceFontName=Black
     outlineSettings {
      bApplyOutlineToDropShadows=false
      bMiteredCorners=false
      bSeparateFillAlpha=false
      outlineMaterial=null
      outlineSize=0
      outlineColor=<rgba 0 0 0 1>
     }
    }
    shadowOffset=<0 0>
    strikeBrush {
     bIsDynamicallyLoaded=false
     drawAs=NoDrawType
     imageType=NoImage
     mirroring=NoMirror
     resourceName=None
     resourceObject=null
     tiling=NoTile
     imageSize=<32 32>
     margin=<ltrb 0 0 0 0>
     outlineSettings {
      bUseBrushTransparency=false
      roundingType=HalfHeightRadius
      width=0
      color {colorUseRule=UseColor_Specified, specifiedColor=<rgba 0 0 0 0>}
      cornerRadii=<0 0 0 0>
     }
     tintColor {colorUseRule=UseColor_Specified, specifiedColor=<rgba 1 1 1 1>}
     uVRegion {bIsValid=false, max=<0 0>, min=<0 0>}
    }
   }
   slot {class=BorderSlot, properties {padding=<ltrb 8 0 4 0>}}
  }
  Border_ItemRarity {
   widgetClassPath=&Border
   bIsVariable=true
   parent=Rarity+Type
   widget=Border_ItemRarity
   properties {
    verticalAlignment=VAlign_Center
    background {
     bIsDynamicallyLoaded=false
     drawAs=Box
     imageType=NoImage
     mirroring=NoMirror
     resourceName=None
     resourceObject=/Game/Athena/UI/Season/Art/M_UI_BattlePass_SimpleTag
     tiling=NoTile
     imageSize=<32 32>
     margin=<ltrb 0.5 0 0.5 0>
     outlineSettings {
      bUseBrushTransparency=false
      roundingType=HalfHeightRadius
      width=0
      color {colorUseRule=UseColor_Specified, specifiedColor=<rgba 0 0 0 0>}
      cornerRadii=<0 0 0 0>
     }
     tintColor {colorUseRule=UseColor_Specified, specifiedColor=<rgba 1 1 1 1>}
     uVRegion {bIsValid=false, max=<0 0>, min=<0 0>}
    }
    padding=<ltrb 24 0 24 0>
   }
   slot {class=HorizontalBoxSlot, properties {horizontalAlignment=HAlign_Left, padding=<ltrb 8 0 0 0>}}
  }
  Border_RarityTint {
   widgetClassPath=&Border
   bIsVariable=true
   parent=Border_ItemRarity
   widget=Border_RarityTint
   properties {
    background {
     bIsDynamicallyLoaded=false
     drawAs=NoDrawType
     imageType=NoImage
     mirroring=NoMirror
     resourceName=None
     resourceObject=null
     tiling=NoTile
     imageSize=<32 32>
     margin=<ltrb 0 0 0 0>
     outlineSettings {
      bUseBrushTransparency=false
      roundingType=HalfHeightRadius
      width=0
      color {colorUseRule=UseColor_Specified, specifiedColor=<rgba 0 0 0 0>}
      cornerRadii=<0 0 0 0>
     }
     tintColor {colorUseRule=UseColor_Specified, specifiedColor=<rgba 1 1 1 1>}
     uVRegion {bIsValid=false, max=<0 0>, min=<0 0>}
    }
    padding=<ltrb 0 0 0 0>
   }
   slot {class=BorderSlot, properties {verticalAlignment=VAlign_Center, padding=<ltrb 24 0 24 0>}}
  }
  Text_ItemRarity {
   widgetClassPath=&CommonTextBlock
   bInherited=true
   bIsVariable=true
   parent=Border_RarityTint
   widget=Text_ItemRarity
   properties {
    text=Item Rarity
    textTransformPolicy=ToUpper
    font {
     bForceMonospaced=false
     bMaterialIsStencil=false
     fontMaterial=null
     fontObject=&BurbankSmall
     letterSpacing=0
     monospacedWidth=1
     size=15
     skewAmount=0
     typefaceFontName=Black
     outlineSettings {
      bApplyOutlineToDropShadows=false
      bMiteredCorners=false
      bSeparateFillAlpha=false
      outlineMaterial=null
      outlineSize=0
      outlineColor=<rgba 0 0 0 1>
     }
    }
    margin=<ltrb 0 -4 0 -4>
    renderTransform {angle=0, scale=<1 1>, shear=<-12 0>, translation=<0 0>}
    shadowOffset=<0 0>
   }
   slot {class=BorderSlot, properties {padding=<ltrb 0 0 0 0>}}
  }
  Text_ItemName {
   widgetClassPath=&CommonTextBlock
   bInherited=true
   bIsVariable=true
   parent=ContentOverlay
   widget=Text_ItemName
   properties {
    clipping=ClipToBoundsWithoutIntersecting
    justification=Right
    scrollStyle=/Game/Athena/Prototype/Blueprints/Galileo/UI/TextScrollStyle-Base-Ticker.TextScrollStyle-Base-Ticker_C
    style=/FortUILibrary/Styles/Typography/_TextStyle_SubH2._TextStyle_SubH2_C
    text=Item Name
    textTransformPolicy=ToUpper
    font {
     bForceMonospaced=false
     bMaterialIsStencil=false
     fontMaterial=null
     fontObject=&Burbank
     letterSpacing=0
     monospacedWidth=1
     size=33
     skewAmount=0
     typefaceFontName=BurbankRegular-Black
     outlineSettings {
      bApplyOutlineToDropShadows=false
      bMiteredCorners=false
      bSeparateFillAlpha=false
      outlineMaterial=null
      outlineSize=0
      outlineColor=<rgba 0 0 0 1>
     }
    }
    shadowOffset=<0 0>
    strikeBrush {
     bIsDynamicallyLoaded=false
     drawAs=NoDrawType
     imageType=NoImage
     mirroring=NoMirror
     resourceName=None
     resourceObject=null
     tiling=NoTile
     imageSize=<32 32>
     margin=<ltrb 0 0 0 0>
     outlineSettings {
      bUseBrushTransparency=false
      roundingType=HalfHeightRadius
      width=0
      color {colorUseRule=UseColor_Specified, specifiedColor=<rgba 0 0 0 0>}
      cornerRadii=<0 0 0 0>
     }
     tintColor {colorUseRule=UseColor_Specified, specifiedColor=<rgba 1 1 1 1>}
     uVRegion {bIsValid=false, max=<0 0>, min=<0 0>}
    }
   }
   slot {class=OverlaySlot, properties {horizontalAlignment=HAlign_Right, padding=<ltrb 0 32 0 0>}}
  }
  VerticalBox_0 {
   widgetClassPath=&VerticalBox
   parent=ContentOverlay
   widget=VerticalBox_0
   slot {class=OverlaySlot, properties {horizontalAlignment=HAlign_Fill, padding=<ltrb 0 92 0 0>}}
  }
  OverrideDescriptionVisibility {
   widgetClassPath=&Overlay
   bIsVariable=true
   parent=VerticalBox_0
   widget=OverrideDescriptionVisibility
   slot {class=VerticalBoxSlot}
  }
  RichText_Description {
   widgetClassPath=&CommonRichTextBlock
   bInherited=true
   bIsVariable=true
   parent=OverrideDescriptionVisibility
   widget=RichText_Description
   properties {
    bOverrideDefaultStyle=true
    justification=Right
    mobileTextBlockScale=1.2
    text="The description goes here!"
    wrapTextAt=800
    defaultTextStyleOverride {
     overflowPolicy=Clip
     transformPolicy=None
     colorAndOpacity {colorUseRule=UseColor_Specified, specifiedColor=<rgba 0.2159 0.6584 1 1>}
     font {
      bForceMonospaced=false
      bMaterialIsStencil=false
      fontMaterial=null
      fontObject=&BurbankSmall
      letterSpacing=0
      monospacedWidth=1
      size=16
      skewAmount=0.3
      typefaceFontName=Bold
      outlineSettings {
       bApplyOutlineToDropShadows=false
       bMiteredCorners=false
       bSeparateFillAlpha=false
       outlineMaterial=null
       outlineSize=0
       outlineColor=<rgba 0 0 0 1>
      }
     }
     highlightColor {colorUseRule=UseColor_Specified, specifiedColor=<rgba 0 0 0 1>}
     highlightShape {
      bIsDynamicallyLoaded=false
      drawAs=Image
      imageType=NoImage
      mirroring=NoMirror
      resourceName=None
      resourceObject=null
      tiling=NoTile
      imageSize=<32 32>
      margin=<ltrb 0 0 0 0>
      outlineSettings {
       bUseBrushTransparency=false
       roundingType=HalfHeightRadius
       width=0
       color {colorUseRule=UseColor_Specified, specifiedColor=<rgba 0 0 0 0>}
       cornerRadii=<0 0 0 0>
      }
      tintColor {colorUseRule=UseColor_Specified, specifiedColor=<rgba 1 1 1 1>}
      uVRegion {bIsValid=false, max=<0 0>, min=<0 0>}
     }
     selectedBackgroundColor {colorUseRule=UseColor_Foreground, specifiedColor=<rgba 1 0 1 1>}
     shadowColorAndOpacity=<rgba 0 0 0 1>
     shadowOffset=<0 0>
     strikeBrush {
      bIsDynamicallyLoaded=false
      drawAs=Image
      imageType=NoImage
      mirroring=NoMirror
      resourceName=None
      resourceObject=null
      tiling=NoTile
      imageSize=<32 32>
      margin=<ltrb 0 0 0 0>
      outlineSettings {
       bUseBrushTransparency=false
       roundingType=HalfHeightRadius
       width=0
       color {colorUseRule=UseColor_Specified, specifiedColor=<rgba 0 0 0 0>}
       cornerRadii=<0 0 0 0>
      }
      tintColor {colorUseRule=UseColor_Specified, specifiedColor=<rgba 1 1 1 1>}
      uVRegion {bIsValid=false, max=<0 0>, min=<0 0>}
     }
     underlineBrush {
      bIsDynamicallyLoaded=false
      drawAs=Image
      imageType=NoImage
      mirroring=NoMirror
      resourceName=None
      resourceObject=null
      tiling=NoTile
      imageSize=<32 32>
      margin=<ltrb 0 0 0 0>
      outlineSettings {
       bUseBrushTransparency=false
       roundingType=HalfHeightRadius
       width=0
       color {colorUseRule=UseColor_Specified, specifiedColor=<rgba 0 0 0 0>}
       cornerRadii=<0 0 0 0>
      }
      tintColor {colorUseRule=UseColor_Specified, specifiedColor=<rgba 1 1 1 1>}
      uVRegion {bIsValid=false, max=<0 0>, min=<0 0>}
     }
    }
   }
   slot {class=OverlaySlot, properties {horizontalAlignment=HAlign_Fill, verticalAlignment=VAlign_Fill, padding=<ltrb -2 0 4 0>}}
  }
  OverrideWarningVisibility {widgetClassPath=&Overlay, bIsVariable=true, parent=VerticalBox_0, widget=OverrideWarningVisibility, slot {class=VerticalBoxSlot}}
  ExclusiveCalloutBox {
   widgetClassPath=&HorizontalBox
   bIsVariable=true
   parent=OverrideWarningVisibility
   widget=ExclusiveCalloutBox
   slot {class=OverlaySlot}
  }
  IconSwitcher {
   widgetClassPath=&WidgetSwitcher
   bIsVariable=true
   parent=ExclusiveCalloutBox
   widget=IconSwitcher
   slot {class=HorizontalBoxSlot, properties {horizontalAlignment=HAlign_Center, verticalAlignment=VAlign_Center}}
  }
  WarningOvr {widgetClassPath=&Overlay, parent=IconSwitcher, widget=WarningOvr, slot {class=WidgetSwitcherSlot}}
  Pulse {
   widgetClassPath=&Image
   parent=WarningOvr
   widget=Pulse
   properties {
    visibility=SelfHitTestInvisible
    brush {
     bIsDynamicallyLoaded=false
     drawAs=Image
     imageType=NoImage
     mirroring=NoMirror
     resourceName=None
     resourceObject=/FortniteCoreUI/Customization/Art/M_UI_BattlePassTrack_ExclusiveIndicator_Pulse
     tiling=NoTile
     imageSize=<32 32>
     margin=<ltrb 0 0 0 0>
     outlineSettings {
      bUseBrushTransparency=false
      roundingType=HalfHeightRadius
      width=0
      color {colorUseRule=UseColor_Specified, specifiedColor=<rgba 0 0 0 0>}
      cornerRadii=<0 0 0 0>
     }
     tintColor {colorUseRule=UseColor_Specified, specifiedColor=<rgba 1 1 1 1>}
     uVRegion {bIsValid=false, max=<0 0>, min=<0 0>}
    }
   }
   slot {class=OverlaySlot, properties {horizontalAlignment=HAlign_Fill, verticalAlignment=VAlign_Fill}}
  }
  ExclusiveCalloutIcon {
   widgetClassPath=&Image
   parent=WarningOvr
   widget=ExclusiveCalloutIcon
   properties {
    visibility=SelfHitTestInvisible
    brush {
     bIsDynamicallyLoaded=false
     drawAs=Image
     imageType=NoImage
     mirroring=NoMirror
     resourceName=None
     resourceObject=/FortniteCoreUI/Customization/Art/M_UI_BattlePassTrack_ExclusiveIndicator
     tiling=NoTile
     imageSize=<40 40>
     margin=<ltrb 0 0 0 0>
     outlineSettings {
      bUseBrushTransparency=false
      roundingType=HalfHeightRadius
      width=0
      color {colorUseRule=UseColor_Specified, specifiedColor=<rgba 0 0 0 0>}
      cornerRadii=<0 0 0 0>
     }
     tintColor {colorUseRule=UseColor_Specified, specifiedColor=<rgba 1 1 1 1>}
     uVRegion {bIsValid=false, max=<0 0>, min=<0 0>}
    }
   }
   slot {class=OverlaySlot, properties {horizontalAlignment=HAlign_Fill, verticalAlignment=VAlign_Fill}}
  }
  CommonTextBlock_67 {
   widgetClassPath=&CommonTextBlock
   parent=WarningOvr
   widget=CommonTextBlock_67
   properties {
    text="!"
    colorAndOpacity {colorUseRule=UseColor_Specified, specifiedColor=<rgba 0.9301 1 0.0103 1>}
    font {
     bForceMonospaced=false
     bMaterialIsStencil=false
     fontMaterial=null
     fontObject=&BurbankSmall
     letterSpacing=0
     monospacedWidth=1
     size=16
     skewAmount=0
     typefaceFontName=Black
     outlineSettings {
      bApplyOutlineToDropShadows=false
      bMiteredCorners=false
      bSeparateFillAlpha=false
      outlineMaterial=null
      outlineSize=0
      outlineColor=<rgba 0 0 0 1>
     }
    }
    margin=<ltrb 0 -4 0 -4>
    renderTransform {angle=0, scale=<1 1>, shear=<-4 0>, translation=<-2 5>}
    shadowOffset=<0 0>
   }
   slot {class=OverlaySlot, properties {horizontalAlignment=HAlign_Center, verticalAlignment=VAlign_Center}}
  }
  LockedOvr {widgetClassPath=&Overlay, parent=IconSwitcher, widget=LockedOvr, slot {class=WidgetSwitcherSlot}}
  Pulse_1 {
   widgetClassPath=&Image
   parent=LockedOvr
   widget=Pulse_1
   properties {
    visibility=SelfHitTestInvisible
    brush {
     bIsDynamicallyLoaded=false
     drawAs=Image
     imageType=NoImage
     mirroring=NoMirror
     resourceName=None
     resourceObject=/FortniteCoreUI/Customization/Art/M_UI_AngledLockedBox_BuiltInEmote_Pulse
     tiling=NoTile
     imageSize=<32 32>
     margin=<ltrb 0 0 0 0>
     outlineSettings {
      bUseBrushTransparency=false
      roundingType=HalfHeightRadius
      width=0
      color {colorUseRule=UseColor_Specified, specifiedColor=<rgba 0 0 0 0>}
      cornerRadii=<0 0 0 0>
     }
     tintColor {colorUseRule=UseColor_Specified, specifiedColor=<rgba 1 1 1 1>}
     uVRegion {bIsValid=false, max=<0 0>, min=<0 0>}
    }
   }
   slot {class=OverlaySlot, properties {horizontalAlignment=HAlign_Fill, verticalAlignment=VAlign_Fill}}
  }
  ExclusiveCalloutIcon_1 {
   widgetClassPath=&Image
   parent=LockedOvr
   widget=ExclusiveCalloutIcon_1
   properties {
    visibility=SelfHitTestInvisible
    brush {
     bIsDynamicallyLoaded=false
     drawAs=Image
     imageType=NoImage
     mirroring=NoMirror
     resourceName=None
     resourceObject=/FortniteCoreUI/Customization/Art/M_UI_AngledLockedBox_BuiltInEmote
     tiling=NoTile
     imageSize=<40 40>
     margin=<ltrb 0 0 0 0>
     outlineSettings {
      bUseBrushTransparency=false
      roundingType=HalfHeightRadius
      width=0
      color {colorUseRule=UseColor_Specified, specifiedColor=<rgba 0 0 0 0>}
      cornerRadii=<0 0 0 0>
     }
     tintColor {colorUseRule=UseColor_Specified, specifiedColor=<rgba 1 1 1 1>}
     uVRegion {bIsValid=false, max=<0 0>, min=<0 0>}
    }
   }
   slot {class=OverlaySlot, properties {horizontalAlignment=HAlign_Fill, verticalAlignment=VAlign_Fill}}
  }
  ExclusiveCalloutText {
   widgetClassPath=&CommonRichTextBlock
   bIsVariable=true
   parent=ExclusiveCalloutBox
   widget=ExclusiveCalloutText
   properties {
    bOverrideDefaultStyle=true
    mobileTextBlockScale=1.15
    text=Callout explaining Exclusive Item behavior
    wrapTextAt=448
    defaultTextStyleOverride {
     overflowPolicy=Clip
     transformPolicy=None
     colorAndOpacity {colorUseRule=UseColor_Specified, specifiedColor=<rgba 0.9823 1 0.4233 1>}
     font {
      bForceMonospaced=false
      bMaterialIsStencil=false
      fontMaterial=null
      fontObject=&BurbankSmall
      letterSpacing=0
      monospacedWidth=1
      size=16
      skewAmount=0.3
      typefaceFontName=Bold
      outlineSettings {
       bApplyOutlineToDropShadows=false
       bMiteredCorners=false
       bSeparateFillAlpha=false
       outlineMaterial=null
       outlineSize=0
       outlineColor=<rgba 0 0 0 1>
      }
     }
     highlightColor {colorUseRule=UseColor_Specified, specifiedColor=<rgba 0 0 0 1>}
     highlightShape {
      bIsDynamicallyLoaded=false
      drawAs=Image
      imageType=NoImage
      mirroring=NoMirror
      resourceName=None
      resourceObject=null
      tiling=NoTile
      imageSize=<32 32>
      margin=<ltrb 0 0 0 0>
      outlineSettings {
       bUseBrushTransparency=false
       roundingType=HalfHeightRadius
       width=0
       color {colorUseRule=UseColor_Specified, specifiedColor=<rgba 0 0 0 0>}
       cornerRadii=<0 0 0 0>
      }
      tintColor {colorUseRule=UseColor_Specified, specifiedColor=<rgba 1 1 1 1>}
      uVRegion {bIsValid=false, max=<0 0>, min=<0 0>}
     }
     selectedBackgroundColor {colorUseRule=UseColor_Foreground, specifiedColor=<rgba 1 0 1 1>}
     shadowColorAndOpacity=<rgba 0 0 0 1>
     shadowOffset=<0 0>
     strikeBrush {
      bIsDynamicallyLoaded=false
      drawAs=Image
      imageType=NoImage
      mirroring=NoMirror
      resourceName=None
      resourceObject=null
      tiling=NoTile
      imageSize=<32 32>
      margin=<ltrb 0 0 0 0>
      outlineSettings {
       bUseBrushTransparency=false
       roundingType=HalfHeightRadius
       width=0
       color {colorUseRule=UseColor_Specified, specifiedColor=<rgba 0 0 0 0>}
       cornerRadii=<0 0 0 0>
      }
      tintColor {colorUseRule=UseColor_Specified, specifiedColor=<rgba 1 1 1 1>}
      uVRegion {bIsValid=false, max=<0 0>, min=<0 0>}
     }
     underlineBrush {
      bIsDynamicallyLoaded=false
      drawAs=Image
      imageType=NoImage
      mirroring=NoMirror
      resourceName=None
      resourceObject=null
      tiling=NoTile
      imageSize=<32 32>
      margin=<ltrb 0 0 0 0>
      outlineSettings {
       bUseBrushTransparency=false
       roundingType=HalfHeightRadius
       width=0
       color {colorUseRule=UseColor_Specified, specifiedColor=<rgba 0 0 0 0>}
       cornerRadii=<0 0 0 0>
      }
      tintColor {colorUseRule=UseColor_Specified, specifiedColor=<rgba 1 1 1 1>}
      uVRegion {bIsValid=false, max=<0 0>, min=<0 0>}
     }
    }
   }
   slot {
    class=HorizontalBoxSlot
    properties {horizontalAlignment=HAlign_Center, verticalAlignment=VAlign_Center, padding=<ltrb 2 0 4 0>, size {sizeRule=Fill, value=1}}
   }
  }
  VerticalSpacer {
   widgetClassPath=&Spacer
   bIsVariable=true
   parent=HorizontalBox_0
   widget=VerticalSpacer
   properties {size=<40 160>}
   slot {class=HorizontalBoxSlot, properties {horizontalAlignment=HAlign_Right, verticalAlignment=VAlign_Bottom}}
  }
 ]
}
