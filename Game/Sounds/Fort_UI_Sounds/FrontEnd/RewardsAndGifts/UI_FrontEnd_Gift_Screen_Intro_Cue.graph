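# Sound cue for the front-end gift screen intro, played as a layered UI sting.
# A mixer plays a claim-reward screen appear sound immediately, then layers in a Battle Pass menu swoosh and a Ranked close whoosh at staggered fixed delays and lower volumes.
# It uses the Fort_UI sound class and overrides concurrency to allow few simultaneous instances, stopping the oldest.
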SoundCue /Game/Sounds/Fort_UI_Sounds/FrontEnd/RewardsAndGifts/UI_FrontEnd_Gift_Screen_Intro_Cue
root=Mixer_0
nodes [
 {id=Delay_0, children [WavePlayer_4], properties {delayMax=0.2, delayMin=0.2}}
 {id=Delay_1, children [WavePlayer_2], properties {delayMax=0.4, delayMin=0.4}}
 {id=Mixer_0, children [WavePlayer_0,Delay_0,Delay_1], properties {inputVolume [1,0.7,0.6]}}
 {id=WavePlayer_0, properties {soundWaveAssetPtr=/Game/Sounds/Fort_UI_Sounds/FrontEnd/RewardsAndGifts/wav/UI_ClaimReward_Screen_Appear}}
 {id=WavePlayer_2, properties {soundWaveAssetPtr=/Game/Sounds/Fort_UI_Sounds/FrontEnd/New_FrontEnd/Ranked/wav/UI_Ranked_CloseWhoosh}}
 {id=WavePlayer_4, properties {soundWaveAssetPtr=/Game/Sounds/Fort_UI_Sounds/FrontEnd/BattlePass/BP_ToolKit/BP_MenuSwoosh_L_to_R}}
]
properties {
 bOverrideConcurrency=true
 maxDistance=2097152
 soundClassObject=&Fort_UI
 volumeMultiplier=0.9
 concurrencyOverrides {maxCount=2, resolutionRule=StopOldest}
 concurrencySet [/Game/Sounds/Concurrency_Rules/FrontEnd_UI_Menu_Max_03_Stop_Oldest]
}
