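# Gifting price widget based on AthenaGiftingPriceWidget, showing either a V-Bucks price with struck-through regular price or a real-money price with a platform-filtered PlayStation Store logo.
# SetPresentationMode picks the layout in a visibility switcher or collapses the widget when Hidden.
# PlaySelectedAnim and PlayUnselectedAnim run one color animation forward or reversed to tint price texts and strike line dark teal when selected.
# A false PlayAnim flag starts playback at the end time, skipping the transition.
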WidgetBlueprint /Gifting/UI/RecipientGiftingPrice
animations [Selected {
  binding_count=4
  length_seconds=0.15
  object=&self:Selected
  sequence {bindings [
    Text_FinalPrice {
     id=#1
     possessed_class=CommonTextBlock
     tracks [{
       type=MovieSceneColorTrack
       display_name=Color and Opacity
       sections [{
         type=MovieSceneColorSection
         end_frame=3
         start_frame=0
         channels [
          Color.R {type=MovieSceneScriptingFloatChannel, keys [[0,1],[3,0.04]]}
          Color.G {type=MovieSceneScriptingFloatChannel, keys [[0,1],[3,0.3]]}
          Color.B {type=MovieSceneScriptingFloatChannel, keys [[0,1],[3,0.32]]}
          Color.A {type=MovieSceneScriptingFloatChannel, keys [[0,0.5],[3,0.8]]}
         ]
        }]
      }]
    }
    IMG_Strike {
     id=#2
     possessed_class=Image
     tracks [{
       type=MovieSceneColorTrack
       display_name=Color and Opacity
       sections [{
         type=MovieSceneColorSection
         end_frame=3
         start_frame=0
         channels [
          Color.R {type=MovieSceneScriptingFloatChannel, keys [[0,1],[3,0.04]]}
          Color.G {type=MovieSceneScriptingFloatChannel, keys [[0,1],[3,0.3]]}
          Color.B {type=MovieSceneScriptingFloatChannel, keys [[0,1],[3,0.32]]}
          Color.A {type=MovieSceneScriptingFloatChannel, keys [[0,0.5],[3,0.8]]}
         ]
        }]
      }]
    }
    Text_RealMoneyPrice {
     id=#3
     possessed_class=CommonTextBlock
     tracks [{
       type=MovieSceneColorTrack
       display_name=Color and Opacity
       sections [{
         type=MovieSceneColorSection
         end_frame=3
         start_frame=0
         channels [
          Color.R {type=MovieSceneScriptingFloatChannel, keys [[0,1],[3,0.04]]}
          Color.G {type=MovieSceneScriptingFloatChannel, keys [[0,1],[3,0.3]]}
          Color.B {type=MovieSceneScriptingFloatChannel, keys [[0,1],[3,0.32]]}
          Color.A {type=MovieSceneScriptingFloatChannel, keys [[0,0.5],[3,0.8]]}
         ]
        }]
      }]
    }
    Text_RegularPrice {
     id=#4
     possessed_class=CommonTextBlock
     tracks [{
       type=MovieSceneColorTrack
       display_name=Color and Opacity
       sections [{
         type=MovieSceneColorSection
         end_frame=3
         start_frame=0
         channels [
          Color.R {type=MovieSceneScriptingFloatChannel, keys [[0,1],[3,0.04]]}
          Color.G {type=MovieSceneScriptingFloatChannel, keys [[0,1],[3,0.3]]}
          Color.B {type=MovieSceneScriptingFloatChannel, keys [[0,1],[3,0.32]]}
          Color.A {type=MovieSceneScriptingFloatChannel, keys [[0,0.5],[3,0.8]]}
         ]
        }]
      }]
    }
   ]}
 }]
graphs [
 EventGraph {links [], nodes []}
 PlaySelectedAnim {nodes [
   n1 = PlaySelectedAnim!() -> (then, PlayAnim:bool=~) {is=title K2Node_FunctionEntry}
   n2 = Get Selected() -> (Selected:Widget Animation Object Reference)
   > PlayAnimation(InAnimation=@n2, StartAtTime=Select(Option 0:double=GetEndTime(self=@n2), Option 1:double=0, Index=@n1) -> (ReturnValue:double))
  ]}
 PlayUnselectedAnim {nodes [
   n1 = PlayUnselectedAnim!() -> (then, PlayAnim:bool=~) {is=title K2Node_FunctionEntry}
   n2 = Get Selected()
   > PlayAnimation(InAnimation=@n2, StartAtTime=Select(Option 0=GetEndTime(self=@n2), Option 1=0, Index=@n1), PlayMode=Reverse)
  ]}
 SetPresentationMode {nodes [
   n1 = SetPresentationMode!() -> (then, Mode:EGiftingPricePresentationMode Enum=MtxCurrency) {is=title K2Node_FunctionEntry}
   > n2 = Switch on EGiftingPricePresentationMode!(execute, Selection:EGiftingPricePresentationMode Enum=@n1) -> (MtxCurrency:Exec, RealMoney:Exec, Hidden:Exec) {is=title K2Node_SwitchEnum}
   SetVisibility(execute=@n2.Hidden, InVisibility=Collapsed)
   SetActiveWidget(
    execute=[@n2.MtxCurrency, @n2.RealMoney]
    self=Get Switcher_Prices() -> (Switcher_Prices:Common Visibility Switcher Object Reference)
    Widget=Select!(
     MtxCurrency:Widget Object Reference=Get Widget_VBucks() -> (Widget_VBucks:Horizontal Box Object Reference)
     RealMoney:Widget Object Reference=Get Widget_RealMoneyPrice() -> (Widget_RealMoneyPrice:Overlay Object Reference)
     Hidden:Widget Object Reference
     Index:EGiftingPricePresentationMode Enum=Get Mode() -> (Mode:EGiftingPricePresentationMode Enum=MtxCurrency)
    ) -> (ReturnValue:Widget Object Reference)
   )
   > SetVisibility(InVisibility=SelfHitTestInvisible)
  ]}
]
variables []
widgets {
 info {inheritedWidgetCount=4, parentClass=/Script/GiftingUI.AthenaGiftingPriceWidget, rootWidgetClass=&Overlay, widgetCount=16}
 widgets [
  SubtitleContainer {widgetClassPath=&Overlay, bIsVariable=true, widget=SubtitleContainer, properties {visibility=HitTestInvisible}}
  ScaleBox {
   widgetClassPath=&ScaleBox
   parent=SubtitleContainer
   widget=ScaleBox
   properties {clipping=ClipToBoundsWithoutIntersecting, stretch=ScaleToFitX, stretchDirection=DownOnly}
   slot {class=OverlaySlot, properties {horizontalAlignment=HAlign_Fill, verticalAlignment=VAlign_Fill}}
  }
  Switcher_Prices {
   widgetClassPath=&CommonVisibilitySwitcher
   bIsVariable=true
   parent=ScaleBox
   widget=Switcher_Prices
   slot {class=ScaleBoxSlot, properties {horizontalAlignment=HAlign_Fill}}
  }
  Widget_VBucks {
   widgetClassPath=&HorizontalBox
   bIsVariable=true
   parent=Switcher_Prices
   widget=Widget_VBucks
   slot {class=CommonVisibilitySwitcherSlot, properties {horizontalAlignment=HAlign_Center, verticalAlignment=VAlign_Center}}
  }
  Overlay_SalePrice {
   widgetClassPath=&Overlay
   bInherited=true
   parent=Widget_VBucks
   widget=Overlay_SalePrice
   slot {class=HorizontalBoxSlot, properties {horizontalAlignment=HAlign_Right, verticalAlignment=VAlign_Center, padding=<ltrb 0 0 4 0>}}
  }
  Text_RegularPrice {
   widgetClassPath=&CommonTextBlock
   bInherited=true
   bIsVariable=true
   parent=Overlay_SalePrice
   widget=Text_RegularPrice
   properties {
    justification=Center
    mobileFontSizeMultiplier=1.2
    text=1500
    textTransformPolicy=ToUpper
    colorAndOpacity {colorUseRule=UseColor_Specified, specifiedColor=<rgba 1 1 1 0.5>}
    font {
     bForceMonospaced=false
     bMaterialIsStencil=false
     fontMaterial=null
     fontObject=&Burbank
     letterSpacing=0
     monospacedWidth=1
     size=15
     skewAmount=0
     typefaceFontName=BurbankRegular-Black
     outlineSettings {
      bApplyOutlineToDropShadows=false
      bMiteredCorners=false
      bSeparateFillAlpha=false
      outlineMaterial=null
      outlineSize=0
      outlineColor=<rgba 0 0 0 1>
     }
    }
    renderTransform {angle=0, scale=<1 1>, shear=<-12 0>, translation=<0 0>}
    shadowOffset=<0 0>
   }
   slot {class=OverlaySlot, properties {horizontalAlignment=HAlign_Right, padding=<ltrb 0 0 4 0>}}
  }
  IMG_Strike {
   widgetClassPath=&Image
   bIsVariable=true
   parent=Overlay_SalePrice
   widget=IMG_Strike
   properties {
    renderOpacity=0.75
    brush {
     bIsDynamicallyLoaded=false
     drawAs=Image
     imageType=NoImage
     mirroring=NoMirror
     resourceName=None
     resourceObject=/Gifting/UI/Art/M_UI_StrikeThrough_Line_GiftingPurchaseButton
     tiling=NoTile
     imageSize=<32 4>
     margin=<ltrb 0 0 0 0>
     outlineSettings {
      bUseBrushTransparency=false
      roundingType=HalfHeightRadius
      width=0
      color {colorUseRule=UseColor_Specified, specifiedColor=<rgba 0 0 0 0>}
      cornerRadii=<0 0 0 0>
     }
     tintColor {colorUseRule=UseColor_Specified, specifiedColor=<rgba 1 1 1 1>}
     uVRegion {bIsValid=false, max=<0 0>, min=<0 0>}
    }
    colorAndOpacity=<rgba 1 1 1 0.5>
    renderTransform {angle=0, scale=<1.125 1>, shear=<0 -3>, translation=<-2 0>}
   }
   slot {class=OverlaySlot, properties {horizontalAlignment=HAlign_Fill, verticalAlignment=VAlign_Center}}
  }
  Widget_VBuckPrice {widgetClassPath=&Overlay, parent=Widget_VBucks, widget=Widget_VBuckPrice, slot {class=HorizontalBoxSlot}}
  Cost {
   widgetClassPath=&HorizontalBox
   parent=Widget_VBuckPrice
   widget=Cost
   slot {class=OverlaySlot, properties {horizontalAlignment=HAlign_Fill, verticalAlignment=VAlign_Fill}}
  }
  Text_FinalPrice {
   widgetClassPath=&CommonTextBlock
   bInherited=true
   bIsVariable=true
   parent=Cost
   widget=Text_FinalPrice
   properties {
    mobileFontSizeMultiplier=1.2
    text=950
    colorAndOpacity {colorUseRule=UseColor_Specified, specifiedColor=<rgba 1 1 1 0.5>}
    font {
     bForceMonospaced=false
     bMaterialIsStencil=false
     fontMaterial=null
     fontObject=&Burbank
     letterSpacing=0
     monospacedWidth=1
     size=15
     skewAmount=0
     typefaceFontName=BurbankRegular-Black
     outlineSettings {
      bApplyOutlineToDropShadows=false
      bMiteredCorners=false
      bSeparateFillAlpha=false
      outlineMaterial=null
      outlineSize=0
      outlineColor=<rgba 0 0 0 1>
     }
    }
    renderTransform {angle=0, scale=<1 1>, shear=<-12 0>, translation=<0 0>}
    shadowOffset=<0 0>
   }
   slot {class=HorizontalBoxSlot, properties {verticalAlignment=VAlign_Center, padding=<ltrb 10 0 10 0>}}
  }
  Image_VBuck {
   widgetClassPath=&Image
   bIsVariable=true
   parent=Cost
   widget=Image_VBuck
   properties {
    brush {
     bIsDynamicallyLoaded=false
     drawAs=Image
     imageType=NoImage
     mirroring=NoMirror
     resourceName=None
     resourceObject=/Game/UI/Foundation/Shell/Textures/T-Icon-VBucks-L
     tiling=NoTile
     imageSize=<48 48>
     margin=<ltrb 0 0 0 0>
     outlineSettings {
      bUseBrushTransparency=false
      roundingType=HalfHeightRadius
      width=0
      color {colorUseRule=UseColor_Specified, specifiedColor=<rgba 0 0 0 0>}
      cornerRadii=<0 0 0 0>
     }
     tintColor {colorUseRule=UseColor_Specified, specifiedColor=<rgba 1 1 1 1>}
     uVRegion {bIsValid=false, max=<0 0>, min=<0 0>}
    }
    colorAndOpacity=<rgba 1 1 1 0.5>
    renderTransform {angle=20, scale=<1 1>, shear=<0 0>, translation=<0 0>}
   }
   slot {class=HorizontalBoxSlot, properties {horizontalAlignment=HAlign_Center, verticalAlignment=VAlign_Center}}
  }
  Widget_RealMoneyPrice {
   widgetClassPath=&Overlay
   bIsVariable=true
   parent=Switcher_Prices
   widget=Widget_RealMoneyPrice
   slot {class=CommonVisibilitySwitcherSlot, properties {horizontalAlignment=HAlign_Center, verticalAlignment=VAlign_Center}}
  }
  HorizontalBox_0 {
   widgetClassPath=&HorizontalBox
   parent=Widget_RealMoneyPrice
   widget=HorizontalBox_0
   slot {class=OverlaySlot, properties {horizontalAlignment=HAlign_Fill, verticalAlignment=VAlign_Fill}}
  }
  Text_RealMoneyPrice {
   widgetClassPath=&CommonTextBlock
   bInherited=true
   parent=HorizontalBox_0
   widget=Text_RealMoneyPrice
   properties {
    justification=Center
    mobileFontSizeMultiplier=1.2
    text=$xx.xx
    textTransformPolicy=ToUpper
    colorAndOpacity {colorUseRule=UseColor_Specified, specifiedColor=<rgba 1 1 1 0.5>}
    font {
     bForceMonospaced=false
     bMaterialIsStencil=false
     fontMaterial=null
     fontObject=&Burbank
     letterSpacing=0
     monospacedWidth=1
     size=15
     skewAmount=0
     typefaceFontName=BurbankRegular-Black
     outlineSettings {
      bApplyOutlineToDropShadows=false
      bMiteredCorners=false
      bSeparateFillAlpha=false
      outlineMaterial=null
      outlineSize=0
      outlineColor=<rgba 0 0 0 1>
     }
    }
    shadowOffset=<0 0>
   }
   slot {class=HorizontalBoxSlot, properties {verticalAlignment=VAlign_Center, padding=<ltrb 0 0 8 0>}}
  }
  CommonVisibilityWidget_0 {
   widgetClassPath=&CommonVisibilityWidgetLegacy
   parent=HorizontalBox_0
   widget=CommonVisibilityWidget_0
   properties {
    bShowForAndroid=false
    bShowForErebus=false
    bShowForIOS=false
    bShowForMac=false
    bShowForPC=false
    bShowForSwitch2=false
    bShowForTouch=false
    bShowForXBox=false
    bShowForXSX=false
    padding=<ltrb 0 0 0 0>
   }
   slot {class=HorizontalBoxSlot, properties {horizontalAlignment=HAlign_Left, verticalAlignment=VAlign_Center}}
  }
  PlayStationStoreLogo {
   widgetClassPath=&Image
   bIsVariable=true
   parent=CommonVisibilityWidget_0
   widget=PlayStationStoreLogo
   properties {
    clipping=ClipToBoundsWithoutIntersecting
    visibility=SelfHitTestInvisible
    brush {
     bIsDynamicallyLoaded=false
     drawAs=Image
     imageType=NoImage
     mirroring=NoMirror
     resourceName=None
     resourceObject=/Game/UI/Frontend/ItemShop/Art/T_UI_PS_Store_Logo_64x
     tiling=NoTile
     imageSize=<40 40>
     margin=<ltrb 0 0 0 0>
     outlineSettings {
      bUseBrushTransparency=false
      roundingType=HalfHeightRadius
      width=0
      color {colorUseRule=UseColor_Specified, specifiedColor=<rgba 0 0 0 0>}
      cornerRadii=<0 0 0 0>
     }
     tintColor {colorUseRule=UseColor_Specified, specifiedColor=<rgba 1 1 1 1>}
     uVRegion {bIsValid=false, max=<0 0>, min=<0 0>}
    }
    colorAndOpacity=<rgba 1 1 1 0.5>
    renderTransform {angle=0, scale=<1 1>, shear=<0 0>, translation=<0 -2>}
   }
   slot {class=BorderSlot, properties {padding=<ltrb 0 0 0 0>}}
  }
 ]
}
