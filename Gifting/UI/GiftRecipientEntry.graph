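# Widget Blueprint for one recipient row in the gifting UI, with avatar, social name, ownership status strip and a touch-only confirm button.
# SetPresentationMode switches the status panel among Loading, GiftPrice, AlreadyOwned and Unavailable, enabling the input action only for GiftPrice.
# Highlight changes play selection animations forward or reversed, jumping to the end when animation is off, and forward selection to the price widget.
# The mobile confirm button appears and animates only in GiftPrice mode.
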WidgetBlueprint /Gifting/UI/GiftRecipientEntry
animations [
 Mobile_ConfirmPushed {
  binding_count=3
  length_seconds=0.05
  object=&self:Mobile_ConfirmPushed
  sequence {bindings [
    AnimatedInputBorder {
     id=#1
     display_name=Border_MobileConfirm
     possessed_class=Border
     tracks [
      {
       type=MovieSceneColorTrack
       display_name=Brush Color
       sections [{
         type=MovieSceneColorSection
         channels [
          Color.R {type=MovieSceneScriptingFloatChannel}
          Color.G {type=MovieSceneScriptingFloatChannel}
          Color.B {type=MovieSceneScriptingFloatChannel}
          Color.A {type=MovieSceneScriptingFloatChannel, const [1,1]}
         ]
        }]
      }
      {
       type=MovieSceneWidgetMaterialTrack
       sections [{
         type=MovieSceneParameterSection
         channels [
          AnimateStateChange {type=MovieSceneScriptingFloatChannel, const [1,1]}
          RGBR {type=MovieSceneScriptingFloatChannel, keys [[0,0.0056],[1,0.0343]]}
          RGBG {type=MovieSceneScriptingFloatChannel, keys [[0,0.4452],[1,1]]}
          RGBB {type=MovieSceneScriptingFloatChannel, keys [[0,0.8879],[1,0.227]]}
          RGBA {type=MovieSceneScriptingFloatChannel, const [1,1]}
         ]
        }]
      }
      {
       type=MovieSceneFloatTrack
       display_name=Render Opacity
       sections [{type=MovieSceneFloatSection, channels [None {type=MovieSceneScriptingFloatChannel, const [1,1]}]}]
      }
     ]
    }
    TextColorBorder {
     id=#2
     possessed_class=Border
     tracks [{
       type=MovieSceneColorTrack
       display_name=Content Color and Opacity
       sections [{
         type=MovieSceneColorSection
         channels [
          Color.R {type=MovieSceneScriptingFloatChannel, keys [[0,1],[1,0]]}
          Color.G {type=MovieSceneScriptingFloatChannel, keys [[0,1],[1,0.147]]}
          Color.B {type=MovieSceneScriptingFloatChannel, keys [[0,1],[1,0.063]]}
          Color.A {type=MovieSceneScriptingFloatChannel, const [1,1]}
         ]
        }]
      }]
    }
    VisibilityWidget_ConfirmButton {
     id=#3
     display_name=Text_MobileConfirm
     possessed_class=CommonVisibilityWidgetLegacy
     tracks [{
       type=MovieSceneFloatTrack
       display_name=Render Opacity
       sections [{type=MovieSceneFloatSection, channels [None {type=MovieSceneScriptingFloatChannel, const [1,1]}]}]
      }]
    }
   ]}
 }
 Selected {
  binding_count=8
  length_seconds=0.15
  object=&self:Selected
  sequence {bindings [
    OwnedBacking {
     id=#4
     possessed_class=Image
     tracks [{
       type=MovieSceneColorTrack
       display_name=Color and Opacity
       sections [{
         type=MovieSceneColorSection
         channels [
          Color.R {type=MovieSceneScriptingFloatChannel, keys [[0,0.0037],[2,0.0063]]}
          Color.G {type=MovieSceneScriptingFloatChannel, keys [[0,0.075],[2,0.125]]}
          Color.B {type=MovieSceneScriptingFloatChannel, keys [[0,0.0394],[2,0.0656]]}
          Color.A {type=MovieSceneScriptingFloatChannel, const [1,1]}
         ]
        }]
      }]
    }
    OwnedItems_ColorBorder {
     id=#5
     possessed_class=Border
     tracks [{
       type=MovieSceneColorTrack
       display_name=Content Color and Opacity
       sections [{
         type=MovieSceneColorSection
         channels [
          Color.R {type=MovieSceneScriptingFloatChannel, keys [[0,1],[3,0.0409]]}
          Color.G {type=MovieSceneScriptingFloatChannel, keys [[0,1],[3,0.2918]]}
          Color.B {type=MovieSceneScriptingFloatChannel, keys [[0,1],[3,0.3278]]}
          Color.A {type=MovieSceneScriptingFloatChannel, const [1,1]}
         ]
        }]
      }]
    }
    Price_BackingColorFiller {
     id=#6
     possessed_class=Image
     tracks [{
       type=MovieSceneColorTrack
       display_name=Color and Opacity
       sections [{
         type=MovieSceneColorSection
         end_frame=3
         start_frame=0
         channels [
          Color.R {type=MovieSceneScriptingFloatChannel, keys [[0,0.0052],[3,0.5029]]}
          Color.G {type=MovieSceneScriptingFloatChannel, keys [[0,0.0561],[3,1]]}
          Color.B {type=MovieSceneScriptingFloatChannel, keys [[0,0.1779],[3,0.9216]]}
          Color.A {type=MovieSceneScriptingFloatChannel, const [1,1]}
         ]
        }]
      }]
    }
    Border-HoverOverlay {
     id=#7
     possessed_class=Border
     tracks [{
       type=MovieScene2DTransformTrack
       sections [{
         type=MovieScene2DTransformSection
         channels [
          Translation.X {type=MovieSceneScriptingFloatChannel, keys [[0,0],[3.0003,-8]]}
          Translation.Y {type=MovieSceneScriptingFloatChannel, const [0,1]}
         ]
        }]
      }]
    }
    Border_UnavailableOverlay {
     id=#8
     possessed_class=Border
     tracks [{
       type=MovieScene2DTransformTrack
       sections [{
         type=MovieScene2DTransformSection
         channels [
          Translation.X {type=MovieSceneScriptingFloatChannel, keys [[0,0],[3.0003,-8]]}
          Translation.Y {type=MovieSceneScriptingFloatChannel, const [0,1]}
         ]
        }]
      }]
    }
    Border_Background {
     id=#9
     possessed_class=Border
     tracks [{
       type=MovieSceneWidgetMaterialTrack
       sections [{type=MovieSceneParameterSection, channels [Animate_BackingColor {type=MovieSceneScriptingFloatChannel, keys [[0,0],[3,1]]}]}]
      }]
    }
    SocialName_Color {
     id=#10
     possessed_class=Border
     tracks [{
       type=MovieSceneColorTrack
       display_name=Content Color and Opacity
       sections [{
         type=MovieSceneColorSection
         channels [
          Color.R {type=MovieSceneScriptingFloatChannel, keys [[0,1],[2,0]]}
          Color.G {type=MovieSceneScriptingFloatChannel, keys [[0,1],[2,0]]}
          Color.B {type=MovieSceneScriptingFloatChannel, keys [[0,1],[2,0]]}
          Color.A {type=MovieSceneScriptingFloatChannel, const [1,1]}
         ]
        }]
      }]
    }
    Border_Background {
     id=#11
     possessed_class=Border
     tracks [{
       type=MovieScene2DTransformTrack
       sections [{
         type=MovieScene2DTransformSection
         channels [
          Translation.X {type=MovieSceneScriptingFloatChannel, keys [[0,0],[3.0003,-8]]}
          Translation.Y {type=MovieSceneScriptingFloatChannel, const [0,1]}
          Angle {type=MovieSceneScriptingFloatChannel, const [0,1]}
          Scale.X {type=MovieSceneScriptingFloatChannel, const [1,1]}
          Scale.Y {type=MovieSceneScriptingFloatChannel, const [1,1]}
          Shear.X {type=MovieSceneScriptingFloatChannel, const [0,1]}
          Shear.Y {type=MovieSceneScriptingFloatChannel, const [0,1]}
         ]
        }]
      }]
    }
   ]}
 }
 Mobile_ConfirmAppear {
  binding_count=3
  length_seconds=0.2
  object=&self:Mobile_ConfirmAppear
  sequence {bindings [
    AnimatedInputBorder {
     id=#1
     display_name=Border_MobileConfirm
     possessed_class=Border
     tracks [
      {
       type=MovieSceneColorTrack
       display_name=Brush Color
       sections [{
         type=MovieSceneColorSection
         channels [
          Color.R {type=MovieSceneScriptingFloatChannel}
          Color.G {type=MovieSceneScriptingFloatChannel}
          Color.B {type=MovieSceneScriptingFloatChannel}
          Color.A {type=MovieSceneScriptingFloatChannel, keys [[0,0],[2,1]]}
         ]
        }]
      }
      {
       type=MovieSceneWidgetMaterialTrack
       sections [{
         type=MovieSceneParameterSection
         channels [
          AnimateStateChange {type=MovieSceneScriptingFloatChannel, keys [[0,0],[3,1],[4,1]]}
          RGBR {type=MovieSceneScriptingFloatChannel, const [0.0056,1]}
          RGBG {type=MovieSceneScriptingFloatChannel, const [0.4452,1]}
          RGBB {type=MovieSceneScriptingFloatChannel, const [0.8879,1]}
          RGBA {type=MovieSceneScriptingFloatChannel, const [1,1]}
         ]
        }]
      }
      {
       type=MovieSceneFloatTrack
       display_name=Render Opacity
       sections [{type=MovieSceneFloatSection, channels [None {type=MovieSceneScriptingFloatChannel, keys [[0,0],[1,1]]}]}]
      }
     ]
    }
    TextColorBorder {
     id=#12
     possessed_class=Border
     tracks [{
       type=MovieSceneColorTrack
       display_name=Content Color and Opacity
       sections [{
         type=MovieSceneColorSection
         channels [
          Color.R {type=MovieSceneScriptingFloatChannel, const [1,1]}
          Color.G {type=MovieSceneScriptingFloatChannel, const [1,1]}
          Color.B {type=MovieSceneScriptingFloatChannel, const [1,1]}
          Color.A {type=MovieSceneScriptingFloatChannel, const [1,1]}
         ]
        }]
      }]
    }
    VisibilityWidget_ConfirmButton {
     id=#3
     display_name=Text_MobileConfirm
     possessed_class=CommonVisibilityWidgetLegacy
     tracks [{
       type=MovieSceneFloatTrack
       display_name=Render Opacity
       sections [{type=MovieSceneFloatSection, channels [None {type=MovieSceneScriptingFloatChannel, keys [[0,0],[2,1]]}]}]
      }]
    }
   ]}
 }
 State_UnavailableLoading {
  binding_count=3
  length_seconds=0.05
  object=&self:State_UnavailableLoading
  sequence {bindings [
    Unavailable_OverlayImage {
     id=#13
     possessed_class=Image
     tracks [{
       type=MovieSceneFloatTrack
       display_name=Render Opacity
       sections [{type=MovieSceneFloatSection, end_frame=1, start_frame=0, channels [None {type=MovieSceneScriptingFloatChannel, const [1,1]}]}]
      }]
    }
    Border_Background {
     id=#14
     possessed_class=Border
     tracks [{
       type=MovieSceneFloatTrack
       display_name=Render Opacity
       sections [{type=MovieSceneFloatSection, channels [None {type=MovieSceneScriptingFloatChannel, const [0.5,1]}]}]
      }]
    }
   ]}
 }
 State_AvailableLoaded {
  binding_count=2
  length_seconds=0.1
  object=&self:State_AvailableLoaded
  sequence {bindings [
    Unavailable_OverlayImage {
     id=#13
     possessed_class=Image
     tracks [{
       type=MovieSceneFloatTrack
       display_name=Render Opacity
       sections [{type=MovieSceneFloatSection, end_frame=2, start_frame=0, channels [None {type=MovieSceneScriptingFloatChannel, keys [[0,1],[2,0]]}]}]
      }]
    }
    Border_Background {
     id=#14
     possessed_class=Border
     tracks [{
       type=MovieSceneFloatTrack
       display_name=Render Opacity
       sections [{type=MovieSceneFloatSection, channels [None {type=MovieSceneScriptingFloatChannel, keys [[0,0.5],[2,1]]}]}]
      }]
    }
   ]}
 }
 Selected_UnavailableLoading {
  binding_count=6
  length_seconds=0.15
  object=&self:Selected_UnavailableLoading
  sequence {bindings [
    Border_Background {
     id=#15
     possessed_class=Border
     tracks [{
       type=MovieScene2DTransformTrack
       sections [{
         type=MovieScene2DTransformSection
         channels [
          Translation.X {type=MovieSceneScriptingFloatChannel, keys [[0,0],[3.0003,-8]]}
          Translation.Y {type=MovieSceneScriptingFloatChannel, const [0,1]}
          Angle {type=MovieSceneScriptingFloatChannel, const [0,1]}
          Scale.X {type=MovieSceneScriptingFloatChannel, const [1,1]}
          Scale.Y {type=MovieSceneScriptingFloatChannel, const [1,1]}
          Shear.X {type=MovieSceneScriptingFloatChannel, const [0,1]}
          Shear.Y {type=MovieSceneScriptingFloatChannel, const [0,1]}
         ]
        }]
      }]
    }
    Unavailable_OverlayImage {
     id=#13
     possessed_class=Image
     tracks [{
       type=MovieSceneFloatTrack
       display_name=Render Opacity
       sections [{type=MovieSceneFloatSection, channels [None {type=MovieSceneScriptingFloatChannel, keys [[0,1],[2,0.8]]}]}]
      }]
    }
    Border_UnavailableOverlay {
     id=#16
     possessed_class=Border
     tracks [{
       type=MovieScene2DTransformTrack
       sections [{
         type=MovieScene2DTransformSection
         channels [
          Translation.X {type=MovieSceneScriptingFloatChannel, keys [[0,0],[3.0003,-8]]}
          Translation.Y {type=MovieSceneScriptingFloatChannel, const [0,1]}
         ]
        }]
      }]
    }
    Border_Background {
     id=#14
     possessed_class=Border
     tracks [{
       type=MovieSceneFloatTrack
       display_name=Render Opacity
       sections [{type=MovieSceneFloatSection, channels [None {type=MovieSceneScriptingFloatChannel, keys [[0,0.5],[2,0.7]]}]}]
      }]
    }
    Border-HoverOverlay {
     id=#17
     possessed_class=Border
     tracks [{
       type=MovieScene2DTransformTrack
       sections [{
         type=MovieScene2DTransformSection
         channels [
          Translation.X {type=MovieSceneScriptingFloatChannel, keys [[0,0],[3.0003,-8]]}
          Translation.Y {type=MovieSceneScriptingFloatChannel, const [0,1]}
         ]
        }]
      }]
    }
   ]}
 }
]
graphs [
 BP_OnHighlightedStateChanged {nodes [
   n1 = BP_OnHighlightedStateChanged!() -> (then, bInIsHighlighted:bool, bPlayAnimation:bool) {is=title K2Node_FunctionEntry}
   > Set IsHighlighted(IsHighlighted:bool=@n1) -> (Output_Get:bool)
   > n2 = Branch(Condition="Equal (Enum)"(
     A:ERecipientPresentationMode Enum=Get PresentationMode() -> (PresentationMode:ERecipientPresentationMode Enum=Loading)
     B:ERecipientPresentationMode Enum=GiftPrice
    ))
   n3 = Get Mobile_ConfirmAppear() -> (Mobile_ConfirmAppear:Widget Animation Object Reference)
   > n4 = PlayAnimation(
    InAnimation=@n3
    StartAtTime=Select(Option 0:double=GetEndTime(self=@n3), Option 1:double=0, Index=Get bPlayAnimation() -> (bPlayAnimation:bool)) -> (ReturnValue:double)
    PlayMode=Select(Option 0:EUMGSequencePlayMode Enum=Reverse, Option 1:EUMGSequencePlayMode Enum, Index=Get IsHighlighted() -> (IsHighlighted:bool)) -> (ReturnValue:EUMGSequencePlayMode Enum)
   )
   n5 = Switch on ERecipientPresentationMode!(execute=[@n14.then, @n4.then], Selection:ERecipientPresentationMode Enum=Get PresentationMode() -> (PresentationMode=Loading)) -> (Loading:Exec, GiftPrice:Exec, AlreadyOwned:Exec, Unavailable:Exec) {is=title K2Node_SwitchEnum}
   n6 = Get Selected() -> (Selected:Widget Animation Object Reference)
   n7 = Select(Option 0=Reverse, Index=Get IsHighlighted())
   n8 = PlayAnimation(
    execute=[@n5.AlreadyOwned, @n5.GiftPrice]
    InAnimation=@n6
    StartAtTime=Select(Option 0:double=GetEndTime(self=@n6), Option 1:double=0, Index=Get bPlayAnimation()) -> (ReturnValue:double)
    PlayMode=@n7
   )
   n9 = Cast To RecipientGiftingPrice!(execute=[@n8.then, @n12.then], Object=Get Widget_Price() -> (Widget_Price:Athena Gifting Price Widget Object Reference)) -> (then, CastFailed, AsRecipient Gifting Price:Recipient Gifting Price Object Reference) {is=title K2Node_DynamicCast}
   > n10 = Branch(Condition=Get IsHighlighted())
   > PlaySelectedAnim!(execute, self:Recipient Gifting Price Object Reference=@n9, PlayAnim:bool=Get bPlayAnimation()) -> (then)
   PlayUnselectedAnim!(execute=@n10.else, self:Recipient Gifting Price Object Reference=@n9, PlayAnim:bool=Get bPlayAnimation()) -> (then)
   n11 = Get Selected_UnavailableLoading() -> (Selected_UnavailableLoading:Widget Animation Object Reference)
   n12 = PlayAnimation(
    execute=[@n5.Loading, @n5.Unavailable]
    InAnimation=@n11
    StartAtTime=Select(Option 0=GetEndTime(self=@n11), Option 1=0, Index=Get bPlayAnimation())
    PlayMode=@n7
   )
   n13 = Get Mobile_ConfirmAppear()
   n14 = PlayAnimation(execute=@n2.else, InAnimation=@n13, StartAtTime=GetEndTime(self=@n13), PlayMode=Reverse)
  ]}
 BP_OnTouchSelectionConfirmed {nodes [
   BP_OnTouchSelectionConfirmed() {is=title K2Node_FunctionEntry}
   > Branch(Condition="Equal (Enum)"(A=Get PresentationMode() -> (PresentationMode=Loading), B=GiftPrice))
   > PlayAnimation(InAnimation=Get Mobile_ConfirmPushed() -> (Mobile_ConfirmPushed:Widget Animation Object Reference))
  ]}
 EventGraph {nodes [
   "On Input Icon Updated (InputActionWidget)"() {is=title K2Node_ComponentBoundEvent}
   > n1 = GetIconDynamicMaterial!(execute, self:Common Action Widget Object Reference=Get InputActionWidget() -> (InputActionWidget:Common Action Widget Object Reference)) -> (then, ReturnValue:Material Instance Dynamic Object Reference)
   > SetScalarParameterValue(self=@n1, ParameterName=Use Dark, Value=1)
   > SetScalarParameterValue(self=@n1, ParameterName=Outline Opacity, Value=0)
   Event Construct()
   > n2 = Cast To RecipientGiftingPrice(Object=Get Widget_Price())
   > Set Price Widget(Price Widget:Recipient Gifting Price Object Reference=@n2) -> (Output_Get:Recipient Gifting Price Object Reference)
  ]}
 SetPresentationMode {nodes [
   n1 = SetPresentationMode!() -> (then, Mode:ERecipientPresentationMode Enum=Loading) {is=title K2Node_FunctionEntry}
   > Set PresentationMode(PresentationMode:ERecipientPresentationMode Enum=@n1) -> (Output_Get:ERecipientPresentationMode Enum=Loading)
   n2 = Get PresentationMode() -> (PresentationMode=Loading)
   > n3 = Branch(Condition=AND Boolean(A=Get IsHighlighted(), B="Equal (Enum)"(A=@n2, B=GiftPrice)))
   n4 = Get Mobile_ConfirmAppear()
   > n5 = PlayAnimation(InAnimation=@n4, StartAtTime=GetEndTime(self=@n4))
   SetIsEnabled(
    execute=[@n11.then, @n5.then]
    self=Get InputActionWidget()
    bInIsEnabled=Select!(
     Loading:bool=false
     GiftPrice:bool=true
     AlreadyOwned:bool=false
     Unavailable:bool=false
     Index:ERecipientPresentationMode Enum=Get PresentationMode() -> (PresentationMode=Loading)
    ) -> (ReturnValue:bool)
   )
   > SetActiveWidgetIndex(
    self:Common Visibility Switcher Object Reference=Get Switcher_OwningStatus() -> (Switcher_OwningStatus:Common Visibility Switcher Object Reference)
    Index=Select(Loading:int=0, GiftPrice:int=1, AlreadyOwned:int=2, Unavailable:int=3, Index=Get PresentationMode() -> (PresentationMode=Loading)) -> (ReturnValue:int)
   )
   > n6 = Switch on ERecipientPresentationMode(Selection=Get PresentationMode() -> (PresentationMode=Loading))
   n7 = Branch(execute=[@n6.Loading, @n6.Unavailable], Condition=Get IsHighlighted())
   n8 = Get Selected_UnavailableLoading()
   > PlayAnimation(InAnimation=@n8, StartAtTime=GetEndTime(self=@n8))
   PlayAnimation(execute=@n7.else, InAnimation=Get State_UnavailableLoading() -> (State_UnavailableLoading:Widget Animation Object Reference))
   PlayAnimation(
    execute=[@n6.AlreadyOwned, @n6.GiftPrice]
    InAnimation=Get State_AvailableLoaded() -> (State_AvailableLoaded:Widget Animation Object Reference)
   )
   > Branch(Condition=Get IsHighlighted())
   n9 = Get Selected()
   > PlayAnimation(InAnimation=@n9, StartAtTime=GetEndTime(self=@n9))
   n10 = Get Mobile_ConfirmAppear()
   n11 = PlayAnimation(execute=@n3.else, InAnimation=@n10, StartAtTime=GetEndTime(self=@n10), PlayMode=Reverse)
  ]}
]
variables [
 IsHighlighted {type=boolean}
 PresentationMode {type=ERecipientPresentationMode, default=Loading}
 Price Widget {type=RecipientGiftingPrice_C, default=null}
]
widgets {
 info {inheritedWidgetCount=6, parentClass=/Script/GiftingUI.FortGiftingSocialUserListEntry, rootWidgetClass=&Overlay, widgetCount=46}
 widgets [
  Overlay_1 {widgetClassPath=&Overlay, widget=Overlay_1}
  Border_Background {
   widgetClassPath=&Border
   bIsVariable=true
   parent=Overlay_1
   widget=Border_Background
   properties {
    visibility=SelfHitTestInvisible
    background {
     bIsDynamicallyLoaded=false
     drawAs=Image
     imageType=NoImage
     mirroring=NoMirror
     resourceName=None
     resourceObject=/Gifting/UI/Art/M_UI_Gifting_FriendBackgroundt
     tiling=NoTile
     imageSize=<32 32>
     margin=<ltrb 0 0 0 0>
     outlineSettings {
      bUseBrushTransparency=false
      roundingType=HalfHeightRadius
      width=0
      color {colorUseRule=UseColor_Specified, specifiedColor=<rgba 0 0 0 0>}
      cornerRadii=<0 0 0 0>
     }
     tintColor {colorUseRule=UseColor_Specified, specifiedColor=<rgba 1 1 1 1>}
     uVRegion {bIsValid=false, max=<0 0>, min=<0 0>}
    }
    padding=<ltrb 0 0 0 0>
   }
   slot {class=OverlaySlot, properties {horizontalAlignment=HAlign_Fill, verticalAlignment=VAlign_Fill, padding=<ltrb 8 0 8 0>}}
  }
  HorizontalBox_2 {
   widgetClassPath=&HorizontalBox
   parent=Border_Background
   widget=HorizontalBox_2
   properties {flowDirectionPreference=Culture}
   slot {class=BorderSlot, properties {padding=<ltrb 0 0 0 0>}}
  }
  SB_AvatarIcon {
   widgetClassPath=&SizeBox
   parent=HorizontalBox_2
   widget=SB_AvatarIcon
   properties {
    bOverride_MinDesiredHeight=true
    bOverride_MinDesiredWidth=true
    heightOverride=68
    minDesiredHeight=84
    minDesiredWidth=84
    widthOverride=68
   }
   slot {class=HorizontalBoxSlot}
  }
  ScaleBox_0 {widgetClassPath=&ScaleBox, parent=SB_AvatarIcon, widget=ScaleBox_0, properties {stretch=ScaleToFitY}, slot {class=SizeBoxSlot}}
  Avatar_MemberIcon {
   widgetClassPath=/Script/FortniteUI.FortSocialAvatarIcon
   bInherited=true
   bIsVariable=true
   parent=ScaleBox_0
   widget=Avatar_MemberIcon
   properties {
    avatarMaterial=/Game/UI/SocialPanel/Art/Materials/UI_UserPortrait_Face
    materialTextureParamName=SmallPreview
    brush {
     bIsDynamicallyLoaded=false
     drawAs=Image
     imageType=NoImage
     mirroring=NoMirror
     resourceName=None
     resourceObject=&UI_UserPortrait_Face
     tiling=NoTile
     imageSize=<32 32>
     margin=<ltrb 0 0 0 0>
     outlineSettings {
      bUseBrushTransparency=false
      roundingType=HalfHeightRadius
      width=0
      color {colorUseRule=UseColor_Specified, specifiedColor=<rgba 0 0 0 0>}
      cornerRadii=<0 0 0 0>
     }
     tintColor {colorUseRule=UseColor_Specified, specifiedColor=<rgba 1 1 1 1>}
     uVRegion {bIsValid=false, max=<0 0>, min=<0 0>}
    }
   }
   slot {class=ScaleBoxSlot}
  }
  SB_RecipientInfo {
   widgetClassPath=&SizeBox
   parent=HorizontalBox_2
   widget=SB_RecipientInfo
   properties {bOverride_MinDesiredWidth=true, minDesiredWidth=500}
   slot {class=HorizontalBoxSlot, properties {size {sizeRule=Fill, value=1}}}
  }
  VB_RecipientTextInfo {widgetClassPath=&VerticalBox, parent=SB_RecipientInfo, widget=VB_RecipientTextInfo, slot {class=SizeBoxSlot}}
  SB_NameAndConfirm {
   widgetClassPath=&SizeBox
   parent=VB_RecipientTextInfo
   widget=SB_NameAndConfirm
   properties {minDesiredHeight=69}
   slot {class=VerticalBoxSlot}
  }
  HB_NameAndConfirm {widgetClassPath=&HorizontalBox, parent=SB_NameAndConfirm, widget=HB_NameAndConfirm, slot {class=SizeBoxSlot}}
  SocialName_Color {
   widgetClassPath=&Border
   parent=HB_NameAndConfirm
   widget=SocialName_Color
   properties {
    background {
     bIsDynamicallyLoaded=false
     drawAs=NoDrawType
     imageType=NoImage
     mirroring=NoMirror
     resourceName=None
     resourceObject=null
     tiling=NoTile
     imageSize=<32 32>
     margin=<ltrb 0 0 0 0>
     outlineSettings {
      bUseBrushTransparency=false
      roundingType=HalfHeightRadius
      width=0
      color {colorUseRule=UseColor_Specified, specifiedColor=<rgba 0 0 0 0>}
      cornerRadii=<0 0 0 0>
     }
     tintColor {colorUseRule=UseColor_Specified, specifiedColor=<rgba 1 1 1 1>}
     uVRegion {bIsValid=false, max=<0 0>, min=<0 0>}
    }
    padding=<ltrb 0 0 0 0>
   }
   slot {
    class=HorizontalBoxSlot
    properties {horizontalAlignment=HAlign_Left, verticalAlignment=VAlign_Center, padding=<ltrb 10 10 10 10>, size {sizeRule=Fill, value=1}}
   }
  }
  Text_SocialName {
   widgetClassPath=/Script/SocialUMG.SocialNameTextBlock
   bInherited=true
   bIsVariable=true
   parent=SocialName_Color
   widget=Text_SocialName
   properties {
    bOverrideDefaultStyle=true
    text=Social Display Name
    defaultTextStyleOverride {
     overflowPolicy=Clip
     transformPolicy=None
     colorAndOpacity {colorUseRule=UseColor_Specified, specifiedColor=<rgba 1 1 1 1>}
     font {
      bForceMonospaced=false
      bMaterialIsStencil=false
      fontMaterial=null
      fontObject=&Burbank
      letterSpacing=0
      monospacedWidth=1
      size=18
      skewAmount=0
      typefaceFontName=BurbankRegular-Black
      outlineSettings {
       bApplyOutlineToDropShadows=false
       bMiteredCorners=false
       bSeparateFillAlpha=false
       outlineMaterial=null
       outlineSize=0
       outlineColor=<rgba 0 0 0 1>
      }
     }
     highlightColor {colorUseRule=UseColor_Specified, specifiedColor=<rgba 0 0 0 1>}
     highlightShape {
      bIsDynamicallyLoaded=false
      drawAs=Image
      imageType=NoImage
      mirroring=NoMirror
      resourceName=None
      resourceObject=null
      tiling=NoTile
      imageSize=<32 32>
      margin=<ltrb 0 0 0 0>
      outlineSettings {
       bUseBrushTransparency=false
       roundingType=HalfHeightRadius
       width=0
       color {colorUseRule=UseColor_Specified, specifiedColor=<rgba 0 0 0 0>}
       cornerRadii=<0 0 0 0>
      }
      tintColor {colorUseRule=UseColor_Specified, specifiedColor=<rgba 1 1 1 1>}
      uVRegion {bIsValid=false, max=<0 0>, min=<0 0>}
     }
     selectedBackgroundColor {colorUseRule=UseColor_Foreground, specifiedColor=<rgba 1 0 1 1>}
     shadowColorAndOpacity=<rgba 0 0 0 1>
     shadowOffset=<0 0>
     strikeBrush {
      bIsDynamicallyLoaded=false
      drawAs=Image
      imageType=NoImage
      mirroring=NoMirror
      resourceName=None
      resourceObject=null
      tiling=NoTile
      imageSize=<32 32>
      margin=<ltrb 0 0 0 0>
      outlineSettings {
       bUseBrushTransparency=false
       roundingType=HalfHeightRadius
       width=0
       color {colorUseRule=UseColor_Specified, specifiedColor=<rgba 0 0 0 0>}
       cornerRadii=<0 0 0 0>
      }
      tintColor {colorUseRule=UseColor_Specified, specifiedColor=<rgba 1 1 1 1>}
      uVRegion {bIsValid=false, max=<0 0>, min=<0 0>}
     }
     underlineBrush {
      bIsDynamicallyLoaded=false
      drawAs=Image
      imageType=NoImage
      mirroring=NoMirror
      resourceName=None
      resourceObject=null
      tiling=NoTile
      imageSize=<32 32>
      margin=<ltrb 0 0 0 0>
      outlineSettings {
       bUseBrushTransparency=false
       roundingType=HalfHeightRadius
       width=0
       color {colorUseRule=UseColor_Specified, specifiedColor=<rgba 0 0 0 0>}
       cornerRadii=<0 0 0 0>
      }
      tintColor {colorUseRule=UseColor_Specified, specifiedColor=<rgba 1 1 1 1>}
      uVRegion {bIsValid=false, max=<0 0>, min=<0 0>}
     }
    }
   }
   slot {class=BorderSlot, properties {padding=<ltrb 0 0 0 0>}}
  }
  Overlay_2 {
   widgetClassPath=&Overlay
   parent=HB_NameAndConfirm
   widget=Overlay_2
   slot {class=HorizontalBoxSlot, properties {horizontalAlignment=HAlign_Right}}
  }
  VisibilityWidget_ConfirmButton {
   widgetClassPath=&CommonVisibilityWidgetLegacy
   parent=Overlay_2
   widget=VisibilityWidget_ConfirmButton
   properties {bShowForGamepad=false, bShowForMouseAndKeyboard=false, visibleType=Visible, padding=<ltrb 0 0 0 0>}
   slot {class=OverlaySlot, properties {horizontalAlignment=HAlign_Fill, verticalAlignment=VAlign_Fill}}
  }
  SizeBox_0 {
   widgetClassPath=&SizeBox
   parent=VisibilityWidget_ConfirmButton
   widget=SizeBox_0
   properties {bOverride_WidthOverride=true, widthOverride=220}
   slot {class=BorderSlot, properties {padding=<ltrb 0 0 0 0>}}
  }
  ScaleBox_2 {widgetClassPath=&ScaleBox, parent=SizeBox_0, widget=ScaleBox_2, properties {stretchDirection=DownOnly}, slot {class=SizeBoxSlot}}
  Border_MobileConfirm {
   widgetClassPath=&Border
   parent=ScaleBox_2
   widget=Border_MobileConfirm
   properties {
    horizontalAlignment=HAlign_Center
    renderOpacity=0
    verticalAlignment=VAlign_Center
    background {
     bIsDynamicallyLoaded=false
     drawAs=Image
     imageType=NoImage
     mirroring=NoMirror
     resourceName=None
     resourceObject=/Gifting/UI/Art/M_UI_GiftingMobileConfirmBkg
     tiling=NoTile
     imageSize=<32 32>
     margin=<ltrb 0 0 0 0>
     outlineSettings {
      bUseBrushTransparency=false
      roundingType=HalfHeightRadius
      width=0
      color {colorUseRule=UseColor_Specified, specifiedColor=<rgba 0 0 0 0>}
      cornerRadii=<0 0 0 0>
     }
     tintColor {colorUseRule=UseColor_Specified, specifiedColor=<rgba 1 1 1 1>}
     uVRegion {bIsValid=false, max=<0 0>, min=<0 0>}
    }
    padding=<ltrb 6 0 0 0>
   }
   slot {class=ScaleBoxSlot, properties {horizontalAlignment=HAlign_Fill, verticalAlignment=VAlign_Fill}}
  }
  TextColorBorder {
   widgetClassPath=&Border
   parent=Border_MobileConfirm
   widget=TextColorBorder
   properties {
    background {
     bIsDynamicallyLoaded=false
     drawAs=NoDrawType
     imageType=NoImage
     mirroring=NoMirror
     resourceName=None
     resourceObject=null
     tiling=NoTile
     imageSize=<32 32>
     margin=<ltrb 0 0 0 0>
     outlineSettings {
      bUseBrushTransparency=false
      roundingType=HalfHeightRadius
      width=0
      color {colorUseRule=UseColor_Specified, specifiedColor=<rgba 0 0 0 0>}
      cornerRadii=<0 0 0 0>
     }
     tintColor {colorUseRule=UseColor_Specified, specifiedColor=<rgba 1 1 1 1>}
     uVRegion {bIsValid=false, max=<0 0>, min=<0 0>}
    }
    brushColor=<rgba 0 0.147 0.063 1>
    contentColorAndOpacity=<rgba 0 0.147 0.063 1>
    padding=<ltrb 0 0 0 0>
   }
   slot {class=BorderSlot, properties {horizontalAlignment=HAlign_Center, verticalAlignment=VAlign_Center, padding=<ltrb 6 0 0 0>}}
  }
  Text_MobileConfirm {
   widgetClassPath=&CommonRichTextBlock
   bIsVariable=true
   parent=TextColorBorder
   widget=Text_MobileConfirm
   properties {
    bOverrideDefaultStyle=true
    renderOpacity=0
    text=CONFIRM
    defaultTextStyleOverride {
     overflowPolicy=Clip
     transformPolicy=None
     colorAndOpacity {colorUseRule=UseColor_Specified, specifiedColor=<rgba 1 1 1 1>}
     font {
      bForceMonospaced=false
      bMaterialIsStencil=false
      fontMaterial=null
      fontObject=&Burbank
      letterSpacing=0
      monospacedWidth=1
      size=21
      skewAmount=0
      typefaceFontName=BurbankRegular-Black
      outlineSettings {
       bApplyOutlineToDropShadows=false
       bMiteredCorners=false
       bSeparateFillAlpha=false
       outlineMaterial=null
       outlineSize=0
       outlineColor=<rgba 0 0 0 1>
      }
     }
     highlightColor {colorUseRule=UseColor_Specified, specifiedColor=<rgba 0 0 0 1>}
     highlightShape {
      bIsDynamicallyLoaded=false
      drawAs=Image
      imageType=NoImage
      mirroring=NoMirror
      resourceName=None
      resourceObject=null
      tiling=NoTile
      imageSize=<32 32>
      margin=<ltrb 0 0 0 0>
      outlineSettings {
       bUseBrushTransparency=false
       roundingType=HalfHeightRadius
       width=0
       color {colorUseRule=UseColor_Specified, specifiedColor=<rgba 0 0 0 0>}
       cornerRadii=<0 0 0 0>
      }
      tintColor {colorUseRule=UseColor_Specified, specifiedColor=<rgba 1 1 1 1>}
      uVRegion {bIsValid=false, max=<0 0>, min=<0 0>}
     }
     selectedBackgroundColor {colorUseRule=UseColor_Foreground, specifiedColor=<rgba 1 0 1 1>}
     shadowColorAndOpacity=<rgba 0 0 0 1>
     shadowOffset=<0 0>
     strikeBrush {
      bIsDynamicallyLoaded=false
      drawAs=Image
      imageType=NoImage
      mirroring=NoMirror
      resourceName=None
      resourceObject=null
      tiling=NoTile
      imageSize=<32 32>
      margin=<ltrb 0 0 0 0>
      outlineSettings {
       bUseBrushTransparency=false
       roundingType=HalfHeightRadius
       width=0
       color {colorUseRule=UseColor_Specified, specifiedColor=<rgba 0 0 0 0>}
       cornerRadii=<0 0 0 0>
      }
      tintColor {colorUseRule=UseColor_Specified, specifiedColor=<rgba 1 1 1 1>}
      uVRegion {bIsValid=false, max=<0 0>, min=<0 0>}
     }
     underlineBrush {
      bIsDynamicallyLoaded=false
      drawAs=Image
      imageType=NoImage
      mirroring=NoMirror
      resourceName=None
      resourceObject=null
      tiling=NoTile
      imageSize=<32 32>
      margin=<ltrb 0 0 0 0>
      outlineSettings {
       bUseBrushTransparency=false
       roundingType=HalfHeightRadius
       width=0
       color {colorUseRule=UseColor_Specified, specifiedColor=<rgba 0 0 0 0>}
       cornerRadii=<0 0 0 0>
      }
      tintColor {colorUseRule=UseColor_Specified, specifiedColor=<rgba 1 1 1 1>}
      uVRegion {bIsValid=false, max=<0 0>, min=<0 0>}
     }
    }
    renderTransform {angle=0, scale=<1 1>, shear=<-12 0>, translation=<0 0>}
   }
   slot {class=BorderSlot, properties {padding=<ltrb 0 0 0 0>}}
  }
  Overlay_0 {
   widgetClassPath=&Overlay
   parent=VB_RecipientTextInfo
   widget=Overlay_0
   slot {class=VerticalBoxSlot, properties {size {sizeRule=Fill, value=1}}}
  }
  OwnedStatus_SizeBox {
   widgetClassPath=&SizeBox
   parent=Overlay_0
   widget=OwnedStatus_SizeBox
   properties {bOverride_HeightOverride=true, heightOverride=36, renderTransformPivot=<0.5 1>}
   slot {class=OverlaySlot, properties {horizontalAlignment=HAlign_Fill, verticalAlignment=VAlign_Bottom}}
  }
  Switcher_OwningStatus {
   widgetClassPath=&CommonVisibilitySwitcher
   bIsVariable=true
   parent=OwnedStatus_SizeBox
   widget=Switcher_OwningStatus
   properties {activeWidgetIndex=3}
   slot {class=SizeBoxSlot}
  }
  TileState_Loading {widgetClassPath=&Overlay, parent=Switcher_OwningStatus, widget=TileState_Loading, slot {class=CommonVisibilitySwitcherSlot}}
  Loading_Overlay {
   widgetClassPath=&Overlay
   parent=TileState_Loading
   widget=Loading_Overlay
   slot {class=OverlaySlot, properties {horizontalAlignment=HAlign_Fill, verticalAlignment=VAlign_Fill}}
  }
  Loading_Background {
   widgetClassPath=&Image
   bIsVariable=true
   parent=Loading_Overlay
   widget=Loading_Background
   properties {colorAndOpacity=<rgba 0.0052 0.0561 0.1779 1>}
   slot {class=OverlaySlot, properties {horizontalAlignment=HAlign_Fill, verticalAlignment=VAlign_Fill}}
  }
  HorizontalBox_0 {
   widgetClassPath=&HorizontalBox
   parent=Loading_Overlay
   widget=HorizontalBox_0
   slot {class=OverlaySlot, properties {horizontalAlignment=HAlign_Fill, verticalAlignment=VAlign_Center, padding=<ltrb 10 0 0 0>}}
  }
  RichText_OwnedItems_1 {
   widgetClassPath=&CommonRichTextBlock
   bIsVariable=true
   parent=HorizontalBox_0
   widget=RichText_OwnedItems_1
   properties {
    bOverrideDefaultStyle=true
    text=GETTING INFO...
    textTransformPolicy=ToUpper
    defaultTextStyleOverride {
     overflowPolicy=Clip
     transformPolicy=None
     colorAndOpacity {colorUseRule=UseColor_Specified, specifiedColor=<rgba 0.521 0.624 0.7231 1>}
     font {
      bForceMonospaced=false
      bMaterialIsStencil=false
      fontMaterial=null
      fontObject=&BurbankSmall
      letterSpacing=0
      monospacedWidth=1
      size=16
      skewAmount=0
      typefaceFontName=Black
      outlineSettings {
       bApplyOutlineToDropShadows=false
       bMiteredCorners=false
       bSeparateFillAlpha=false
       outlineMaterial=null
       outlineSize=0
       outlineColor=<rgba 0 0 0 1>
      }
     }
     highlightColor {colorUseRule=UseColor_Specified, specifiedColor=<rgba 0 0 0 1>}
     highlightShape {
      bIsDynamicallyLoaded=false
      drawAs=Image
      imageType=NoImage
      mirroring=NoMirror
      resourceName=None
      resourceObject=null
      tiling=NoTile
      imageSize=<32 32>
      margin=<ltrb 0 0 0 0>
      outlineSettings {
       bUseBrushTransparency=false
       roundingType=HalfHeightRadius
       width=0
       color {colorUseRule=UseColor_Specified, specifiedColor=<rgba 0 0 0 0>}
       cornerRadii=<0 0 0 0>
      }
      tintColor {colorUseRule=UseColor_Specified, specifiedColor=<rgba 1 1 1 1>}
      uVRegion {bIsValid=false, max=<0 0>, min=<0 0>}
     }
     selectedBackgroundColor {colorUseRule=UseColor_Foreground, specifiedColor=<rgba 1 0 1 1>}
     shadowColorAndOpacity=<rgba 0 0 0 1>
     shadowOffset=<0 0>
     strikeBrush {
      bIsDynamicallyLoaded=false
      drawAs=Image
      imageType=NoImage
      mirroring=NoMirror
      resourceName=None
      resourceObject=null
      tiling=NoTile
      imageSize=<32 32>
      margin=<ltrb 0 0 0 0>
      outlineSettings {
       bUseBrushTransparency=false
       roundingType=HalfHeightRadius
       width=0
       color {colorUseRule=UseColor_Specified, specifiedColor=<rgba 0 0 0 0>}
       cornerRadii=<0 0 0 0>
      }
      tintColor {colorUseRule=UseColor_Specified, specifiedColor=<rgba 1 1 1 1>}
      uVRegion {bIsValid=false, max=<0 0>, min=<0 0>}
     }
     underlineBrush {
      bIsDynamicallyLoaded=false
      drawAs=Image
      imageType=NoImage
      mirroring=NoMirror
      resourceName=None
      resourceObject=null
      tiling=NoTile
      imageSize=<32 32>
      margin=<ltrb 0 0 0 0>
      outlineSettings {
       bUseBrushTransparency=false
       roundingType=HalfHeightRadius
       width=0
       color {colorUseRule=UseColor_Specified, specifiedColor=<rgba 0 0 0 0>}
       cornerRadii=<0 0 0 0>
      }
      tintColor {colorUseRule=UseColor_Specified, specifiedColor=<rgba 1 1 1 1>}
      uVRegion {bIsValid=false, max=<0 0>, min=<0 0>}
     }
    }
    renderTransform {angle=0, scale=<1 1>, shear=<-12 0>, translation=<0 0>}
   }
   slot {class=HorizontalBoxSlot, properties {verticalAlignment=VAlign_Center}}
  }
  ScaleBox_1 {
   widgetClassPath=&ScaleBox
   parent=HorizontalBox_0
   widget=ScaleBox_1
   properties {renderTransform {angle=0, scale=<0.8 0.8>, shear=<0 0>, translation=<0 0>}}
   slot {
    class=HorizontalBoxSlot
    properties {horizontalAlignment=HAlign_Right, verticalAlignment=VAlign_Center, padding=<ltrb 0 0 20 0>, size {sizeRule=Fill, value=1}}
   }
  }
  ImageSpinner {
   widgetClassPath=&Image
   parent=ScaleBox_1
   widget=ImageSpinner
   properties {
    visibility=SelfHitTestInvisible
    brush {
     bIsDynamicallyLoaded=false
     drawAs=Image
     imageType=NoImage
     mirroring=NoMirror
     resourceName=None
     resourceObject=/FortUILibrary/UIKit/Throbber/MI_UIKit_Monochrome_Throbber
     tiling=NoTile
     imageSize=<42 42>
     margin=<ltrb 0 0 0 0>
     outlineSettings {
      bUseBrushTransparency=false
      roundingType=HalfHeightRadius
      width=0
      color {colorUseRule=UseColor_Specified, specifiedColor=<rgba 0 0 0 0>}
      cornerRadii=<0 0 0 0>
     }
     tintColor {colorUseRule=UseColor_Specified, specifiedColor=<rgba 1 1 1 1>}
     uVRegion {bIsValid=false, max=<0 0>, min=<0 0>}
    }
   }
   slot {class=ScaleBoxSlot}
  }
  TileState_NotOwned {
   widgetClassPath=&Overlay
   bIsVariable=true
   parent=Switcher_OwningStatus
   widget=TileState_NotOwned
   slot {class=CommonVisibilitySwitcherSlot}
  }
  Price_BackingColorFiller {
   widgetClassPath=&Image
   bIsVariable=true
   parent=TileState_NotOwned
   widget=Price_BackingColorFiller
   properties {colorAndOpacity=<rgba 0.0052 0.0561 0.1779 1>}
   slot {class=OverlaySlot, properties {horizontalAlignment=HAlign_Fill, verticalAlignment=VAlign_Fill}}
  }
  Widget_Price {
   widgetClassPath=&/RecipientGiftingPrice
   bInherited=true
   bIsVariable=true
   parent=TileState_NotOwned
   widget=Widget_Price
   properties {clipping=ClipToBoundsAlways}
   slot {class=OverlaySlot, properties {horizontalAlignment=HAlign_Right, verticalAlignment=VAlign_Center}}
  }
  OwnedItems_ColorBorder {
   widgetClassPath=&Border
   parent=TileState_NotOwned
   widget=OwnedItems_ColorBorder
   properties {
    background {
     bIsDynamicallyLoaded=false
     drawAs=NoDrawType
     imageType=NoImage
     mirroring=NoMirror
     resourceName=None
     resourceObject=null
     tiling=NoTile
     imageSize=<32 32>
     margin=<ltrb 0 0 0 0>
     outlineSettings {
      bUseBrushTransparency=false
      roundingType=HalfHeightRadius
      width=0
      color {colorUseRule=UseColor_Specified, specifiedColor=<rgba 0 0 0 0>}
      cornerRadii=<0 0 0 0>
     }
     tintColor {colorUseRule=UseColor_Specified, specifiedColor=<rgba 1 1 1 1>}
     uVRegion {bIsValid=false, max=<0 0>, min=<0 0>}
    }
    padding=<ltrb 0 0 0 0>
   }
   slot {class=OverlaySlot, properties {verticalAlignment=VAlign_Center, padding=<ltrb 10 0 0 0>}}
  }
  RichText_OwnedItems {
   widgetClassPath=&CommonRichTextBlock
   bInherited=true
   bIsVariable=true
   parent=OwnedItems_ColorBorder
   widget=RichText_OwnedItems
   properties {
    bOverrideDefaultStyle=true
    text="<text color=\"fff\">5 </>/ 10 items owned"
    textTransformPolicy=ToUpper
    defaultTextStyleOverride {
     overflowPolicy=Clip
     transformPolicy=None
     colorAndOpacity {colorUseRule=UseColor_Specified, specifiedColor=<rgba 0.521 0.624 0.7231 1>}
     font {
      bForceMonospaced=false
      bMaterialIsStencil=false
      fontMaterial=null
      fontObject=&BurbankSmall
      letterSpacing=0
      monospacedWidth=1
      size=16
      skewAmount=0
      typefaceFontName=Black
      outlineSettings {
       bApplyOutlineToDropShadows=false
       bMiteredCorners=false
       bSeparateFillAlpha=false
       outlineMaterial=null
       outlineSize=0
       outlineColor=<rgba 0 0 0 1>
      }
     }
     highlightColor {colorUseRule=UseColor_Specified, specifiedColor=<rgba 0 0 0 1>}
     highlightShape {
      bIsDynamicallyLoaded=false
      drawAs=Image
      imageType=NoImage
      mirroring=NoMirror
      resourceName=None
      resourceObject=null
      tiling=NoTile
      imageSize=<32 32>
      margin=<ltrb 0 0 0 0>
      outlineSettings {
       bUseBrushTransparency=false
       roundingType=HalfHeightRadius
       width=0
       color {colorUseRule=UseColor_Specified, specifiedColor=<rgba 0 0 0 0>}
       cornerRadii=<0 0 0 0>
      }
      tintColor {colorUseRule=UseColor_Specified, specifiedColor=<rgba 1 1 1 1>}
      uVRegion {bIsValid=false, max=<0 0>, min=<0 0>}
     }
     selectedBackgroundColor {colorUseRule=UseColor_Foreground, specifiedColor=<rgba 1 0 1 1>}
     shadowColorAndOpacity=<rgba 0 0 0 1>
     shadowOffset=<0 0>
     strikeBrush {
      bIsDynamicallyLoaded=false
      drawAs=Image
      imageType=NoImage
      mirroring=NoMirror
      resourceName=None
      resourceObject=null
      tiling=NoTile
      imageSize=<32 32>
      margin=<ltrb 0 0 0 0>
      outlineSettings {
       bUseBrushTransparency=false
       roundingType=HalfHeightRadius
       width=0
       color {colorUseRule=UseColor_Specified, specifiedColor=<rgba 0 0 0 0>}
       cornerRadii=<0 0 0 0>
      }
      tintColor {colorUseRule=UseColor_Specified, specifiedColor=<rgba 1 1 1 1>}
      uVRegion {bIsValid=false, max=<0 0>, min=<0 0>}
     }
     underlineBrush {
      bIsDynamicallyLoaded=false
      drawAs=Image
      imageType=NoImage
      mirroring=NoMirror
      resourceName=None
      resourceObject=null
      tiling=NoTile
      imageSize=<32 32>
      margin=<ltrb 0 0 0 0>
      outlineSettings {
       bUseBrushTransparency=false
       roundingType=HalfHeightRadius
       width=0
       color {colorUseRule=UseColor_Specified, specifiedColor=<rgba 0 0 0 0>}
       cornerRadii=<0 0 0 0>
      }
      tintColor {colorUseRule=UseColor_Specified, specifiedColor=<rgba 1 1 1 1>}
      uVRegion {bIsValid=false, max=<0 0>, min=<0 0>}
     }
    }
    renderTransform {angle=0, scale=<1 1>, shear=<-12 0>, translation=<0 0>}
   }
   slot {class=BorderSlot, properties {padding=<ltrb 0 0 0 0>}}
  }
  TileState_Owned {
   widgetClassPath=&Overlay
   bIsVariable=true
   parent=Switcher_OwningStatus
   widget=TileState_Owned
   slot {class=CommonVisibilitySwitcherSlot}
  }
  OwnedBacking {
   widgetClassPath=&Image
   bIsVariable=true
   parent=TileState_Owned
   widget=OwnedBacking
   properties {colorAndOpacity=<rgba 0.0037 0.075 0.0394 1>}
   slot {class=OverlaySlot, properties {horizontalAlignment=HAlign_Fill, verticalAlignment=VAlign_Fill}}
  }
  Text_Name_2 {
   widgetClassPath=&CommonTextBlock
   bIsVariable=true
   parent=TileState_Owned
   widget=Text_Name_2
   properties {
    justification=Center
    text=Owned
    textTransformPolicy=ToUpper
    font {
     bForceMonospaced=false
     bMaterialIsStencil=false
     fontMaterial=/FortUILibrary/Systems/ItemShopTile/Materials/M_UI_ItemShop_Text_Price_Owned
     fontObject=&Burbank
     letterSpacing=0
     monospacedWidth=1
     size=16
     skewAmount=0
     typefaceFontName=BurbankRegular-Black
     outlineSettings {
      bApplyOutlineToDropShadows=false
      bMiteredCorners=false
      bSeparateFillAlpha=false
      outlineMaterial=null
      outlineSize=0
      outlineColor=<rgba 0 0 0 1>
     }
    }
    renderTransform {angle=0, scale=<1 1>, shear=<-8 0>, translation=<0 0>}
    shadowOffset=<0 0>
   }
   slot {class=OverlaySlot, properties {horizontalAlignment=HAlign_Right, verticalAlignment=VAlign_Center, padding=<ltrb 0 0 10 0>}}
  }
  TileState_Unavailable {
   widgetClassPath=&Overlay
   bIsVariable=true
   parent=Switcher_OwningStatus
   widget=TileState_Unavailable
   slot {class=CommonVisibilitySwitcherSlot}
  }
  Border_99 {
   widgetClassPath=&Border
   parent=TileState_Unavailable
   widget=Border_99
   properties {brushColor=<rgba 0.0052 0.0561 0.1779 1>, contentColorAndOpacity=<rgba 0.0086 0 0.1094 1>, padding=<ltrb 0 0 0 0>}
   slot {class=OverlaySlot, properties {horizontalAlignment=HAlign_Fill, verticalAlignment=VAlign_Fill}}
  }
  RichText_EligibilityStatus {
   widgetClassPath=&CommonRichTextBlock
   bInherited=true
   bIsVariable=true
   parent=TileState_Unavailable
   widget=RichText_EligibilityStatus
   properties {
    bOverrideDefaultStyle=true
    text=Unavailable
    textTransformPolicy=ToUpper
    defaultTextStyleOverride {
     overflowPolicy=Clip
     transformPolicy=None
     colorAndOpacity {colorUseRule=UseColor_Specified, specifiedColor=<rgba 0.521 0.624 0.7231 1>}
     font {
      bForceMonospaced=false
      bMaterialIsStencil=false
      fontMaterial=null
      fontObject=&BurbankSmall
      letterSpacing=0
      monospacedWidth=1
      size=16
      skewAmount=0
      typefaceFontName=Black
      outlineSettings {
       bApplyOutlineToDropShadows=false
       bMiteredCorners=false
       bSeparateFillAlpha=false
       outlineMaterial=null
       outlineSize=0
       outlineColor=<rgba 0 0 0 1>
      }
     }
     highlightColor {colorUseRule=UseColor_Specified, specifiedColor=<rgba 0 0 0 1>}
     highlightShape {
      bIsDynamicallyLoaded=false
      drawAs=Image
      imageType=NoImage
      mirroring=NoMirror
      resourceName=None
      resourceObject=null
      tiling=NoTile
      imageSize=<32 32>
      margin=<ltrb 0 0 0 0>
      outlineSettings {
       bUseBrushTransparency=false
       roundingType=HalfHeightRadius
       width=0
       color {colorUseRule=UseColor_Specified, specifiedColor=<rgba 0 0 0 0>}
       cornerRadii=<0 0 0 0>
      }
      tintColor {colorUseRule=UseColor_Specified, specifiedColor=<rgba 1 1 1 1>}
      uVRegion {bIsValid=false, max=<0 0>, min=<0 0>}
     }
     selectedBackgroundColor {colorUseRule=UseColor_Foreground, specifiedColor=<rgba 1 0 1 1>}
     shadowColorAndOpacity=<rgba 0 0 0 1>
     shadowOffset=<0 0>
     strikeBrush {
      bIsDynamicallyLoaded=false
      drawAs=Image
      imageType=NoImage
      mirroring=NoMirror
      resourceName=None
      resourceObject=null
      tiling=NoTile
      imageSize=<32 32>
      margin=<ltrb 0 0 0 0>
      outlineSettings {
       bUseBrushTransparency=false
       roundingType=HalfHeightRadius
       width=0
       color {colorUseRule=UseColor_Specified, specifiedColor=<rgba 0 0 0 0>}
       cornerRadii=<0 0 0 0>
      }
      tintColor {colorUseRule=UseColor_Specified, specifiedColor=<rgba 1 1 1 1>}
      uVRegion {bIsValid=false, max=<0 0>, min=<0 0>}
     }
     underlineBrush {
      bIsDynamicallyLoaded=false
      drawAs=Image
      imageType=NoImage
      mirroring=NoMirror
      resourceName=None
      resourceObject=null
      tiling=NoTile
      imageSize=<32 32>
      margin=<ltrb 0 0 0 0>
      outlineSettings {
       bUseBrushTransparency=false
       roundingType=HalfHeightRadius
       width=0
       color {colorUseRule=UseColor_Specified, specifiedColor=<rgba 0 0 0 0>}
       cornerRadii=<0 0 0 0>
      }
      tintColor {colorUseRule=UseColor_Specified, specifiedColor=<rgba 1 1 1 1>}
      uVRegion {bIsValid=false, max=<0 0>, min=<0 0>}
     }
    }
    renderTransform {angle=0, scale=<1 1>, shear=<-12 0>, translation=<0 0>}
   }
   slot {class=OverlaySlot, properties {verticalAlignment=VAlign_Center, padding=<ltrb 10 0 0 0>}}
  }
  Border_UnavailableOverlay {
   widgetClassPath=&Border
   parent=Overlay_1
   widget=Border_UnavailableOverlay
   properties {
    background {
     bIsDynamicallyLoaded=false
     drawAs=NoDrawType
     imageType=NoImage
     mirroring=NoMirror
     resourceName=None
     resourceObject=null
     tiling=NoTile
     imageSize=<32 32>
     margin=<ltrb 0 0 0 0>
     outlineSettings {
      bUseBrushTransparency=false
      roundingType=HalfHeightRadius
      width=0
      color {colorUseRule=UseColor_Specified, specifiedColor=<rgba 0 0 0 0>}
      cornerRadii=<0 0 0 0>
     }
     tintColor {colorUseRule=UseColor_Specified, specifiedColor=<rgba 1 1 1 1>}
     uVRegion {bIsValid=false, max=<0 0>, min=<0 0>}
    }
    padding=<ltrb 0 0 0 0>
   }
   slot {class=OverlaySlot, properties {horizontalAlignment=HAlign_Fill, verticalAlignment=VAlign_Fill, padding=<ltrb 8 0 8 0>}}
  }
  Unavailable_OverlayImage {
   widgetClassPath=&Image
   parent=Border_UnavailableOverlay
   widget=Unavailable_OverlayImage
   properties {renderOpacity=0, colorAndOpacity=<rgba 0 0 0 0.5>}
   slot {class=BorderSlot, properties {padding=<ltrb 0 0 0 0>}}
  }
  Border-HoverOverlay {
   widgetClassPath=&Border
   parent=Overlay_1
   widget=Border-HoverOverlay
   properties {
    background {
     bIsDynamicallyLoaded=false
     drawAs=NoDrawType
     imageType=NoImage
     mirroring=NoMirror
     resourceName=None
     resourceObject=null
     tiling=NoTile
     imageSize=<32 32>
     margin=<ltrb 0 0 0 0>
     outlineSettings {
      bUseBrushTransparency=false
      roundingType=HalfHeightRadius
      width=0
      color {colorUseRule=UseColor_Specified, specifiedColor=<rgba 0 0 0 0>}
      cornerRadii=<0 0 0 0>
     }
     tintColor {colorUseRule=UseColor_Specified, specifiedColor=<rgba 1 1 1 1>}
     uVRegion {bIsValid=false, max=<0 0>, min=<0 0>}
    }
    padding=<ltrb 0 0 0 0>
   }
   slot {class=OverlaySlot, properties {horizontalAlignment=HAlign_Fill, verticalAlignment=VAlign_Fill, padding=<ltrb 8 0 8 0>}}
  }
  Hover_OverlayImage {
   widgetClassPath=&Image
   parent=Border-HoverOverlay
   widget=Hover_OverlayImage
   properties {renderOpacity=0, colorAndOpacity=<rgba 1 1 1 0.15>}
   slot {class=BorderSlot, properties {padding=<ltrb 0 0 0 0>}}
  }
  Overlay_3 {
   widgetClassPath=&Overlay
   parent=Overlay_1
   widget=Overlay_3
   properties {flowDirectionPreference=Culture}
   slot {class=OverlaySlot, properties {horizontalAlignment=HAlign_Fill, verticalAlignment=VAlign_Fill}}
  }
  InputActionWidget {
   widgetClassPath=&CommonActionWidget
   bInherited=true
   bIsVariable=true
   parent=Overlay_3
   widget=InputActionWidget
   properties {visibility=Collapsed, inputActions [{dataTable=&InputActionDataTable, rowName=GenericForward}]}
   slot {class=OverlaySlot, properties {horizontalAlignment=HAlign_Right, padding=<ltrb 0 10 20 0>}}
  }
 ]
}
