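# Search bar widget for gifting, built on FortGiftingUserSearchWidget, for entering a recipient display name with clear and submit buttons.
# Mouse enter and leave play and reverse a hover animation revealing the submit bubble and key prompt, skipping touch events.
# A switch to touch input also reverses it, and clicking the capture border focuses the text field.
# A red Error flash animation exists, but its play node is unconnected to the empty OnSearchCommit event.
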WidgetBlueprint /Gifting/UI/GiftRecipientSearchBar
animations [
 OnHovered {
  binding_count=7
  length_seconds=0.2
  object=&self:OnHovered
  sequence {bindings [
    HorizontalBox_6 {
     id=#1
     possessed_class=HorizontalBox
     tracks [{
       type=MovieScene2DTransformTrack
       sections [{
         type=MovieScene2DTransformSection
         end_frame=4
         start_frame=0
         channels [Translation.X {type=MovieSceneScriptingFloatChannel, keys [[0,0],[4,10]]}]
        }]
      }]
    }
    HoverBubble {
     id=#2
     possessed_class=SizeBox
     tracks [{
       type=MovieScene2DTransformTrack
       sections [{
         type=MovieScene2DTransformSection
         end_frame=4
         start_frame=0
         channels [
          Translation.X {type=MovieSceneScriptingFloatChannel, keys [[0,0],[2,8]]}
          Translation.Y {type=MovieSceneScriptingFloatChannel}
          Angle {type=MovieSceneScriptingFloatChannel}
          Scale.X {type=MovieSceneScriptingFloatChannel}
          Scale.Y {type=MovieSceneScriptingFloatChannel}
          Shear.X {type=MovieSceneScriptingFloatChannel}
          Shear.Y {type=MovieSceneScriptingFloatChannel}
         ]
        }]
      }]
    }
    EntryBorder {
     id=#3
     possessed_class=Border
     tracks [{
       type=MovieScene2DTransformTrack
       sections [{
         type=MovieScene2DTransformSection
         end_frame=4
         start_frame=0
         channels [
          Translation.X {type=MovieSceneScriptingFloatChannel, keys [[0,0],[2,-8]]}
          Translation.Y {type=MovieSceneScriptingFloatChannel}
          Angle {type=MovieSceneScriptingFloatChannel}
          Scale.X {type=MovieSceneScriptingFloatChannel}
          Scale.Y {type=MovieSceneScriptingFloatChannel}
          Shear.X {type=MovieSceneScriptingFloatChannel}
          Shear.Y {type=MovieSceneScriptingFloatChannel}
         ]
        }]
      }]
    }
    InputActionWidget {
     id=#4
     possessed_class=CommonActionWidget
     tracks [{
       type=MovieSceneFloatTrack
       display_name=Render Opacity
       sections [{type=MovieSceneFloatSection, end_frame=4, start_frame=0, channels [None {type=MovieSceneScriptingFloatChannel, keys [[0,0],[1,0],[2,1]]}]}]
      }]
    }
    AnimatedInputBorder {
     id=#5
     possessed_class=Border
     tracks [
      {
       type=MovieSceneColorTrack
       display_name=Brush Color
       sections [{
         type=MovieSceneColorSection
         end_frame=4
         start_frame=0
         channels [
          Color.R {type=MovieSceneScriptingFloatChannel}
          Color.G {type=MovieSceneScriptingFloatChannel}
          Color.B {type=MovieSceneScriptingFloatChannel}
          Color.A {type=MovieSceneScriptingFloatChannel, keys [[0,0],[2,1]]}
         ]
        }]
      }
      {
       type=MovieSceneWidgetMaterialTrack
       sections [{
         type=MovieSceneParameterSection
         end_frame=4
         start_frame=0
         channels [AnimateStateChange {type=MovieSceneScriptingFloatChannel, keys [[0,0],[3,1],[4,1]]}]
        }]
      }
     ]
    }
    TextPromptHB {
     id=#6
     possessed_class=VerticalBox
     tracks [{
       type=MovieScene2DTransformTrack
       sections [{
         type=MovieScene2DTransformSection
         end_frame=4
         start_frame=0
         channels [
          Translation.X {type=MovieSceneScriptingFloatChannel, keys [[0,0],[2,-8]]}
          Translation.Y {type=MovieSceneScriptingFloatChannel}
          Angle {type=MovieSceneScriptingFloatChannel}
          Scale.X {type=MovieSceneScriptingFloatChannel}
          Scale.Y {type=MovieSceneScriptingFloatChannel}
          Shear.X {type=MovieSceneScriptingFloatChannel}
          Shear.Y {type=MovieSceneScriptingFloatChannel}
         ]
        }]
      }]
    }
    SharingIcon {
     id=#7
     display_name=SearchIcon
     possessed_class=Image
     tracks [
      {
       type=MovieScene2DTransformTrack
       sections [{
         type=MovieScene2DTransformSection
         end_frame=4
         start_frame=0
         channels [
          Translation.X {type=MovieSceneScriptingFloatChannel}
          Translation.Y {type=MovieSceneScriptingFloatChannel}
          Angle {type=MovieSceneScriptingFloatChannel}
          Scale.X {type=MovieSceneScriptingFloatChannel, keys [[0,1],[3,1.125],[4,1]]}
          Scale.Y {type=MovieSceneScriptingFloatChannel, keys [[0,1],[3,1.125],[4,1]]}
          Shear.X {type=MovieSceneScriptingFloatChannel}
          Shear.Y {type=MovieSceneScriptingFloatChannel}
         ]
        }]
      }
      {
       type=MovieSceneFloatTrack
       display_name=Render Opacity
       sections [{type=MovieSceneFloatSection, end_frame=4, start_frame=0, channels [None {type=MovieSceneScriptingFloatChannel, keys [[0,0.5],[2,1]]}]}]
      }
     ]
    }
   ]}
 }
 Error {
  binding_count=1
  length_seconds=0.5
  object=&self:Error
  sequence {bindings [EditableText_SearchFriends {
     id=#8
     display_name=SearchFriends_Color
     possessed_class=FortMobileEditableText
     tracks [
      {
       type=MovieSceneColorTrack
       display_name=Content Color and Opacity
       sections [{
         type=MovieSceneColorSection
         end_frame=4
         start_frame=0
         channels [
          Color.R {type=MovieSceneScriptingFloatChannel, const [1,1]}
          Color.G {type=MovieSceneScriptingFloatChannel, const [0.1714,1]}
          Color.B {type=MovieSceneScriptingFloatChannel, const [0.1329,1]}
          Color.A {type=MovieSceneScriptingFloatChannel, const [1,1]}
         ]
        }]
      }
      {
       type=MovieSceneColorTrack
       display_name=Brush Color
       sections [{
         type=MovieSceneColorSection
         end_frame=10
         start_frame=0
         channels [
          Color.R {type=MovieSceneScriptingFloatChannel, const [1,1]}
          Color.G {type=MovieSceneScriptingFloatChannel, const [0.1714,1]}
          Color.B {type=MovieSceneScriptingFloatChannel, const [0.1329,1]}
          Color.A {type=MovieSceneScriptingFloatChannel, keys [[0,0.3],[4,0.6],[5,0.6],[10,0.3]]}
         ]
        }]
      }
     ]
    }]}
 }
]
graphs [
 EventGraph {nodes [
   n1 = HandleInputMethodChanged!() -> (OutputDelegate, then, bNewInputType:ECommonInputType Enum) {is=title K2Node_CustomEvent}
   > Branch(Condition="Equal (Enum)"(A=@n1.bNewInputType, B=Touch))
   > PlayAnimation(InAnimation=Get OnHovered() -> (OnHovered:Widget Animation Object Reference), StartAtTime=1, PlayMode=Reverse, PlaybackSpeed=2)
   n2 = Event OnMouseEnter()
   > Branch(Condition=NOT Boolean(A=Is Touch Event!(Input:"Pointer Event Structure (by ref)"=@n2.MouseEvent) -> (ReturnValue:bool)))
   > PlayAnimation(InAnimation=Get OnHovered())
   n3 = Event OnMouseLeave!() -> (OutputDelegate, then, MouseEvent:"Pointer Event Structure (by ref)") {is=title K2Node_Event}
   > Branch(Condition=NOT Boolean(A=Is Touch Event(Input=@n3.MouseEvent)))
   > PlayAnimation(InAnimation=Get OnHovered(), PlayMode=Reverse, PlaybackSpeed=2)
   Event OnInitialized()
   > Bind Event to On Input Method Changed(self=Get CommonInputSubsystem(), Delegate=@n1)
   PlayAnimationForward(InAnimation=Get Error() -> (Error:Widget Animation Object Reference))
   Event OnSearchCommit!() -> (OutputDelegate, then, bSearchStringShort:bool) {is=title K2Node_Event}
  ]}
 On_ClickCapture_MouseButtonDown {nodes [
   On_ClickCapture_MouseButtonDown!() -> (then, MyGeometry:Geometry Structure, MouseEvent:"Pointer Event Structure (by ref)") {is=title K2Node_FunctionEntry}
   > FocusEditableText(self:Fort Gifting User Search Widget Object Reference)
   > Return Node!(execute, ReturnValue:Event Reply Structure=Handled())
  ]}
]
variables []
widgets {
 info {inheritedWidgetCount=3, parentClass=/Script/GiftingUI.FortGiftingUserSearchWidget, rootWidgetClass=&SizeBox, widgetCount=21}
 widgets [
  SizeBox_FriendSearch {
   widgetClassPath=&SizeBox
   widget=SizeBox_FriendSearch
   properties {bOverride_HeightOverride=true, bOverride_WidthOverride=true, heightOverride=64, maxDesiredWidth=544, minDesiredHeight=64, widthOverride=544}
  }
  EntryBorder {
   widgetClassPath=&Border
   parent=SizeBox_FriendSearch
   widget=EntryBorder
   properties {
    background {
     bIsDynamicallyLoaded=false
     drawAs=Image
     imageType=NoImage
     mirroring=NoMirror
     resourceName=None
     resourceObject=/Gifting/UI/Art/M_UI_Gifting_SearchBordert
     tiling=NoTile
     imageSize=<32 32>
     margin=<ltrb 0 0 0 0>
     outlineSettings {
      bUseBrushTransparency=false
      roundingType=HalfHeightRadius
      width=0
      color {colorUseRule=UseColor_Specified, specifiedColor=<rgba 0 0 0 0>}
      cornerRadii=<0 0 0 0>
     }
     tintColor {colorUseRule=UseColor_Specified, specifiedColor=<rgba 1 1 1 1>}
     uVRegion {bIsValid=false, max=<0 0>, min=<0 0>}
    }
    padding=<ltrb 0 0 0 0>
   }
   slot {class=SizeBoxSlot, properties {padding=<ltrb 8 0 8 0>}}
  }
  HorizontalBox_1 {
   widgetClassPath=&HorizontalBox
   parent=EntryBorder
   widget=HorizontalBox_1
   slot {class=BorderSlot, properties {padding=<ltrb 0 0 0 0>}}
  }
  ClickCapture {
   widgetClassPath=&Border
   bIsVariable=true
   parent=HorizontalBox_1
   widget=ClickCapture
   properties {
    background {
     bIsDynamicallyLoaded=false
     drawAs=NoDrawType
     imageType=NoImage
     mirroring=NoMirror
     resourceName=None
     resourceObject=null
     tiling=NoTile
     imageSize=<32 32>
     margin=<ltrb 0 0 0 0>
     outlineSettings {
      bUseBrushTransparency=false
      roundingType=HalfHeightRadius
      width=0
      color {colorUseRule=UseColor_Specified, specifiedColor=<rgba 0 0 0 0>}
      cornerRadii=<0 0 0 0>
     }
     tintColor {colorUseRule=UseColor_Specified, specifiedColor=<rgba 0 0 0 1>}
     uVRegion {bIsValid=false, max=<0 0>, min=<0 0>}
    }
    brushColor=<rgba 1 1 1 0.4>
    padding=<ltrb 4 4 4 4>
   }
   slot {class=HorizontalBoxSlot, properties {size {sizeRule=Fill, value=1}}}
  }
  Border_1 {
   widgetClassPath=&Border
   parent=ClickCapture
   widget=Border_1
   properties {bOverride_Cursor=true, cursor=TextEditBeam, brushColor=<rgba 0.0052 0.0561 0.1779 1>, padding=<ltrb 10 0 10 0>}
   slot {class=BorderSlot, properties {padding=<ltrb 4 4 4 4>}}
  }
  TextPromptHB {
   widgetClassPath=&HorizontalBox
   parent=Border_1
   widget=TextPromptHB
   slot {class=BorderSlot, properties {padding=<ltrb 10 0 10 0>}}
  }
  SizeBox_1 {
   widgetClassPath=&SizeBox
   parent=TextPromptHB
   widget=SizeBox_1
   properties {maxDesiredWidth=344, widthOverride=285}
   slot {class=HorizontalBoxSlot, properties {size {sizeRule=Fill, value=1}}}
  }
  SearchFriends_Color {
   widgetClassPath=&CommonBorder
   parent=SizeBox_1
   widget=SearchFriends_Color
   properties {
    background {
     bIsDynamicallyLoaded=false
     drawAs=Box
     imageType=NoImage
     mirroring=NoMirror
     resourceName=None
     resourceObject=null
     tiling=NoTile
     imageSize=<32 32>
     margin=<ltrb 0 0 0 0>
     outlineSettings {
      bUseBrushTransparency=false
      roundingType=HalfHeightRadius
      width=0
      color {colorUseRule=UseColor_Specified, specifiedColor=<rgba 0 0 0 0>}
      cornerRadii=<0 0 0 0>
     }
     tintColor {colorUseRule=UseColor_Specified, specifiedColor=<rgba 1 1 1 1>}
     uVRegion {bIsValid=false, max=<0 0>, min=<0 0>}
    }
    brushColor=<rgba 1 0.1714 0.1329 0>
    padding=<ltrb 0 0 0 0>
   }
   slot {class=SizeBoxSlot, properties {verticalAlignment=VAlign_Center}}
  }
  EditableText_SearchFriends {
   widgetClassPath=/Script/FortniteUI.FortMobileEditableText
   bInherited=true
   bIsVariable=true
   parent=SearchFriends_Color
   widget=EditableText_SearchFriends
   properties {
    hintText=ENTER DISPLAY NAME
    mobileFontSizeMultiplier=1.15
    virtualKeyboardDismissAction=TextCommitOnAccept
    virtualKeyboardTrigger=OnAllFocusEvents
    widgetStyle {
     backgroundImageComposing {
      bIsDynamicallyLoaded=false
      drawAs=Border
      imageType=FullColor
      mirroring=NoMirror
      resourceName=../../../Engine/Content/Slate/Old/HyperlinkDotted.png
      resourceObject=null
      tiling=Both
      imageSize=<0 0>
      margin=<ltrb 0 0 0 0.1875>
      outlineSettings {
       bUseBrushTransparency=false
       roundingType=HalfHeightRadius
       width=0
       color {colorUseRule=UseColor_Specified, specifiedColor=<rgba 0 0 0 0>}
       cornerRadii=<0 0 0 0>
      }
      tintColor {colorUseRule=UseColor_Specified, specifiedColor=<rgba 1 1 1 1>}
      uVRegion {bIsValid=false, max=<0 0>, min=<0 0>}
     }
     backgroundImageSelected {
      bIsDynamicallyLoaded=false
      drawAs=Box
      imageType=FullColor
      mirroring=NoMirror
      resourceName=../../../Engine/Content/Slate/Common/EditableTextSelectionBackground.png
      resourceObject=null
      tiling=NoTile
      imageSize=<0 0>
      margin=<ltrb 0.25 0.25 0.25 0.25>
      outlineSettings {
       bUseBrushTransparency=false
       roundingType=HalfHeightRadius
       width=0
       color {colorUseRule=UseColor_Specified, specifiedColor=<rgba 0 0 0 0>}
       cornerRadii=<0 0 0 0>
      }
      tintColor {colorUseRule=UseColor_Specified, specifiedColor=<rgba 1 -1.0683 1 1>}
      uVRegion {bIsValid=false, max=<0 0>, min=<0 0>}
     }
     caretImage {
      bIsDynamicallyLoaded=false
      drawAs=Image
      imageType=FullColor
      mirroring=NoMirror
      resourceName=../../../Engine/Content/Slate/Old/White.png
      resourceObject=null
      tiling=NoTile
      imageSize=<16 16>
      margin=<ltrb 0 0 0 0>
      outlineSettings {
       bUseBrushTransparency=false
       roundingType=HalfHeightRadius
       width=0
       color {colorUseRule=UseColor_Specified, specifiedColor=<rgba 0 0 0 0>}
       cornerRadii=<0 0 0 0>
      }
      tintColor {colorUseRule=UseColor_Specified, specifiedColor=<rgba 1 1 1 1>}
      uVRegion {bIsValid=false, max=<0 0>, min=<0 0>}
     }
     colorAndOpacity {colorUseRule=UseColor_Specified, specifiedColor=<rgba 0 0.402 1 1>}
     font {
      bForceMonospaced=false
      bMaterialIsStencil=false
      fontMaterial=null
      fontObject=&BurbankSmall
      letterSpacing=0
      monospacedWidth=1
      size=17
      skewAmount=0
      typefaceFontName=Bold
      outlineSettings {
       bApplyOutlineToDropShadows=false
       bMiteredCorners=false
       bSeparateFillAlpha=false
       outlineMaterial=null
       outlineSize=0
       outlineColor=<rgba 0 0 0 1>
      }
     }
     hintColorAndOpacity {colorUseRule=UseColor_Foreground, specifiedColor=<rgba 1 0 1 1>}
     selectedBackgroundColor {colorUseRule=UseColor_Foreground, specifiedColor=<rgba 1 0 1 1>}
    }
   }
   slot {class=BorderSlot, properties {padding=<ltrb 0 0 0 0>}}
  }
  ClearButtonSB {
   widgetClassPath=&SizeBox
   parent=TextPromptHB
   widget=ClearButtonSB
   properties {bOverride_HeightOverride=true, bOverride_WidthOverride=true, heightOverride=42, widthOverride=42}
   slot {class=HorizontalBoxSlot, properties {horizontalAlignment=HAlign_Right, verticalAlignment=VAlign_Center}}
  }
  Button_ClearQuery {
   widgetClassPath=/Game/Athena/UI/Matchmaking/SimpleCommonButtonWithInput
   bInherited=true
   bIsVariable=true
   parent=ClearButtonSB
   widget=Button_ClearQuery
   properties {
    bConsumePointerInput=true
    bOverride_Cursor=true
    onButtonBaseClicked=(&self:WidgetTree.Button_ClearQuery.HandleOnButtonClicked)
    onButtonBaseDoubleClicked=(&self:WidgetTree.Button_ClearQuery.HandleOnButtonDoubleClicked)
    onButtonBaseHovered=(&self:WidgetTree.Button_ClearQuery.HandleOnButtonHovered)
    onButtonBaseUnhovered=(&self:WidgetTree.Button_ClearQuery.HandleOnButtonUnhovered)
    onSelectedChangedBase=(&self:WidgetTree.Button_ClearQuery.HandleOnSelectedChanged)
    style=/FortUILibrary/Foundation/Buttons/ButtonStyles/ButtonStyle_ClearSocialSearch.ButtonStyle_ClearSocialSearch_C
    visibility=Visible
    triggeringInputAction {dataTable=&InputActionDataTable, rowName=ClearAll}
   }
   slot {class=SizeBoxSlot}
  }
  HoverBubble {
   widgetClassPath=&SizeBox
   parent=HorizontalBox_1
   widget=HoverBubble
   properties {minDesiredWidth=96}
   slot {class=HorizontalBoxSlot, properties {horizontalAlignment=HAlign_Right, padding=<ltrb 16 0 0 0>}}
  }
  Overlay_4 {widgetClassPath=&Overlay, parent=HoverBubble, widget=Overlay_4, slot {class=SizeBoxSlot}}
  AnimatedInputBorder {
   widgetClassPath=&Border
   parent=Overlay_4
   widget=AnimatedInputBorder
   properties {
    verticalAlignment=VAlign_Top
    visibility=SelfHitTestInvisible
    background {
     bIsDynamicallyLoaded=false
     drawAs=Image
     imageType=NoImage
     mirroring=NoMirror
     resourceName=None
     resourceObject=/Gifting/UI/Art/M_UI_GiftingMobileConfirmBkg
     tiling=NoTile
     imageSize=<64 64>
     margin=<ltrb 1 0 0 0>
     outlineSettings {
      bUseBrushTransparency=false
      roundingType=HalfHeightRadius
      width=0
      color {colorUseRule=UseColor_Specified, specifiedColor=<rgba 0 0 0 0>}
      cornerRadii=<0 0 0 0>
     }
     tintColor {colorUseRule=UseColor_Specified, specifiedColor=<rgba 1 1 1 1>}
     uVRegion {bIsValid=false, max=<0 0>, min=<0 0>}
    }
    padding=<ltrb 0 0 0 0>
   }
   slot {class=OverlaySlot, properties {horizontalAlignment=HAlign_Fill, verticalAlignment=VAlign_Fill}}
  }
  Overlay_0 {
   widgetClassPath=&Overlay
   parent=AnimatedInputBorder
   widget=Overlay_0
   slot {class=BorderSlot, properties {verticalAlignment=VAlign_Top, padding=<ltrb 0 0 0 0>}}
  }
  Button_SubmitQuery {
   widgetClassPath=/Game/Athena/UI/Matchmaking/SimpleCommonButton
   bInherited=true
   bIsVariable=true
   parent=Overlay_0
   widget=Button_SubmitQuery
   properties {
    bIsFocusable=true
    onButtonBaseClicked=(&self:WidgetTree.Button_SubmitQuery.HandleOnButtonClicked)
    onButtonBaseDoubleClicked=(&self:WidgetTree.Button_SubmitQuery.HandleOnButtonDoubleClicked)
    onButtonBaseHovered=(&self:WidgetTree.Button_SubmitQuery.HandleOnButtonHovered)
    onButtonBaseUnhovered=(&self:WidgetTree.Button_SubmitQuery.HandleOnButtonUnhovered)
    onSelectedChangedBase=(&self:WidgetTree.Button_SubmitQuery.HandleOnSelectedChanged)
    style=/FortUILibrary/Foundation/Buttons/ButtonStyles/ButtonStyle-Clear.ButtonStyle-Clear_C
    triggeringInputAction {dataTable=/Gifting/GiftingInputActionDataTable, rowName=Search}
   }
   slot {class=OverlaySlot, properties {horizontalAlignment=HAlign_Fill, verticalAlignment=VAlign_Fill}}
  }
  HorizontalBox_6 {widgetClassPath=&HorizontalBox, parent=Overlay_0, widget=HorizontalBox_6, slot {class=OverlaySlot}}
  InputActionWidget {
   widgetClassPath=&CommonActionWidget
   bIsVariable=true
   parent=HorizontalBox_6
   widget=InputActionWidget
   properties {visibility=SelfHitTestInvisible, inputActions [{dataTable=&InputActionDataTable, rowName=Select}]}
   slot {class=HorizontalBoxSlot, properties {horizontalAlignment=HAlign_Left, verticalAlignment=VAlign_Center, padding=<ltrb 8 0 0 0>}}
  }
  Spacer_7 {
   widgetClassPath=&Spacer
   parent=HorizontalBox_6
   widget=Spacer_7
   properties {size=<8 8>}
   slot {class=HorizontalBoxSlot, properties {size {sizeRule=Fill, value=1}}}
  }
  IconSpacingSB {
   widgetClassPath=&SizeBox
   parent=HorizontalBox_6
   widget=IconSpacingSB
   properties {
    bOverride_MinDesiredHeight=true
    bOverride_MinDesiredWidth=true
    heightOverride=72
    minDesiredHeight=72
    minDesiredWidth=72
    widthOverride=72
   }
   slot {class=HorizontalBoxSlot, properties {horizontalAlignment=HAlign_Right, verticalAlignment=VAlign_Center}}
  }
  SearchIcon {
   widgetClassPath=/Script/FortniteUI.FortMobileImage
   bIsVariable=true
   parent=IconSpacingSB
   widget=SearchIcon
   properties {
    mobileSizeMultiplier=1.2
    renderOpacity=0.5
    visibility=SelfHitTestInvisible
    brush {
     bIsDynamicallyLoaded=false
     drawAs=Image
     imageType=NoImage
     mirroring=NoMirror
     resourceName=None
     resourceObject=/Gifting/UI/Art/T_FriendSearch
     tiling=NoTile
     imageSize=<50 50>
     margin=<ltrb 0 0 0 0>
     outlineSettings {
      bUseBrushTransparency=false
      roundingType=HalfHeightRadius
      width=0
      color {colorUseRule=UseColor_Specified, specifiedColor=<rgba 0 0 0 0>}
      cornerRadii=<0 0 0 0>
     }
     tintColor {colorUseRule=UseColor_Specified, specifiedColor=<rgba 1 1 1 1>}
     uVRegion {bIsValid=false, max=<0 0>, min=<0 0>}
    }
   }
   slot {class=SizeBoxSlot, properties {horizontalAlignment=HAlign_Left, verticalAlignment=VAlign_Center}}
  }
 ]
}
