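# Gifting price widget, derived from AthenaGiftingPriceWidget, showing a gift's cost in V-Bucks with struck-through regular price, or in real money.
# SetPresentationMode switches between those layouts or collapses the widget; the PlayStation Store logo is platform-filtered.
# Activation starts FortMtxStoreVM monitoring and clears the real-price equivalent text; deactivation stops monitoring.
# OnPriceSet broadcasts Price for field-notify and sets Is Mtx Currency, gated by ShouldShowRealPrice.
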WidgetBlueprint /Gifting/UI/GiftingPrice
graphs [
 EventGraph {nodes [
   Event On Activated()
   > StartMonitoringForUpdates!(
    execute
    self:Fort Mcp Catalog VM Object Reference=Get FortMtxStoreVM() -> (FortMtxStoreVM:Fort Mtx Store VM Object Reference)
    RequestingObject:Object Reference
   ) -> (then)
   > Set Text(self=Get Text_RealPrice() -> (Text_RealPrice:Common Text Object Reference))
   Event On Deactivated()
   > StopMonitoringForUpdates!(execute, self:Fort Mcp Catalog VM Object Reference=Get FortMtxStoreVM(), RequestingObject:Object Reference) -> (then)
   n1 = Event OnPriceSet!() -> (OutputDelegate, then, FinalPrice:int=0, bIsMtxCurrency:bool) {is=title K2Node_Event}
   > Set Is Mtx Currency(Is Mtx Currency:bool=AND Boolean(A=@n1.bIsMtxCurrency, B=Get ShouldShowRealPrice() -> (ShouldShowRealPrice:bool))) -> (Output_Get:bool)
   > Set with Broadcast Price!(execute, Price:int=@n1.FinalPrice) -> (then, Output_Get:int=0)
  ]}
 SetPresentationMode {nodes [
   n1 = SetPresentationMode!() -> (then, Mode:EGiftingPricePresentationMode Enum=MtxCurrency) {is=title K2Node_FunctionEntry}
   > n2 = Switch on EGiftingPricePresentationMode!(execute, Selection:EGiftingPricePresentationMode Enum=@n1) -> (MtxCurrency:Exec, RealMoney:Exec, Hidden:Exec) {is=title K2Node_SwitchEnum}
   SetVisibility(execute=@n2.Hidden, InVisibility=Collapsed)
   SetActiveWidget(
    execute=[@n2.MtxCurrency, @n2.RealMoney]
    self=Get Switcher_Prices() -> (Switcher_Prices:Common Visibility Switcher Object Reference)
    Widget=Select!(
     MtxCurrency:Widget Object Reference=Get Widget_VBucks() -> (Widget_VBucks:Horizontal Box Object Reference)
     RealMoney:Widget Object Reference=Get Widget_RealMoneyPrice() -> (Widget_RealMoneyPrice:Overlay Object Reference)
     Hidden:Widget Object Reference
     Index:EGiftingPricePresentationMode Enum=Get Mode() -> (Mode:EGiftingPricePresentationMode Enum=MtxCurrency)
    ) -> (ReturnValue:Widget Object Reference)
   )
   > SetVisibility(InVisibility=SelfHitTestInvisible)
  ]}
]
variables [
 Is Mtx Currency {type=boolean}
 Price {type=integer}
 ShouldShowRealPrice {type=boolean}
]
widgets {
 info {inheritedWidgetCount=4, parentClass=/Script/GiftingUI.AthenaGiftingPriceWidget, rootWidgetClass=&Overlay, widgetCount=17}
 widgets [
  SubtitleContainer {widgetClassPath=&Overlay, bIsVariable=true, widget=SubtitleContainer, properties {visibility=HitTestInvisible}}
  ScaleBox {
   widgetClassPath=&ScaleBox
   parent=SubtitleContainer
   widget=ScaleBox
   properties {clipping=ClipToBoundsWithoutIntersecting, stretch=ScaleToFitX, stretchDirection=DownOnly}
   slot {class=OverlaySlot, properties {horizontalAlignment=HAlign_Fill, verticalAlignment=VAlign_Fill}}
  }
  Switcher_Prices {
   widgetClassPath=&CommonVisibilitySwitcher
   bIsVariable=true
   parent=ScaleBox
   widget=Switcher_Prices
   slot {class=ScaleBoxSlot, properties {horizontalAlignment=HAlign_Fill}}
  }
  Widget_VBucks {
   widgetClassPath=&HorizontalBox
   bIsVariable=true
   parent=Switcher_Prices
   widget=Widget_VBucks
   slot {class=CommonVisibilitySwitcherSlot, properties {horizontalAlignment=HAlign_Center, verticalAlignment=VAlign_Center}}
  }
  Widget_VBuckPrice {widgetClassPath=&Overlay, parent=Widget_VBucks, widget=Widget_VBuckPrice, slot {class=HorizontalBoxSlot}}
  Cost {
   widgetClassPath=&HorizontalBox
   parent=Widget_VBuckPrice
   widget=Cost
   slot {class=OverlaySlot, properties {horizontalAlignment=HAlign_Fill, verticalAlignment=VAlign_Fill}}
  }
  Image_VBuck {
   widgetClassPath=&Image
   bIsVariable=true
   parent=Cost
   widget=Image_VBuck
   properties {brush {
     bIsDynamicallyLoaded=false
     drawAs=Image
     imageType=NoImage
     mirroring=NoMirror
     resourceName=None
     resourceObject=/Game/UI/Foundation/Shell/Textures/T-Icon-VBucks-L
     tiling=NoTile
     imageSize=<48 48>
     margin=<ltrb 0 0 0 0>
     outlineSettings {
      bUseBrushTransparency=false
      roundingType=HalfHeightRadius
      width=0
      color {colorUseRule=UseColor_Specified, specifiedColor=<rgba 0 0 0 0>}
      cornerRadii=<0 0 0 0>
     }
     tintColor {colorUseRule=UseColor_Specified, specifiedColor=<rgba 1 1 1 1>}
     uVRegion {bIsValid=false, max=<0 0>, min=<0 0>}
    }}
   slot {class=HorizontalBoxSlot, properties {horizontalAlignment=HAlign_Center, verticalAlignment=VAlign_Center}}
  }
  Text_FinalPrice {
   widgetClassPath=&CommonTextBlock
   bInherited=true
   bIsVariable=true
   parent=Cost
   widget=Text_FinalPrice
   properties {
    mobileFontSizeMultiplier=1.2
    style=/FortUILibrary/Styles/Typography/_TextStyle_SubH4._TextStyle_SubH4_C
    text=950
    font {
     bForceMonospaced=false
     bMaterialIsStencil=false
     fontMaterial=null
     fontObject=&BurbankSmall
     letterSpacing=0
     monospacedWidth=1
     size=24
     skewAmount=0
     typefaceFontName=Black
     outlineSettings {
      bApplyOutlineToDropShadows=false
      bMiteredCorners=false
      bSeparateFillAlpha=false
      outlineMaterial=null
      outlineSize=0
      outlineColor=<rgba 0 0 0 1>
     }
    }
    shadowOffset=<0 0>
    strikeBrush {
     bIsDynamicallyLoaded=false
     drawAs=NoDrawType
     imageType=NoImage
     mirroring=NoMirror
     resourceName=None
     resourceObject=null
     tiling=NoTile
     imageSize=<32 32>
     margin=<ltrb 0 0 0 0>
     outlineSettings {
      bUseBrushTransparency=false
      roundingType=HalfHeightRadius
      width=0
      color {colorUseRule=UseColor_Specified, specifiedColor=<rgba 0 0 0 0>}
      cornerRadii=<0 0 0 0>
     }
     tintColor {colorUseRule=UseColor_Specified, specifiedColor=<rgba 1 1 1 1>}
     uVRegion {bIsValid=false, max=<0 0>, min=<0 0>}
    }
   }
   slot {class=HorizontalBoxSlot, properties {verticalAlignment=VAlign_Center, padding=<ltrb 4 0 4 0>}}
  }
  Overlay_SalePrice {
   widgetClassPath=&Overlay
   bInherited=true
   parent=Widget_VBucks
   widget=Overlay_SalePrice
   properties {renderOpacity=0.5}
   slot {class=HorizontalBoxSlot, properties {horizontalAlignment=HAlign_Left, verticalAlignment=VAlign_Center, padding=<ltrb 8 0 4 0>}}
  }
  Text_RegularPrice {
   widgetClassPath=&CommonTextBlock
   bInherited=true
   bIsVariable=true
   parent=Overlay_SalePrice
   widget=Text_RegularPrice
   properties {
    justification=Center
    mobileFontSizeMultiplier=1.1
    style=&_TextStyle_SubH4_C
    text=1500
    textTransformPolicy=ToUpper
    font {
     bForceMonospaced=false
     bMaterialIsStencil=false
     fontMaterial=null
     fontObject=&BurbankSmall
     letterSpacing=0
     monospacedWidth=1
     size=24
     skewAmount=0
     typefaceFontName=Black
     outlineSettings {
      bApplyOutlineToDropShadows=false
      bMiteredCorners=false
      bSeparateFillAlpha=false
      outlineMaterial=null
      outlineSize=0
      outlineColor=<rgba 0 0 0 1>
     }
    }
    shadowOffset=<0 0>
    strikeBrush {
     bIsDynamicallyLoaded=false
     drawAs=NoDrawType
     imageType=NoImage
     mirroring=NoMirror
     resourceName=None
     resourceObject=null
     tiling=NoTile
     imageSize=<32 32>
     margin=<ltrb 0 0 0 0>
     outlineSettings {
      bUseBrushTransparency=false
      roundingType=HalfHeightRadius
      width=0
      color {colorUseRule=UseColor_Specified, specifiedColor=<rgba 0 0 0 0>}
      cornerRadii=<0 0 0 0>
     }
     tintColor {colorUseRule=UseColor_Specified, specifiedColor=<rgba 1 1 1 1>}
     uVRegion {bIsValid=false, max=<0 0>, min=<0 0>}
    }
   }
   slot {class=OverlaySlot, properties {horizontalAlignment=HAlign_Right}}
  }
  IMG_Strike {
   widgetClassPath=&Image
   bIsVariable=true
   parent=Overlay_SalePrice
   widget=IMG_Strike
   properties {
    renderOpacity=0.75
    brush {
     bIsDynamicallyLoaded=false
     drawAs=Image
     imageType=NoImage
     mirroring=NoMirror
     resourceName=None
     resourceObject=/Gifting/UI/Art/M_UI_StrikeThrough_Line_GiftingPurchaseButton
     tiling=NoTile
     imageSize=<32 4>
     margin=<ltrb 0 0 0 0>
     outlineSettings {
      bUseBrushTransparency=false
      roundingType=HalfHeightRadius
      width=0
      color {colorUseRule=UseColor_Specified, specifiedColor=<rgba 0 0 0 0>}
      cornerRadii=<0 0 0 0>
     }
     tintColor {colorUseRule=UseColor_Specified, specifiedColor=<rgba 1 1 1 1>}
     uVRegion {bIsValid=false, max=<0 0>, min=<0 0>}
    }
    renderTransform {angle=0, scale=<1.125 1>, shear=<0 -3>, translation=<-2 0>}
   }
   slot {class=OverlaySlot, properties {horizontalAlignment=HAlign_Fill, verticalAlignment=VAlign_Center}}
  }
  Text_RealPrice {
   widgetClassPath=&CommonTextBlock
   bIsVariable=true
   parent=Widget_VBucks
   widget=Text_RealPrice
   properties {
    bAutoCollapseWithEmptyText=true
    bIsScrollingEnabled=false
    text="[PH] (equivalent to $8.99*)"
    visibility=Collapsed
    font {
     bForceMonospaced=false
     bMaterialIsStencil=false
     fontMaterial=null
     fontObject=&HeadingNow
     letterSpacing=0
     monospacedWidth=1
     size=16.5
     skewAmount=0
     typefaceFontName=74 Regular
     outlineSettings {
      bApplyOutlineToDropShadows=false
      bMiteredCorners=false
      bSeparateFillAlpha=false
      outlineMaterial=null
      outlineSize=0
      outlineColor=<rgba 0 0 0 1>
     }
    }
    shadowOffset=<0 0>
   }
   slot {class=HorizontalBoxSlot, properties {horizontalAlignment=HAlign_Left, verticalAlignment=VAlign_Center, padding=<ltrb 4 0 0 0>}}
  }
  Widget_RealMoneyPrice {
   widgetClassPath=&Overlay
   bIsVariable=true
   parent=Switcher_Prices
   widget=Widget_RealMoneyPrice
   slot {class=CommonVisibilitySwitcherSlot, properties {horizontalAlignment=HAlign_Center, verticalAlignment=VAlign_Center}}
  }
  HorizontalBox_0 {
   widgetClassPath=&HorizontalBox
   parent=Widget_RealMoneyPrice
   widget=HorizontalBox_0
   slot {class=OverlaySlot, properties {horizontalAlignment=HAlign_Fill, verticalAlignment=VAlign_Fill}}
  }
  Text_RealMoneyPrice {
   widgetClassPath=&CommonTextBlock
   bInherited=true
   parent=HorizontalBox_0
   widget=Text_RealMoneyPrice
   properties {
    justification=Center
    mobileFontSizeMultiplier=1.2
    text=$xx.xx
    textTransformPolicy=ToUpper
    font {
     bForceMonospaced=false
     bMaterialIsStencil=false
     fontMaterial=null
     fontObject=&Burbank
     letterSpacing=0
     monospacedWidth=1
     size=24
     skewAmount=0
     typefaceFontName=BurbankRegular-Black
     outlineSettings {
      bApplyOutlineToDropShadows=false
      bMiteredCorners=false
      bSeparateFillAlpha=false
      outlineMaterial=null
      outlineSize=0
      outlineColor=<rgba 0 0 0 1>
     }
    }
    shadowOffset=<0 0>
   }
   slot {class=HorizontalBoxSlot}
  }
  CommonVisibilityWidget_0 {
   widgetClassPath=&CommonVisibilityWidgetLegacy
   parent=HorizontalBox_0
   widget=CommonVisibilityWidget_0
   properties {
    bShowForAndroid=false
    bShowForErebus=false
    bShowForIOS=false
    bShowForMac=false
    bShowForPC=false
    bShowForSwitch2=false
    bShowForTouch=false
    bShowForXBox=false
    bShowForXSX=false
    padding=<ltrb 0 0 0 0>
   }
   slot {class=HorizontalBoxSlot, properties {horizontalAlignment=HAlign_Left, verticalAlignment=VAlign_Center, padding=<ltrb 8 0 0 0>}}
  }
  PlayStationStoreLogo {
   widgetClassPath=&Image
   bIsVariable=true
   parent=CommonVisibilityWidget_0
   widget=PlayStationStoreLogo
   properties {
    clipping=ClipToBoundsWithoutIntersecting
    visibility=SelfHitTestInvisible
    brush {
     bIsDynamicallyLoaded=false
     drawAs=Image
     imageType=NoImage
     mirroring=NoMirror
     resourceName=None
     resourceObject=/Game/UI/Frontend/ItemShop/Art/T_UI_PS_Store_Logo_64x
     tiling=NoTile
     imageSize=<40 40>
     margin=<ltrb 0 0 0 0>
     outlineSettings {
      bUseBrushTransparency=false
      roundingType=HalfHeightRadius
      width=0
      color {colorUseRule=UseColor_Specified, specifiedColor=<rgba 0 0 0 0>}
      cornerRadii=<0 0 0 0>
     }
     tintColor {colorUseRule=UseColor_Specified, specifiedColor=<rgba 1 1 1 1>}
     uVRegion {bIsValid=false, max=<0 0>, min=<0 0>}
    }
    renderTransform {angle=0, scale=<1 1>, shear=<0 0>, translation=<0 -2>}
   }
   slot {class=BorderSlot, properties {padding=<ltrb 0 0 0 0>}}
  }
 ]
}
