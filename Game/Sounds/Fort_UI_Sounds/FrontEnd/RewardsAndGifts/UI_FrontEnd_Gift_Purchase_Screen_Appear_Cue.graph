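# Sound cue for the front-end gift purchase screen appearing, routed to the Fort_UI sound class.
# A single wave player plays the UI_Gift_Player_Tile_Select wave directly, with no randomization or modulation.
# Concurrency override is on, capping instances at two with stop-oldest resolution and a short retrigger time, so the referenced front-end menu concurrency set is superseded.
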SoundCue /Game/Sounds/Fort_UI_Sounds/FrontEnd/RewardsAndGifts/UI_FrontEnd_Gift_Purchase_Screen_Appear_Cue
root=WavePlayer_2
nodes [{id=WavePlayer_2, properties {soundWaveAssetPtr=/Game/Sounds/Fort_UI_Sounds/FrontEnd/RewardsAndGifts/wav/UI_Gift_Player_Tile_Select}}]
properties {
 bOverrideConcurrency=true
 maxDistance=2097152
 soundClassObject=&Fort_UI
 volumeMultiplier=1
 concurrencyOverrides {maxCount=2, resolutionRule=StopOldest, retriggerTime=0.2}
 concurrencySet [/Game/Sounds/Concurrency_Rules/FrontEnd_UI_Menu_Max_03_Stop_Oldest]
}
