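# UI sound cue for the front-end gift purchase screen disappearing.
# A single wave player feeds the output directly, reusing a medium modal whoosh from the Battle Pass toolkit at reduced volume in the UI sound class.
# A concurrency override caps it at two instances, stopping the oldest, with a short retrigger time, and takes precedence over the assigned front-end menu concurrency set.
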SoundCue /Game/Sounds/Fort_UI_Sounds/FrontEnd/RewardsAndGifts/UI_FrontEnd_Gift_Purchase_Screen_Dissapear_Cue
root=WavePlayer_1
nodes [{id=WavePlayer_1, properties {soundWaveAssetPtr=/Game/Sounds/Fort_UI_Sounds/FrontEnd/BattlePass/BP_ToolKit/BP_Modal_Whoosh_Med_03}}]
properties {
 bOverrideConcurrency=true
 maxDistance=2097152
 soundClassObject=&Fort_UI
 volumeMultiplier=0.2
 concurrencyOverrides {maxCount=2, resolutionRule=StopOldest, retriggerTime=0.2}
 concurrencySet [/Game/Sounds/Concurrency_Rules/FrontEnd_UI_Menu_Max_03_Stop_Oldest]
}
