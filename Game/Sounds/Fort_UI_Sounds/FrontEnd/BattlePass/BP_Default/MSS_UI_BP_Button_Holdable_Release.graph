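# Battle Pass UI sound for releasing a holdable button, a one-shot stereo MetaSound Source.
# It is a preset of the default loud holdable-button release sound, feeding its Sounds array the loud press-hold release wave.
# Playback is pitched up at reduced volume on the Fort_UI sound class, with virtualization disabled.
# A concurrency override allows a single instance, stopping the oldest.
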MetaSoundSource /Game/Sounds/Fort_UI_Sounds/FrontEnd/BattlePass/BP_Default/MSS_UI_BP_Button_Holdable_Release
preset_parent=/Game/Sounds/Fort_UI_Sounds/FrontEnd/DefaultSounds/HoldableButtons/Holdable_Loud/MSS_UI_Button_Holdable_Loud_Release
inputs [
 Sounds {type=WaveAsset:Array}
 UE.Source.OnPlay {type=Trigger}
]
interfaces [UE.OutputFormat.Stereo:1.0, UE.Source.OneShot:1.0, UE.Source:1.1]
outputs [
 "UE.OutputFormat.Stereo.Audio:0" {type=Audio}
 "UE.OutputFormat.Stereo.Audio:1" {type=Audio}
 UE.Source.OneShot.OnFinished {type=Trigger}
]
pages [Default {input_defaults {UE.Source.OnPlay=null, Sounds [/Game/Sounds/Fort_UI_Sounds/FrontEnd/DefaultSounds/HoldableButtons/wav/UI_Button_Press_Hold_Loud_Release]}}]
properties {
 assetImportData=&self:AssetImportData
 bOverrideConcurrency=true
 numChannels=2
 outputFormat=Stereo
 pitch=1.5
 sampleRate=48000
 soundClassObject=&Fort_UI
 virtualizationMode=Disabled
 volume=0.629
 concurrencyOverrides {maxCount=1, resolutionRule=StopOldest}
}
