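# Translucent UI material that procedurally draws a triangle-over-bar icon, apparently for a hold mouse/touch toast.
# A triangle function's fill and a rectangle built from Ceil step differences on scaled, offset UVs are added to blend near-black to white.
# A tightly smoothstepped radial gradient on the scaled UVs forms a circular mask around the icon.
# The shapes are all computed from constants, without textures or exposed parameters.
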Material /Game/UI/Foundation/Toasts/Materials/M_UI_HoldMouse_Touch
graph {
 properties {
  bUsedWithStaticMesh=false
  blendMode=BLEND_Translucent
  materialDomain=MD_UI
  pixelDepthOffsetMode=PDOM_Legacy
  usedShadingModels=Default Lit
 }
 nodes [
  n1 = MaterialFunctionCall(Texture Scale=Constant() {r=0.875}) {materialFunction=&ScaleUVsByCenter}
  n2 = MaterialFunctionCall!(UVs=@n1.UVs, Offset=Constant2Vector() {g=0.075}, Y_MaskRadius, Texture Scale) {materialFunction=/Game/Characters/Player/Male/Medium/Bodies/M_MED_Gumball/Materials/MF_ScaleAndOffsetUVs}
  n3 = Constant2Vector() {g=-0.2, r=-0.5}
  n4 = Constant2Vector() {g=-0.2, r=0.5}
  n5 = Constant2Vector() {g=0.3}
  n6 = ComponentMask(None=@n2.UVs) {g=true}
  n7 = Subtract(A=@n6) {constB=0.62}
  n8 = Subtract(A=@n6) {constB=0.72}
  n9 = ComponentMask(None=@n2.UVs) {r=true}
  n10 = Subtract(A=@n9) {constB=0.175}
  n11 = Subtract(A=@n9) {constB=0.825}
  n12 = Multiply(A=Subtract(A=Ceil(None=@n7), B=Ceil(None=@n8)), B=Subtract(A=Ceil(None=@n10), B=Ceil(None=@n11)))
  LinearInterpolate(
   A=Constant3Vector() {constant=<rgba 0 0 0.0056 1>}
   B=Constant() {r=1}
   Alpha=Add(
    A=MaterialFunctionCall!(UVs=@n2.UVs, Corner 01=@n3, Corner 02=@n4, Corner 03=@n5, Glow Min, Glow Max, Outline Thickness) {materialFunction=/FortUILibrary/MaterialFunctions/Shapes/MF_UI_WIP_Triangle}.Fill
    B=Saturate(None=@n12)
   )
  )
  MaterialFunctionCall(
   Alpha=MaterialFunctionCall(UVs=@n1.UVs) {materialFunction=&RadialGradientExponential}.RadialGradientExponential
   Min=Constant()
   Max=Constant() {r=0.05}
  ) {materialFunction=&SmoothStep}
 ]
}
