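# Holdable "Purchase Gift!" button widget for the gifting UI, built on AthenaGiftingPurchaseButton, with a price display, input prompt and optional hold label.
# Hover, or press on touch, whitens the button, plays triangle and backing animations, and reveals the prompt, whose click icon hides on gamepad.
# Hold events drive the radial progress animation, scaled to HoldTime and reversed on early release.
# Press, hold, release and completion play sounds, fading stored looping or press audio.
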WidgetBlueprint /Gifting/UI/GiftingPurchaseButton
animations [
 HoverTrianglesAnim {
  binding_count=2
  length_seconds=1
  object=&self:HoverTrianglesAnim
  sequence {bindings [
    HoverTriangles {
     id=#1
     possessed_class=Image
     tracks [{
       type=MovieSceneWidgetMaterialTrack
       sections [{
         type=MovieSceneParameterSection
         channels [
          Alpha {type=MovieSceneScriptingFloatChannel, keys [[0,0],[9,1],[20,0.75]]}
          UseFade {type=MovieSceneScriptingFloatChannel, keys [[0,0],[2,0.5],[20,1]]}
          NoiseIntensity {type=MovieSceneScriptingFloatChannel, keys [[0,1],[2,1],[20,0.5]]}
          NoiseLift {type=MovieSceneScriptingFloatChannel, keys [[0,0],[2,0.333],[20,0.25]]}
         ]
        }]
      }]
    }
    BackingHoverLift {
     id=#2
     possessed_class=Image
     tracks [{
       type=MovieSceneFloatTrack
       display_name=Render Opacity
       sections [{type=MovieSceneFloatSection, end_frame=4, start_frame=0, channels [None {type=MovieSceneScriptingFloatChannel, keys [[0,0],[4,1]]}]}]
      }]
    }
   ]}
 }
 ResetTrianglesAnim {
  binding_count=1
  length_seconds=0.1
  object=&self:ResetTrianglesAnim
  sequence {bindings [HoverTriangles {
     id=#3
     possessed_class=Image
     tracks [{
       type=MovieSceneWidgetMaterialTrack
       sections [{type=MovieSceneParameterSection, channels [Alpha {type=MovieSceneScriptingFloatChannel, const [0,2]}]}]
      }]
    }]}
 }
 HoldProgress {
  binding_count=1
  length_seconds=1
  object=&self:HoldProgress
  sequence {bindings [ProgressCircle {
     id=#4
     possessed_class=Image
     tracks [{
       type=MovieSceneWidgetMaterialTrack
       sections [{
         type=MovieSceneParameterSection
         end_frame=20
         start_frame=0
         channels [Progress {type=MovieSceneScriptingFloatChannel, keys [[0,0],[20,1]]}]
        }]
      }]
    }]}
 }
]
graphs [EventGraph {nodes [
   "On Input Icon Updated (InputActionWidget)"() {is=title K2Node_ComponentBoundEvent}
   > n1 = GetIconDynamicMaterial!(execute, self:Common Action Widget Object Reference=Get InputActionWidget() -> (InputActionWidget:Common Action Widget Object Reference)) -> (then, ReturnValue:Material Instance Dynamic Object Reference)
   > SetScalarParameterValue(self=@n1, ParameterName=Use Dark, Value=1)
   > SetScalarParameterValue(self=@n1, ParameterName=Outline Opacity, Value=0)
   n2 = InputChanged!() -> (OutputDelegate, then, bNewInputType:ECommonInputType Enum) {is=title K2Node_CustomEvent}
   > Branch(Condition=IsHovered())
   > SetVisibility(
    self=Get PressOrClickIcon() -> (PressOrClickIcon:Fort Mobile Image Object Reference)
    InVisibility=Select(Option 0=SelfHitTestInvisible, Option 1=Collapsed, Index="Equal (Enum)"(A=@n2.bNewInputType, B=Gamepad))
   )
   Event PreConstruct()
   > SetVisibility(
    self=[
     Get Spacer_InputAction() -> (Spacer_InputAction:Spacer Object Reference)
     Get ProgressCircle() -> (ProgressCircle:Image Object Reference)
     Get Spacer_EvenSize() -> (Spacer_EvenSize:Spacer Object Reference)
    ]
    InVisibility=SelfHitTestInvisible
   )
   > SetVisibility(self=Get Overlay_Input_Keybind() -> (Overlay_Input_Keybind:Overlay Object Reference), InVisibility=Hidden)
   n3 = Get!(self:Widget Object Reference=Get TextBlock_HoldPrompt() -> (TextBlock_HoldPrompt:Common Text Object Reference)) -> (
    RenderTransform_Translation_X:double
    RenderTransform_Translation_Y:double
    RenderTransform_Scale:Vector 2D Structure
    RenderTransform_Shear:Vector 2D Structure
    RenderTransform_Angle:float=~
   )
   > SetRenderTranslation!(
    execute
    self=Get VerticalBox_Text() -> (VerticalBox_Text:Vertical Box Object Reference)
    Translation_X:double=0
    Translation_Y:double=Select(
     Option 0:double=0
     Option 1:double=float * float(A=@n3.RenderTransform_Translation_Y, B=-0.5) {is=title K2Node_PromotableOperator}
     Index=GetIsHoldRequired!(self:Fort Holdable Button Object Reference) -> (ReturnValue:bool)
    ) -> (ReturnValue:double)
   ) -> (then)
   > Set members in Slate Brush!(
    execute:Exec
    StructRef:"Slate Brush Structure (by ref)"=Get ProgressMaterialBrush!(self:Common Action Widget Object Reference=Get InputActionWidget()) -> (ProgressMaterialBrush:Slate Brush Structure)
    ResourceObject:Object Reference=/Game/UI/Materials/EmptyProgressMaterial
   ) -> (then:Exec, StructOut:"Slate Brush Structure (by ref)") {is=title K2Node_SetFieldsInStruct}
   Event On Hovered()
   > n4 = Branch(Condition=IsUsingTouch(PlayerController=GetOwningPlayer()))
   n5 = Get ButtonColor() -> (ButtonColor:Image Object Reference)
   SetColorAndOpacity(execute=[@n19.then, @n4.else], self:Image Object Reference=@n5, InColorAndOpacity:Linear Color Structure=(R=1,G=1,B=1,A=1))
   n6 = Get HoverTrianglesAnim() -> (HoverTrianglesAnim:Widget Animation Object Reference)
   > PlayAnimationForward(InAnimation=@n6)
   n7 = Get Overlay_Input_Keybind()
   > SetVisibility(self=@n7, InVisibility=SelfHitTestInvisible)
   > n8 = Branch(Condition=GetIsHoldRequired())
   n9 = GetCurrentInputType(self=Get CommonInputSubsystem())
   > SetVisibility(
    self=Get PressOrClickIcon()
    InVisibility=Select(Option 0:ESlateVisibility Enum=SelfHitTestInvisible, Option 1:ESlateVisibility Enum=Collapsed, Index="Equal (Enum)"(A=@n9, B=Gamepad)) -> (ReturnValue:ESlateVisibility Enum)
   )
   > SetVisibility(self=Get ProgressCircle(), InVisibility=SelfHitTestInvisible)
   SetVisibility(execute=@n8.else, self=[Get ProgressCircle(), Get PressOrClickIcon()], InVisibility=Collapsed)
   n10 = Event BP_OnHoldStarted!() -> (OutputDelegate, then, HoldPercentage:float, bInUseSecondaryHoldAnimation:bool) {is=title K2Node_Event}
   n11 = Get HoldProgress() -> (HoldProgress:Widget Animation Object Reference)
   n12 = GetEndTime(self=@n11)
   > PlayAnimation(
    InAnimation=@n11
    StartAtTime=float * float(A=@n12, B=@n10.HoldPercentage) {is=title K2Node_CommutativeAssociativeBinaryOperator}
    PlaybackSpeed=float / float(A=@n12, B=Get HoldTime() -> (HoldTime:float)) {is=title K2Node_CallFunction}
    bRestoreState=true
   )
   > n13 = SpawnSound2D!(
    execute
    Sound:Sound Base Object Reference=Get HoldSound() -> (HoldSound:Sound Base Object Reference)
    VolumeMultiplier:float=1
    PitchMultiplier:float=1
    StartTime:float=0
    ConcurrencySettings:Sound Concurrency Object Reference
    bPersistAcrossLevelTransition:bool=false
    bAutoDestroy:bool=true
   ) -> (then, ReturnValue:Audio Component Object Reference)
   > Set HoldSoundComponent(HoldSoundComponent:Audio Component Object Reference=@n13) -> (Output_Get:Audio Component Object Reference)
   n14 = Event BP_OnHoldEnded!() -> (OutputDelegate, then, HoldPercentage:float, bInUseSecondaryHoldAnimation:bool) {is=title K2Node_Event}
   > Branch(Condition="float > float"(A=@n14.HoldPercentage, B=0) {is=title K2Node_CallFunction})
   n15 = Get HoldProgress()
   > PlayAnimationReverse(InAnimation=@n15, PlaybackSpeed=float / float(A=GetEndTime(self=@n15), B=Get HoldTime()))
   > SpawnSound2D(
    Sound=Get HoldReleaseSound() -> (HoldReleaseSound:Sound Base Object Reference)
    VolumeMultiplier=1
    PitchMultiplier=1
    StartTime=0
    bPersistAcrossLevelTransition=false
    bAutoDestroy=true
   )
   > n16 = Get!(execute:Exec) -> (then:Exec, else:Exec, HoldSoundComponent:Audio Component Object Reference)
   > FadeOut!(
    execute
    self:Audio Component Object Reference=@n16
    FadeOutDuration:float=0.2
    FadeVolumeLevel:float=0
    FadeCurve:EAudioFaderCurve Enum=Linear
   ) -> (then)
   n17 = Event On Pressed() {is=title K2Node_Event}
   n18 = SpawnSound2D(
    execute=[@n17.then, @n31.then]
    Sound=Get HoldPressedSound() -> (HoldPressedSound:Sound Base Object Reference)
    VolumeMultiplier=1
    PitchMultiplier=1
    StartTime=0
    bPersistAcrossLevelTransition=false
    bAutoDestroy=true
   )
   > Set PressedSoundComponent(PressedSoundComponent:Audio Component Object Reference=@n18) -> (Output_Get:Audio Component Object Reference)
   > n19 = Branch(Condition=IsUsingTouch(PlayerController=GetOwningPlayer()))
   Event On Released() {is=title K2Node_Event}
   > n20 = Get!(execute) -> (then, else, PressedSoundComponent:Audio Component Object Reference)
   > FadeOut(self=@n20, FadeOutDuration=0.2, FadeVolumeLevel=0, FadeCurve=Linear)
   > SpawnSound2D(Sound=Get HoldReleaseSound(), VolumeMultiplier=1, PitchMultiplier=1, StartTime=0, bPersistAcrossLevelTransition=false, bAutoDestroy=true)
   > n21 = Branch(Condition=IsUsingTouch(PlayerController=GetOwningPlayer()))
   SetColorAndOpacity(execute=[@n32.then, @n30.else, @n21.then], self=@n5, InColorAndOpacity=(R=1,G=1,B=0,A=1))
   > StopAnimation(InAnimation=@n6)
   > PlayAnimationForward(InAnimation=Get ResetTrianglesANim() -> (ResetTrianglesANim:Widget Animation Object Reference))
   > SetVisibility(self=@n7, InVisibility=Hidden)
   Event OnHoldRequiredChanged!() -> (OutputDelegate, then, bHoldRequired:bool) {is=title K2Node_Event}
   n22 = Select(Option 0=Collapsed, Option 1=HitTestInvisible, Index=GetIsHoldRequired())
   n23 = Get TextBlock_HoldPrompt()
   > SetVisibility(self=@n23, InVisibility=@n22)
   n24 = Get!(self:Widget Object Reference=@n23) -> (
    RenderTransform_Translation_X:double
    RenderTransform_Translation_Y:double
    RenderTransform_Scale:Vector 2D Structure
    RenderTransform_Shear:Vector 2D Structure
    RenderTransform_Angle:float=~
   )
   > SetRenderTranslation(
    self=Get VerticalBox_Text()
    Translation_X=0
    Translation_Y=Select(
     Option 0:double=0
     Option 1:double=float * float(A=@n24.RenderTransform_Translation_Y, B=-0.5) {is=title K2Node_PromotableOperator}
     Index=GetIsHoldRequired()
    ) -> (ReturnValue:double)
   )
   Event BP_OnHoldCompleted() {is=title K2Node_Event}
   > n25 = SpawnSound2D(
    Sound=Get HoldCompleteSound() -> (HoldCompleteSound:Sound Base Object Reference)
    VolumeMultiplier=1
    PitchMultiplier=1
    StartTime=0
    bPersistAcrossLevelTransition=false
    bAutoDestroy=true
   )
   StopAllAnimations(execute=[@n25.then, @n33.then], self:User Widget Object Reference)
   > n26 = GetDynamicMaterial(self=Get ProgressCircle())
   > SetScalarParameterValue(self=@n26, ParameterName=Progress, Value=0)
   > n27 = Get!(execute:Exec) -> (then:Exec, else:Exec, HoldSoundComponent:Audio Component Object Reference)
   > FadeOut(self=@n27, FadeOutDuration=0.2, FadeVolumeLevel=0, FadeCurve=Linear)
   Event Construct()
   > SetVisibility(self=Get TextBlock_HoldPrompt(), InVisibility=@n22)
   n28 = Create Event()
   > Assign On Input Method Changed!(
    execute:Exec
    self:Common Input Object Reference=Get CommonInputSubsystem!() -> (ReturnValue:Common Input Object Reference) {is=title K2Node_GetBlueprintContextEx}
    Delegate:"Delegate (by ref)"=@n28
   ) -> (then:Exec) {is=title K2Node_AssignDelegate}
   Event On Unhovered()
   > n29 = Get!(execute) -> (then, else, PressedSoundComponent:Audio Component Object Reference)
   > FadeOut(self=@n29, FadeOutDuration=0.2, FadeVolumeLevel=0, FadeCurve=Linear)
   > n30 = Branch(Condition=IsUsingTouch(PlayerController=GetOwningPlayer()))
   n31 = Event On Selected()
   n32 = Event On Disabled() {is=title K2Node_Event}
   n33 = Event BP_OnHoldReset() {is=title K2Node_Event}
   Event Destruct()
   > Unbind Event from On Input Method Changed(self=Get CommonInputSubsystem(), Delegate=@n28)
  ]}]
variables [
 HoldCompleteSound {type=SoundBase, default=/Game/Sounds/Fort_UI_Sounds/FrontEnd/BattlePass/BP_Default/MSS_UI_BP_Button_Holdable_Complete}
 HoldPressedSound {type=SoundBase, default=/Game/Sounds/Fort_UI_Sounds/FrontEnd/BattlePass/BP_Default/MSS_UI_BP_Button_Holdable_Press}
 HoldReleaseSound {type=SoundBase, default=/Game/Sounds/Fort_UI_Sounds/FrontEnd/BattlePass/BP_Default/MSS_UI_BP_Button_Holdable_Release}
 HoldSound {type=SoundBase, default=/Game/Sounds/Fort_UI_Sounds/FrontEnd/BattlePass/BP_Default/MSS_UI_BP_Button_Holdable_Hold_lp}
 HoldSoundComponent {type=AudioComponent, default=null}
 PressedSoundComponent {type=AudioComponent, default=null}
]
widgets {
 info {inheritedWidgetCount=3, parentClass=/Script/GiftingUI.AthenaGiftingPurchaseButton, rootWidgetClass=&Overlay, widgetCount=21}
 widgets [
  Container {widgetClassPath=&Overlay, widget=Container}
  Backing {
   widgetClassPath=&Image
   bIsVariable=true
   parent=Container
   widget=Backing
   properties {
    visibility=SelfHitTestInvisible
    brush {
     bIsDynamicallyLoaded=false
     drawAs=Image
     imageType=NoImage
     mirroring=NoMirror
     resourceName=None
     resourceObject=null
     tiling=NoTile
     imageSize=<32 32>
     margin=<ltrb 0 0 0 0>
     outlineSettings {
      bUseBrushTransparency=false
      roundingType=HalfHeightRadius
      width=0
      color {colorUseRule=UseColor_Specified, specifiedColor=<rgba 0 0 0 0>}
      cornerRadii=<0 0 0 0>
     }
     tintColor {colorUseRule=UseColor_Specified, specifiedColor=<rgba 0 0.0908 1 1>}
     uVRegion {bIsValid=false, max=<0 0>, min=<0 0>}
    }
   }
   slot {class=OverlaySlot, properties {horizontalAlignment=HAlign_Fill, verticalAlignment=VAlign_Fill}}
  }
  BackingHoverLift {
   widgetClassPath=&Image
   bIsVariable=true
   parent=Container
   widget=BackingHoverLift
   properties {
    renderOpacity=0
    visibility=SelfHitTestInvisible
    brush {
     bIsDynamicallyLoaded=false
     drawAs=Image
     imageType=NoImage
     mirroring=NoMirror
     resourceName=None
     resourceObject=/Gifting/UI/Art/M_UI_Gifting_Purchase_ButtonBacking
     tiling=NoTile
     imageSize=<32 32>
     margin=<ltrb 0 0 0 0>
     outlineSettings {
      bUseBrushTransparency=false
      roundingType=HalfHeightRadius
      width=0
      color {colorUseRule=UseColor_Specified, specifiedColor=<rgba 0 0 0 0>}
      cornerRadii=<0 0 0 0>
     }
     tintColor {colorUseRule=UseColor_Specified, specifiedColor=<rgba 1 1 1 1>}
     uVRegion {bIsValid=false, max=<0 0>, min=<0 0>}
    }
   }
   slot {class=OverlaySlot, properties {horizontalAlignment=HAlign_Fill, verticalAlignment=VAlign_Fill}}
  }
  HoverTriangles {
   widgetClassPath=&Image
   bIsVariable=true
   parent=Container
   widget=HoverTriangles
   properties {brush {
     bIsDynamicallyLoaded=false
     drawAs=Image
     imageType=NoImage
     mirroring=NoMirror
     resourceName=None
     resourceObject=/Game/Items/Traps/Blueprints/Toys/UI_Art_Assets/M_UI_BP_Preview_Button_Orange
     tiling=NoTile
     imageSize=<32 32>
     margin=<ltrb 0 0 0 0>
     outlineSettings {
      bUseBrushTransparency=false
      roundingType=HalfHeightRadius
      width=0
      color {colorUseRule=UseColor_Specified, specifiedColor=<rgba 0 0 0 0>}
      cornerRadii=<0 0 0 0>
     }
     tintColor {colorUseRule=UseColor_Specified, specifiedColor=<rgba 1 1 1 1>}
     uVRegion {bIsValid=false, max=<0 0>, min=<0 0>}
    }}
   slot {class=OverlaySlot, properties {horizontalAlignment=HAlign_Fill, verticalAlignment=VAlign_Fill}}
  }
  VerticalBox_2 {
   widgetClassPath=&VerticalBox
   parent=Container
   widget=VerticalBox_2
   slot {class=OverlaySlot, properties {horizontalAlignment=HAlign_Fill, verticalAlignment=VAlign_Center, padding=<ltrb 8 8 8 8>}}
  }
  Widget_Price {
   widgetClassPath=&/GiftingPrice
   bInherited=true
   bIsVariable=true
   parent=VerticalBox_2
   widget=Widget_Price
   properties {shouldShowRealPrice=true}
   slot {class=VerticalBoxSlot, properties {size {sizeRule=Fill, value=1}}}
  }
  SizeBox_Height {
   widgetClassPath=&SizeBox
   parent=VerticalBox_2
   widget=SizeBox_Height
   properties {bOverride_MinDesiredHeight=true, minDesiredHeight=48}
   slot {class=VerticalBoxSlot, properties {padding=<ltrb 0 5 0 0>}}
  }
  Overlay_0 {widgetClassPath=&Overlay, parent=SizeBox_Height, widget=Overlay_0, slot {class=SizeBoxSlot}}
  ButtonColor {
   widgetClassPath=&Image
   bIsVariable=true
   parent=Overlay_0
   widget=ButtonColor
   properties {colorAndOpacity=<rgba 1 1 0 1>}
   slot {class=OverlaySlot, properties {horizontalAlignment=HAlign_Fill, verticalAlignment=VAlign_Fill}}
  }
  HorizontalBox_ButtonContent {
   widgetClassPath=&HorizontalBox
   bIsVariable=true
   parent=Overlay_0
   widget=HorizontalBox_ButtonContent
   properties {flowDirectionPreference=Culture}
   slot {class=OverlaySlot, properties {horizontalAlignment=HAlign_Fill, verticalAlignment=VAlign_Fill}}
  }
  Overlay_Input_Keybind {
   widgetClassPath=&Overlay
   bIsVariable=true
   parent=HorizontalBox_ButtonContent
   widget=Overlay_Input_Keybind
   slot {class=HorizontalBoxSlot, properties {horizontalAlignment=HAlign_Left, verticalAlignment=VAlign_Center, padding=<ltrb 11 0 0 0>}}
  }
  PressOrClickIcon {
   widgetClassPath=/Script/FortniteUI.FortMobileImage
   bIsVariable=true
   parent=Overlay_Input_Keybind
   widget=PressOrClickIcon
   properties {
    mobileSizeMultiplier=1.3
    visibility=Collapsed
    brush {
     bIsDynamicallyLoaded=false
     drawAs=Image
     imageType=NoImage
     mirroring=NoMirror
     resourceName=None
     resourceObject=/Game/UI/Foundation/Toasts/Materials/M_UI_HoldMouse_Touch
     tiling=NoTile
     imageSize=<36 36>
     margin=<ltrb 0 0 0 0>
     outlineSettings {
      bUseBrushTransparency=false
      roundingType=HalfHeightRadius
      width=0
      color {colorUseRule=UseColor_Specified, specifiedColor=<rgba 0 0 0 0>}
      cornerRadii=<0 0 0 0>
     }
     tintColor {colorUseRule=UseColor_Specified, specifiedColor=<rgba 1 1 1 1>}
     uVRegion {bIsValid=false, max=<0 0>, min=<0 0>}
    }
   }
   slot {class=OverlaySlot, properties {horizontalAlignment=HAlign_Center, verticalAlignment=VAlign_Center}}
  }
  ProgressCircle {
   widgetClassPath=&Image
   bIsVariable=true
   parent=Overlay_Input_Keybind
   widget=ProgressCircle
   properties {
    visibility=SelfHitTestInvisible
    brush {
     bIsDynamicallyLoaded=false
     drawAs=Image
     imageType=NoImage
     mirroring=NoMirror
     resourceName=None
     resourceObject=/Game/UI/Foundation/Toasts/Materials/MI_UI_RadialProgressBuyButtons
     tiling=NoTile
     imageSize=<48 48>
     margin=<ltrb 0 0 0 0>
     outlineSettings {
      bUseBrushTransparency=false
      roundingType=HalfHeightRadius
      width=0
      color {colorUseRule=UseColor_Specified, specifiedColor=<rgba 0 0 0 0>}
      cornerRadii=<0 0 0 0>
     }
     tintColor {colorUseRule=UseColor_Specified, specifiedColor=<rgba 1 1 1 1>}
     uVRegion {bIsValid=false, max=<0 0>, min=<0 0>}
    }
   }
   slot {class=OverlaySlot, properties {horizontalAlignment=HAlign_Center, verticalAlignment=VAlign_Center}}
  }
  InputActionWidget {
   widgetClassPath=&CommonActionWidget
   bInherited=true
   bIsVariable=true
   parent=Overlay_Input_Keybind
   widget=InputActionWidget
   properties {
    progressMaterialParam=Percentage
    progressMaterialBrush {
     bIsDynamicallyLoaded=false
     drawAs=Image
     imageType=NoImage
     mirroring=NoMirror
     resourceName=None
     resourceObject=/Game/UI/Materials/HollowCircularProgressBar
     tiling=NoTile
     imageSize=<36 36>
     margin=<ltrb 0 0 0 0>
     outlineSettings {
      bUseBrushTransparency=false
      roundingType=HalfHeightRadius
      width=0
      color {colorUseRule=UseColor_Specified, specifiedColor=<rgba 0 0 0 0>}
      cornerRadii=<0 0 0 0>
     }
     tintColor {colorUseRule=UseColor_Specified, specifiedColor=<rgba 1 1 0 1>}
     uVRegion {bIsValid=false, max=<0 0>, min=<0 0>}
    }
   }
   slot {class=OverlaySlot, properties {horizontalAlignment=HAlign_Center, verticalAlignment=VAlign_Center, padding=<ltrb 4 0 4 0>}}
  }
  Spacer_InputAction {
   widgetClassPath=&Spacer
   bIsVariable=true
   parent=Overlay_Input_Keybind
   widget=Spacer_InputAction
   properties {size=<48 1>}
   slot {class=OverlaySlot, properties {verticalAlignment=VAlign_Center}}
  }
  LeftSideImage {
   widgetClassPath=&Image
   bIsVariable=true
   parent=HorizontalBox_ButtonContent
   widget=LeftSideImage
   properties {
    visibility=Collapsed
    brush {
     bIsDynamicallyLoaded=false
     drawAs=NoDrawType
     imageType=NoImage
     mirroring=NoMirror
     resourceName=None
     resourceObject=null
     tiling=NoTile
     imageSize=<0 0>
     margin=<ltrb 0 0 0 0>
     outlineSettings {
      bUseBrushTransparency=false
      roundingType=HalfHeightRadius
      width=0
      color {colorUseRule=UseColor_Specified, specifiedColor=<rgba 0 0 0 0>}
      cornerRadii=<0 0 0 0>
     }
     tintColor {colorUseRule=UseColor_Specified, specifiedColor=<rgba 1 1 1 1>}
     uVRegion {bIsValid=false, max=<0 0>, min=<0 0>}
    }
   }
   slot {class=HorizontalBoxSlot, properties {horizontalAlignment=HAlign_Center, verticalAlignment=VAlign_Center, padding=<ltrb 4 4 4 4>}}
  }
  VerticalBox_Text {
   widgetClassPath=&VerticalBox
   bIsVariable=true
   parent=HorizontalBox_ButtonContent
   widget=VerticalBox_Text
   slot {class=HorizontalBoxSlot, properties {verticalAlignment=VAlign_Center, padding=<ltrb 10 0 10 0>, size {sizeRule=Fill, value=1}}}
  }
  Text_Title {
   widgetClassPath=&CommonTextBlock
   bInherited=true
   parent=VerticalBox_Text
   widget=Text_Title
   properties {
    justification=Center
    mobileFontSizeMultiplier=1.2
    style=/FortUILibrary/Foundation/Buttons/TextStyles/TextStyle_Button_HoldPurchase.TextStyle_Button_HoldPurchase_C
    text="Purchase Gift!"
    textTransformPolicy=ToUpper
    colorAndOpacity {colorUseRule=UseColor_Specified, specifiedColor=<rgba 0.0009 0.006 0.0395 1>}
    font {
     bForceMonospaced=false
     bMaterialIsStencil=false
     fontMaterial=null
     fontObject=&BurbankSmall
     letterSpacing=0
     monospacedWidth=1
     size=22
     skewAmount=0
     typefaceFontName=Black
     outlineSettings {
      bApplyOutlineToDropShadows=false
      bMiteredCorners=false
      bSeparateFillAlpha=false
      outlineMaterial=null
      outlineSize=0
      outlineColor=<rgba 0 0 0 1>
     }
    }
    shadowOffset=<0 0>
   }
   slot {class=VerticalBoxSlot, properties {horizontalAlignment=HAlign_Center, verticalAlignment=VAlign_Center}}
  }
  TextBlock_HoldPrompt {
   widgetClassPath=&CommonTextBlock
   bIsVariable=true
   parent=VerticalBox_Text
   widget=TextBlock_HoldPrompt
   properties {
    bIsScrollingEnabled=false
    justification=Center
    mobileFontSizeMultiplier=1.15
    scrollStyle=&TextScrollStyle-Base_C
    style=/FortUILibrary/Foundation/Buttons/TextStyles/TextStyle_Button_HoldPurchase_SubText.TextStyle_Button_HoldPurchase_SubText_C
    text=(HOLD)
    textTransformPolicy=ToUpper
    visibility=Collapsed
    colorAndOpacity {colorUseRule=UseColor_Specified, specifiedColor=<rgba 0.0009 0.006 0.0395 1>}
    font {
     bForceMonospaced=false
     bMaterialIsStencil=false
     fontMaterial=null
     fontObject=&BurbankSmall
     letterSpacing=0
     monospacedWidth=1
     size=18
     skewAmount=0
     typefaceFontName=Black
     outlineSettings {
      bApplyOutlineToDropShadows=false
      bMiteredCorners=false
      bSeparateFillAlpha=false
      outlineMaterial=null
      outlineSize=0
      outlineColor=<rgba 0 0 0 1>
     }
    }
    renderTransform {angle=0, scale=<1 1>, shear=<0 0>, translation=<0 -4>}
    shadowOffset=<0 0>
   }
   slot {class=VerticalBoxSlot, properties {size {sizeRule=Fill, value=1}}}
  }
  RightExtraContentSlot {
   widgetClassPath=&NamedSlot
   bIsVariable=true
   parent=HorizontalBox_ButtonContent
   widget=RightExtraContentSlot
   properties {bExposeOnInstanceOnly=true}
   slot {class=HorizontalBoxSlot}
  }
  Spacer_EvenSize {
   widgetClassPath=&Spacer
   bIsVariable=true
   parent=HorizontalBox_ButtonContent
   widget=Spacer_EvenSize
   properties {size=<48 1>}
   slot {class=HorizontalBoxSlot, properties {horizontalAlignment=HAlign_Right, verticalAlignment=VAlign_Center}}
  }
 ]
}
