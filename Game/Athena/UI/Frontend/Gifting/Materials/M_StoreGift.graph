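# Translucent UI material for the store gift box, tinting the box and ribbon separately through a channel mask texture.
# An optional UseTriangles switch adds an aspect-corrected, rotated triangle pattern that pulses between two channels, fades vertically and shifts color over time.
# The gift diffuse texture shades both tints and is hard-light blended on top.
# The panned noise lerps identical inputs, so it has no visible effect.
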Material /Game/Athena/UI/Frontend/Gifting/Materials/M_StoreGift
graph {
 properties {
  bUsedWithStaticMesh=false
  blendMode=BLEND_Translucent
  materialDomain=MD_UI
  pixelDepthOffsetMode=PDOM_Legacy
  usedShadingModels=Default Lit
 }
 nodes [
  n1 = TextureSampleParameter2D() {parameterName=ColorChannels, samplerType=SAMPLERTYPE_Masks, texture=&/T_StoreGift-channels}
  Clamp(None=Multiply(A=Max(A=@n1.R, B=@n1.B)) {constB=1.5})
  MaterialFunctionCall() {materialFunction=&ScreenAlignedUVs}
  n2 = VectorParameter() {parameterName=BoxColor2, defaultValue=<rgba 0.4921 0 0.8228 1>}
  n3 = LinearInterpolate(
   A=@n2.RGB
   B=VectorParameter() {parameterName=BoxColor2Shift, defaultValue=<rgba 0.4921 0 0.8228 1>}.RGB
   Alpha=MaterialFunctionCall(Speed=ScalarParameter() {defaultValue=0.2, parameterName=ColorShiftSpeed}) {materialFunction=&DebugTimeSine}.Result
  ) {constB=0.2}
  MaterialFunctionCall(Hue Shift Percentage=@n3, Texture=@n2.RGB) {materialFunction=&HueShift}
  ScalarParameter() {defaultValue=0.2, parameterName=HueShftAmount}
  n4 = VectorParameter() {parameterName=BoxColor, defaultValue=<rgba 0.4921 0 0.8228 1>}
  n5 = MaterialFunctionCall(Float2=MaterialFunctionCall() {materialFunction=&ScreenResolution}.Visible Resolution) {materialFunction=&BreakOutFloat2Components}
  n6 = Max(A=@n5.G) {constB=0.1}
  n7 = TextureCoordinate()
  n8 = Multiply(A=AppendVector(A=Divide(A=@n5.R, B=@n6), B=Constant() {r=1}), B=@n7) {constB=10}
  n9 = AppendVector(
   A=ScalarParameter() {defaultValue=16, group=TriangleBackground, parameterName=TriangleTiles-U}
   B=ScalarParameter() {defaultValue=10, group=TriangleBackground, parameterName=TriangleTiles-V}
  )
  n10 = TextureSample(UVs=Rotator(
    Coordinate=Multiply(A=@n8, B=@n9) {constB=20}
    Time=ScalarParameter() {defaultValue=0.125, group=TriangleBackground, parameterName=TriangleRotation}
   ) {speed=6.2832}) {texture=/Game/UI/Frontend/Campaign/EventCallOut/T-UI-TriangleTile}
  n11 = AppendVector(
   A=ScalarParameter() {defaultValue=1, group=TriangleBackground, parameterName=TriangleBlendTiles-U}
   B=ScalarParameter() {defaultValue=0.5, group=TriangleBackground, parameterName=TriangleBlendTiles-V}
  )
  n12 = Time()
  n13 = ScalarParameter() {defaultValue=1, group=TriangleBackground, parameterName=TimeMultiplier}
  n14 = TextureSample(UVs=Panner(Coordinate=Multiply(A=@n7, B=@n11), Time=Multiply(A=@n12, B=@n13)) {speedX=-0.25, speedY=0.125}) {samplerType=SAMPLERTYPE_Grayscale, texture=&LowResBlurredNoise}
  n15 = LinearInterpolate(A=@n10.B, B=@n10.B, Alpha=@n14.R)
  n16 = LinearInterpolate(A=@n10.R, B=@n10.R, Alpha=@n14.R)
  n17 = Clamp(None=Sine(None=@n12) {period=4})
  n18 = Power(
   Base=Multiply(A=LinearInterpolate(A=@n15, B=@n16, Alpha=@n17), B=ScalarParameter() {defaultValue=1, parameterName=NoiseLevelLift})
   Exp=ScalarParameter() {defaultValue=1, parameterName=Noise_Contrast}
  )
  n19 = ComponentMask(None=TextureCoordinate()) {g=true}
  n20 = LinearInterpolate(
   A=@n4.RGB
   B=Multiply(A=Saturate(None=@n18), B=@n3)
   Alpha=Power(Base=OneMinus(None=@n19), Exp=ScalarParameter() {defaultValue=2, parameterName=TriangleFalloffPow})
  )
  n21 = TextureSampleParameter2D() {parameterName=GiftTexture, texture=&/T_StoreGift-diffuse}
  n22 = LinearInterpolate(B=@n21.RGB, Alpha=ScalarParameter() {defaultValue=1, parameterName=BoxAmbo}) {constA=1}
  n23 = ScalarParameter() {defaultValue=1, parameterName=RibbonAmbo}
  n24 = LinearInterpolate(
   A=StaticSwitchParameter(True=Multiply(A=@n22, B=@n20), False=Multiply(A=@n22, B=@n4.RGB)) {parameterName=UseTriangles}
   B=Multiply(
    A=LinearInterpolate(B=@n21.RGB, Alpha=@n23) {constA=1}
    B=VectorParameter() {parameterName=RibbonColor, defaultValue=<rgba 0 0.9823 0.1504 1>}.RGB
   )
   Alpha=@n1.R
  )
  n25 = LinearInterpolate(
   A=@n24
   B=MaterialFunctionCall!(Base=@n24, Blend=@n21.RGB) {materialFunction=/Engine/Functions/Engine_MaterialFunctions03/Blends/Blend_HardLight}.Result
   Alpha=ScalarParameter() {defaultValue=0.581, parameterName=HardLightAmount, sliderMax=1}
  )
  n26 = ScalarParameter() {defaultValue=1, parameterName=FinalColorPow}
  StaticSwitchParameter(True=Multiply(A=Power(Base=@n25, Exp=@n26), B=ScalarParameter() {defaultValue=1, parameterName=FinalColorMultiply}), False=@n25) {parameterName=UseTriangles}
 ]
}
