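# Selectable gift-wrap option entry for the gifting UI, derived from FortGiftingWrapOptionListEntry.
# A clipped overlay layers a gradient background and inherited lazy-loaded gift image beneath selection and white hover highlights.
# The event graph plays Hover forward only for unselected, non-touch input, reversing it on unhover unless selected.
# Selected and Unselected animations drive scale, opacity and the highlight material's hover parameter.
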WidgetBlueprint /Gifting/UI/GiftingWrapOptionEntry
animations [
 Hover {
  binding_count=2
  length_seconds=0.05
  object=&self:Hover
  sequence {bindings [
    Image_hoverwhite {
     id=#1
     possessed_class=Image
     tracks [{
       type=MovieSceneFloatTrack
       display_name=Render Opacity
       sections [{type=MovieSceneFloatSection, channels [None {type=MovieSceneScriptingFloatChannel, keys [[0,0],[1,1]]}]}]
      }]
    }
    SizeBox_0 {
     id=#2
     display_name=Overlay_0
     possessed_class=SizeBox
     tracks [
      {
       type=MovieScene2DTransformTrack
       sections [{
         type=MovieScene2DTransformSection
         channels [
          Translation.X {type=MovieSceneScriptingFloatChannel}
          Translation.Y {type=MovieSceneScriptingFloatChannel}
          Angle {type=MovieSceneScriptingFloatChannel}
          Scale.X {type=MovieSceneScriptingFloatChannel, keys [[0,1],[1,1.03]]}
          Scale.Y {type=MovieSceneScriptingFloatChannel, keys [[0,1],[1,1.03]]}
          Shear.X {type=MovieSceneScriptingFloatChannel}
          Shear.Y {type=MovieSceneScriptingFloatChannel}
         ]
        }]
      }
      {
       type=MovieSceneFloatTrack
       display_name=Render Opacity
       sections [{type=MovieSceneFloatSection, channels [None {type=MovieSceneScriptingFloatChannel, keys [[0,0.7],[1,1]]}]}]
      }
     ]
    }
   ]}
 }
 Selected {
  binding_count=2
  length_seconds=0.15
  object=&self:Selected
  sequence {bindings [
    Image_selected {
     id=#3
     possessed_class=Image
     tracks [{
       type=MovieSceneWidgetMaterialTrack
       sections [{type=MovieSceneParameterSection, channels [hover {type=MovieSceneScriptingFloatChannel, keys [[0,0],[3,1]]}]}]
      }]
    }
    SizeBox_0 {
     id=#4
     display_name=Overlay_0
     possessed_class=SizeBox
     tracks [
      {
       type=MovieScene2DTransformTrack
       sections [{
         type=MovieScene2DTransformSection
         channels [
          Scale.X {type=MovieSceneScriptingFloatChannel, keys [[0,1.02],[3,1.1]]}
          Scale.Y {type=MovieSceneScriptingFloatChannel, keys [[0,1.02],[3,1.1]]}
         ]
        }]
      }
      {
       type=MovieSceneFloatTrack
       display_name=Render Opacity
       sections [{type=MovieSceneFloatSection, channels [None {type=MovieSceneScriptingFloatChannel, const [1,1]}]}]
      }
     ]
    }
   ]}
 }
 Unselected {
  binding_count=2
  length_seconds=0.15
  object=&self:Unselected
  sequence {bindings [
    Image_selected {
     id=#3
     possessed_class=Image
     tracks [{
       type=MovieSceneWidgetMaterialTrack
       sections [{type=MovieSceneParameterSection, channels [hover {type=MovieSceneScriptingFloatChannel, keys [[0,1],[3,0]]}]}]
      }]
    }
    SizeBox_0 {
     id=#4
     display_name=Overlay_0
     possessed_class=SizeBox
     tracks [
      {
       type=MovieScene2DTransformTrack
       sections [{
         type=MovieScene2DTransformSection
         channels [
          Scale.X {type=MovieSceneScriptingFloatChannel, keys [[0,1.1],[3,1]]}
          Scale.Y {type=MovieSceneScriptingFloatChannel, keys [[0,1.1],[3,1]]}
         ]
        }]
      }
      {
       type=MovieSceneFloatTrack
       display_name=Render Opacity
       sections [{type=MovieSceneFloatSection, channels [None {type=MovieSceneScriptingFloatChannel, keys [[0,1],[3,0.7]]}]}]
      }
     ]
    }
   ]}
 }
]
graphs [EventGraph {nodes [
   Event On Hovered()
   n1 = GetSelected!(self:Common Button Base Object Reference) -> (ReturnValue:bool)
   n2 = IsUsingTouch(PlayerController=GetOwningPlayer())
   > Branch(Condition=AND Boolean(A=NOT Boolean(A=@n2), B=NOT Boolean(A=@n1)))
   n3 = Get Hover() -> (Hover:Widget Animation Object Reference)
   > PlayAnimationForward(InAnimation=@n3)
   Event On Unhovered()
   > n4 = Branch(Condition=@n1)
   PlayAnimationReverse(execute=@n4.else, InAnimation=@n3)
   Event On Deselected()
   > PlayAnimationForward(InAnimation=Get Unselected() -> (Unselected:Widget Animation Object Reference))
   Event On Selected()
   > PlayAnimationForward(InAnimation=Get Selected() -> (Selected:Widget Animation Object Reference))
  ]}]
variables []
widgets {
 info {inheritedWidgetCount=2, parentClass=/Script/GiftingUI.FortGiftingWrapOptionListEntry, rootWidgetClass=&Overlay, widgetCount=7}
 widgets [
  Overlay_0 {widgetClassPath=&Overlay, widget=Overlay_0, properties {renderOpacity=0.7}}
  ImageClip_Overlay {
   widgetClassPath=&Overlay
   parent=Overlay_0
   widget=ImageClip_Overlay
   properties {clipping=ClipToBounds}
   slot {class=OverlaySlot, properties {horizontalAlignment=HAlign_Fill, verticalAlignment=VAlign_Fill}}
  }
  Image_Background {
   widgetClassPath=&Image
   parent=ImageClip_Overlay
   widget=Image_Background
   properties {brush {
     bIsDynamicallyLoaded=false
     drawAs=Image
     imageType=NoImage
     mirroring=NoMirror
     resourceName=None
     resourceObject=/FortniteCoreUI/Frontend/Art/M_UI_RadialGradient_MobileStoreSelect
     tiling=NoTile
     imageSize=<32 32>
     margin=<ltrb 0 0 0 0>
     outlineSettings {
      bUseBrushTransparency=false
      roundingType=HalfHeightRadius
      width=0
      color {colorUseRule=UseColor_Specified, specifiedColor=<rgba 0 0 0 0>}
      cornerRadii=<0 0 0 0>
     }
     tintColor {colorUseRule=UseColor_Specified, specifiedColor=<rgba 1 1 1 1>}
     uVRegion {bIsValid=false, max=<0 0>, min=<0 0>}
    }}
   slot {class=OverlaySlot, properties {horizontalAlignment=HAlign_Fill, verticalAlignment=VAlign_Fill}}
  }
  Image_Gift {
   widgetClassPath=&FortLazyImage
   bInherited=true
   parent=ImageClip_Overlay
   widget=Image_Gift
   properties {
    visibility=Visible
    brush {
     bIsDynamicallyLoaded=false
     drawAs=Image
     imageType=NoImage
     mirroring=NoMirror
     resourceName=None
     resourceObject=/Game/Athena/UI/Frontend/Gifting/Materials/M_StoreGift
     tiling=NoTile
     imageSize=<200 200>
     margin=<ltrb 0 0 0 0>
     outlineSettings {
      bUseBrushTransparency=false
      roundingType=HalfHeightRadius
      width=0
      color {colorUseRule=UseColor_Specified, specifiedColor=<rgba 0 0 0 0>}
      cornerRadii=<0 0 0 0>
     }
     tintColor {colorUseRule=UseColor_Specified, specifiedColor=<rgba 1 1 1 1>}
     uVRegion {bIsValid=false, max=<0 0>, min=<0 0>}
    }
    loadingBackgroundBrush {
     bIsDynamicallyLoaded=false
     drawAs=Image
     imageType=NoImage
     mirroring=NoMirror
     resourceName=None
     resourceObject=null
     tiling=NoTile
     imageSize=<200 200>
     margin=<ltrb 0 0 0 0>
     outlineSettings {
      bUseBrushTransparency=false
      roundingType=HalfHeightRadius
      width=0
      color {colorUseRule=UseColor_Specified, specifiedColor=<rgba 0 0 0 0>}
      cornerRadii=<0 0 0 0>
     }
     tintColor {colorUseRule=UseColor_Specified, specifiedColor=<rgba 1 1 1 1>}
     uVRegion {bIsValid=false, max=<0 0>, min=<0 0>}
    }
    renderTransform {angle=0, scale=<1.2 1.2>, shear=<0 0>, translation=<30 30>}
   }
   slot {class=OverlaySlot, properties {horizontalAlignment=HAlign_Fill, verticalAlignment=VAlign_Fill}}
  }
  Image_selected {
   widgetClassPath=&Image
   bIsVariable=true
   parent=Overlay_0
   widget=Image_selected
   properties {brush {
     bIsDynamicallyLoaded=false
     drawAs=Image
     imageType=NoImage
     mirroring=NoMirror
     resourceName=None
     resourceObject=/Game/Athena/Collections/Art/Fishing/M_Ui_CollectionEntry_Hover_Inst
     tiling=NoTile
     imageSize=<32 32>
     margin=<ltrb 0 0 0 0>
     outlineSettings {
      bUseBrushTransparency=false
      roundingType=HalfHeightRadius
      width=0
      color {colorUseRule=UseColor_Specified, specifiedColor=<rgba 0 0 0 0>}
      cornerRadii=<0 0 0 0>
     }
     tintColor {colorUseRule=UseColor_Specified, specifiedColor=<rgba 1 1 1 1>}
     uVRegion {bIsValid=false, max=<0 0>, min=<0 0>}
    }}
   slot {class=OverlaySlot, properties {horizontalAlignment=HAlign_Fill, verticalAlignment=VAlign_Fill}}
  }
  Image_hoverwhite {
   widgetClassPath=&Image
   parent=Overlay_0
   widget=Image_hoverwhite
   properties {renderOpacity=0, colorAndOpacity=<rgba 1 1 1 0.1>}
   slot {class=OverlaySlot, properties {horizontalAlignment=HAlign_Fill, verticalAlignment=VAlign_Fill}}
  }
  InputActionWidget {widgetClassPath=&CommonActionWidget, bInherited=true, widget=null}
 ]
}
