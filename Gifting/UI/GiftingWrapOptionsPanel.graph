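# Gift-wrap selection panel for the gifting UI, where the player picks a wrap and confirms it.
# Lays out a "Wrapping" header, a horizontal tile view of GiftingWrapOptionEntry widgets and a confirm button bound to the ConfirmWrapping input action.
# The native AthenaGiftingWrapOptionsPanel parent supplies the tile view and button bindings; the Blueprint graph only plays Intro on activation.
# Intro staggers a short fade and slide-in from the left across five sections.
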WidgetBlueprint /Gifting/UI/GiftingWrapOptionsPanel
animations [Intro {
  binding_count=6
  length_seconds=0.3501
  object=&self:Intro
  sequence {bindings [
    TileView_WrapOptions {
     id=#1
     possessed_class=CommonTileView
     tracks [
      {
       type=MovieSceneFloatTrack
       display_name=Render Opacity
       sections [{type=MovieSceneFloatSection, channels [None {type=MovieSceneScriptingFloatChannel, keys [[3.002,0],[6.0023,1]]}]}]
      }
      {
       type=MovieScene2DTransformTrack
       sections [{
         type=MovieScene2DTransformSection
         channels [
          Translation.X {type=MovieSceneScriptingFloatChannel, keys [[3.002,-130],[6.002,0]]}
          Translation.Y {type=MovieSceneScriptingFloatChannel, const [0,1]}
         ]
        }]
      }
     ]
    }
    Text_PanelInfo {
     id=#2
     possessed_class=CommonTextBlock
     tracks [
      {
       type=MovieSceneFloatTrack
       display_name=Render Opacity
       sections [{type=MovieSceneFloatSection, channels [None {type=MovieSceneScriptingFloatChannel, keys [[1.0007,0],[4.001,1]]}]}]
      }
      {
       type=MovieScene2DTransformTrack
       sections [{
         type=MovieScene2DTransformSection
         channels [
          Translation.X {type=MovieSceneScriptingFloatChannel, keys [[1.0007,-130],[4.0007,0]]}
          Translation.Y {type=MovieSceneScriptingFloatChannel, const [0,1]}
         ]
        }]
      }
     ]
    }
    SafeZone_Header {
     id=#3
     possessed_class=SafeZone
     tracks [
      {
       type=MovieScene2DTransformTrack
       sections [{
         type=MovieScene2DTransformSection
         channels [
          Translation.X {type=MovieSceneScriptingFloatChannel, keys [[0,-130],[3,0]]}
          Translation.Y {type=MovieSceneScriptingFloatChannel, const [0,1]}
          Angle {type=MovieSceneScriptingFloatChannel, const [0,1]}
          Scale.X {type=MovieSceneScriptingFloatChannel, const [1,1]}
          Scale.Y {type=MovieSceneScriptingFloatChannel, const [1,1]}
          Shear.X {type=MovieSceneScriptingFloatChannel, const [0,1]}
          Shear.Y {type=MovieSceneScriptingFloatChannel, const [0,1]}
         ]
        }]
      }
      {
       type=MovieSceneFloatTrack
       display_name=Render Opacity
       sections [{type=MovieSceneFloatSection, channels [None {type=MovieSceneScriptingFloatChannel, keys [[0,0],[3.0003,1]]}]}]
      }
     ]
    }
    SB_Button {
     id=#4
     possessed_class=SizeBox
     tracks [
      {
       type=MovieSceneFloatTrack
       display_name=Render Opacity
       sections [{type=MovieSceneFloatSection, channels [None {type=MovieSceneScriptingFloatChannel, keys [[4.001,0],[7.0013,1]]}]}]
      }
      {
       type=MovieScene2DTransformTrack
       sections [{type=MovieScene2DTransformSection, channels [Translation.X {type=MovieSceneScriptingFloatChannel, keys [[4.001,-130],[7.001,0]]}]}]
      }
     ]
    }
    SizeBox_MainContent {
     id=#5
     possessed_class=SizeBox
     tracks [
      {
       type=MovieSceneFloatTrack
       display_name=Render Opacity
       sections [{type=MovieSceneFloatSection, channels [None {type=MovieSceneScriptingFloatChannel, keys [[2,0],[5.0003,1]]}]}]
      }
      {
       type=MovieScene2DTransformTrack
       sections [{
         type=MovieScene2DTransformSection
         channels [
          Translation.X {type=MovieSceneScriptingFloatChannel, keys [[2,-130],[5,0]]}
          Translation.Y {type=MovieSceneScriptingFloatChannel, const [0,1]}
         ]
        }]
      }
     ]
    }
   ]}
 }]
graphs [EventGraph {nodes [
   Event On Activated()
   > PlayAnimation(InAnimation=Get Intro())
  ]}]
variables []
widgets {
 info {inheritedWidgetCount=2, parentClass=/Script/GiftingUI.AthenaGiftingWrapOptionsPanel, rootWidgetClass=&Overlay, widgetCount=24}
 widgets [
  Modal_ConfirmGifting {widgetClassPath=&Overlay, widget=Modal_ConfirmGifting}
  VB_GiftingInfo {
   widgetClassPath=&VerticalBox
   parent=Modal_ConfirmGifting
   widget=VB_GiftingInfo
   slot {class=OverlaySlot, properties {horizontalAlignment=HAlign_Fill, verticalAlignment=VAlign_Fill}}
  }
  SafeZone_Header {
   widgetClassPath=&SafeZone
   bIsVariable=true
   parent=VB_GiftingInfo
   widget=SafeZone_Header
   properties {padBottom=false, padRight=false}
   slot {class=VerticalBoxSlot, properties {horizontalAlignment=HAlign_Left, verticalAlignment=VAlign_Top, padding=<ltrb 100 120 0 0>}}
  }
  ModalHeader {widgetClassPath=&HorizontalBox, bIsVariable=true, parent=SafeZone_Header, widget=ModalHeader, slot {class=SafeZoneSlot}}
  ScaleB_HeaderIcon {
   widgetClassPath=&ScaleBox
   parent=ModalHeader
   widget=ScaleB_HeaderIcon
   slot {class=HorizontalBoxSlot, properties {padding=<ltrb 0 0 10 0>}}
  }
  HeaderIcon {
   widgetClassPath=&Image
   parent=ScaleB_HeaderIcon
   widget=HeaderIcon
   properties {brush {
     bIsDynamicallyLoaded=false
     drawAs=Image
     imageType=NoImage
     mirroring=NoMirror
     resourceName=None
     resourceObject=/Gifting/UI/Art/T_Gifting
     tiling=NoTile
     imageSize=<88 80>
     margin=<ltrb 0 0 0 0>
     outlineSettings {
      bUseBrushTransparency=false
      roundingType=HalfHeightRadius
      width=0
      color {colorUseRule=UseColor_Specified, specifiedColor=<rgba 0 0 0 0>}
      cornerRadii=<0 0 0 0>
     }
     tintColor {colorUseRule=UseColor_Specified, specifiedColor=<rgba 1 1 1 1>}
     uVRegion {bIsValid=false, max=<0 0>, min=<0 0>}
    }}
   slot {class=ScaleBoxSlot}
  }
  DividerLine {
   widgetClassPath=&Image
   bIsVariable=true
   parent=ModalHeader
   widget=DividerLine
   properties {brush {
     bIsDynamicallyLoaded=false
     drawAs=Image
     imageType=NoImage
     mirroring=NoMirror
     resourceName=None
     resourceObject=/Game/Items/Traps/Blueprints/Toys/UI_Art_Assets/M_UI_ActionDividerLine
     tiling=NoTile
     imageSize=<24 32>
     margin=<ltrb 0 0 0 0>
     outlineSettings {
      bUseBrushTransparency=false
      roundingType=HalfHeightRadius
      width=0
      color {colorUseRule=UseColor_Specified, specifiedColor=<rgba 0 0 0 0>}
      cornerRadii=<0 0 0 0>
     }
     tintColor {colorUseRule=UseColor_Specified, specifiedColor=<rgba 0.0369 0.7835 0.9131 1>}
     uVRegion {bIsValid=false, max=<0 0>, min=<0 0>}
    }}
   slot {class=HorizontalBoxSlot}
  }
  VerticalBox_74 {
   widgetClassPath=&VerticalBox
   parent=ModalHeader
   widget=VerticalBox_74
   properties {renderTransform {angle=0, scale=<1 1>, shear=<-12 0>, translation=<0 0>}}
   slot {class=HorizontalBoxSlot, properties {padding=<ltrb 4 0 0 0>}}
  }
  HeaderColor {
   widgetClassPath=&Border
   bIsVariable=true
   parent=VerticalBox_74
   widget=HeaderColor
   properties {
    background {
     bIsDynamicallyLoaded=false
     drawAs=NoDrawType
     imageType=NoImage
     mirroring=NoMirror
     resourceName=None
     resourceObject=null
     tiling=NoTile
     imageSize=<32 32>
     margin=<ltrb 0 0 0 0>
     outlineSettings {
      bUseBrushTransparency=false
      roundingType=HalfHeightRadius
      width=0
      color {colorUseRule=UseColor_Specified, specifiedColor=<rgba 0 0 0 0>}
      cornerRadii=<0 0 0 0>
     }
     tintColor {colorUseRule=UseColor_Specified, specifiedColor=<rgba 1 1 1 1>}
     uVRegion {bIsValid=false, max=<0 0>, min=<0 0>}
    }
    padding=<ltrb 0 -5 0 0>
   }
   slot {class=VerticalBoxSlot}
  }
  HeaderMainText {
   widgetClassPath=&CommonTextBlock
   bIsVariable=true
   parent=HeaderColor
   widget=HeaderMainText
   properties {
    text=Wrapping
    textTransformPolicy=ToUpper
    font {
     bForceMonospaced=false
     bMaterialIsStencil=false
     fontMaterial=null
     fontObject=&Burbank
     letterSpacing=0
     monospacedWidth=1
     size=38
     skewAmount=0
     typefaceFontName=BurbankRegular-Black
     outlineSettings {
      bApplyOutlineToDropShadows=false
      bMiteredCorners=false
      bSeparateFillAlpha=false
      outlineMaterial=null
      outlineSize=0
      outlineColor=<rgba 0 0 0 1>
     }
    }
    margin=<ltrb 4 0 4 0>
    shadowOffset=<0 0>
   }
   slot {class=BorderSlot, properties {padding=<ltrb 0 -5 0 0>}}
  }
  SubtextColor {
   widgetClassPath=&Border
   bIsVariable=true
   parent=VerticalBox_74
   widget=SubtextColor
   properties {
    background {
     bIsDynamicallyLoaded=false
     drawAs=NoDrawType
     imageType=NoImage
     mirroring=NoMirror
     resourceName=None
     resourceObject=null
     tiling=NoTile
     imageSize=<32 32>
     margin=<ltrb 0 0 0 0>
     outlineSettings {
      bUseBrushTransparency=false
      roundingType=HalfHeightRadius
      width=0
      color {colorUseRule=UseColor_Specified, specifiedColor=<rgba 0 0 0 0>}
      cornerRadii=<0 0 0 0>
     }
     tintColor {colorUseRule=UseColor_Specified, specifiedColor=<rgba 1 1 1 1>}
     uVRegion {bIsValid=false, max=<0 0>, min=<0 0>}
    }
    padding=<ltrb 0 -10 0 0>
   }
   slot {class=VerticalBoxSlot}
  }
  SubHeaderText {
   widgetClassPath=&CommonTextBlock
   bIsVariable=true
   parent=SubtextColor
   widget=SubHeaderText
   properties {
    text=Gifting
    textTransformPolicy=ToUpper
    colorAndOpacity {colorUseRule=UseColor_Specified, specifiedColor=<rgba 0.0369 0.7835 0.9131 1>}
    font {
     bForceMonospaced=false
     bMaterialIsStencil=false
     fontMaterial=null
     fontObject=&Burbank
     letterSpacing=0
     monospacedWidth=1
     size=20
     skewAmount=0
     typefaceFontName=BurbankRegular-Black
     outlineSettings {
      bApplyOutlineToDropShadows=false
      bMiteredCorners=false
      bSeparateFillAlpha=false
      outlineMaterial=null
      outlineSize=0
      outlineColor=<rgba 0 0 0 1>
     }
    }
    margin=<ltrb 4 0 4 0>
    shadowOffset=<0 0>
   }
   slot {class=BorderSlot, properties {padding=<ltrb 0 -10 0 0>}}
  }
  SizeBox_MainContent {
   widgetClassPath=&SizeBox
   bIsVariable=true
   parent=VB_GiftingInfo
   widget=SizeBox_MainContent
   properties {bOverride_MinDesiredHeight=true, heightOverride=500, minDesiredHeight=390}
   slot {class=VerticalBoxSlot, properties {padding=<ltrb 0 20 0 0>}}
  }
  Overlay_57 {widgetClassPath=&Overlay, parent=SizeBox_MainContent, widget=Overlay_57, slot {class=SizeBoxSlot}}
  Background {
   widgetClassPath=&Image
   parent=Overlay_57
   widget=Background
   properties {
    visibility=SelfHitTestInvisible
    brush {
     bIsDynamicallyLoaded=false
     drawAs=Image
     imageType=NoImage
     mirroring=NoMirror
     resourceName=None
     resourceObject=/Game/Athena/UI/Subscriptions/Art/M_UI_Subs_Color_FIll_Modals
     tiling=NoTile
     imageSize=<32 32>
     margin=<ltrb 0 0 0 0>
     outlineSettings {
      bUseBrushTransparency=false
      roundingType=HalfHeightRadius
      width=0
      color {colorUseRule=UseColor_Specified, specifiedColor=<rgba 0 0 0 0>}
      cornerRadii=<0 0 0 0>
     }
     tintColor {colorUseRule=UseColor_Specified, specifiedColor=<rgba 1 1 1 0.5>}
     uVRegion {bIsValid=false, max=<0 0>, min=<0 0>}
    }
   }
   slot {class=OverlaySlot, properties {horizontalAlignment=HAlign_Fill, verticalAlignment=VAlign_Fill}}
  }
  BackgroundLine {
   widgetClassPath=&Image
   parent=Overlay_57
   widget=BackgroundLine
   properties {
    visibility=SelfHitTestInvisible
    brush {
     bIsDynamicallyLoaded=false
     drawAs=Image
     imageType=NoImage
     mirroring=NoMirror
     resourceName=None
     resourceObject=/FortUILibrary/Foundation/Art_ToolKit/AngledBoxes/M_UI_MoreInfo_BottomDividerLIne_Blue
     tiling=NoTile
     imageSize=<32 26>
     margin=<ltrb 0 0 0 0>
     outlineSettings {
      bUseBrushTransparency=false
      roundingType=HalfHeightRadius
      width=0
      color {colorUseRule=UseColor_Specified, specifiedColor=<rgba 0 0 0 0>}
      cornerRadii=<0 0 0 0>
     }
     tintColor {colorUseRule=UseColor_Specified, specifiedColor=<rgba 0.3263 0.3457 0.4149 1>}
     uVRegion {bIsValid=false, max=<0 0>, min=<0 0>}
    }
   }
   slot {class=OverlaySlot, properties {horizontalAlignment=HAlign_Fill, verticalAlignment=VAlign_Bottom}}
  }
  Overlay_GiftInfo {
   widgetClassPath=&Overlay
   parent=Overlay_57
   widget=Overlay_GiftInfo
   slot {class=OverlaySlot, properties {horizontalAlignment=HAlign_Fill, verticalAlignment=VAlign_Fill, padding=<ltrb 40 0 40 0>}}
  }
  ScaleBox {
   widgetClassPath=&ScaleBox
   parent=Overlay_GiftInfo
   widget=ScaleBox
   properties {stretch=None, stretchDirection=DownOnly}
   slot {class=OverlaySlot, properties {horizontalAlignment=HAlign_Fill, verticalAlignment=VAlign_Fill}}
  }
  CommonBorder_3 {
   widgetClassPath=&CommonBorder
   parent=ScaleBox
   widget=CommonBorder_3
   properties {padding=<ltrb 4 120 4 60>}
   slot {class=ScaleBoxSlot, properties {horizontalAlignment=HAlign_Fill}}
  }
  SizeBox_WrapOptions {
   widgetClassPath=&SizeBox
   parent=CommonBorder_3
   widget=SizeBox_WrapOptions
   properties {bOverride_HeightOverride=true, heightOverride=340}
   slot {class=BorderSlot, properties {padding=<ltrb 4 120 4 60>}}
  }
  TileView_WrapOptions {
   widgetClassPath=/Script/CommonUI.CommonTileView
   bInherited=true
   bIsVariable=true
   parent=SizeBox_WrapOptions
   widget=TileView_WrapOptions
   properties {
    entryHeight=280
    entryWidgetClass=&/GiftingWrapOptionEntry.GiftingWrapOptionEntry_C
    entryWidth=280
    horizontalEntrySpacing=40
    navigation=&self:WidgetTree.TileView_WrapOptions.WidgetNavigation_1
    numDesignerPreviewEntries=4
    orientation=Orient_Horizontal
    tileAlignment=CenterAligned
    verticalEntrySpacing=40
   }
   slot {class=SizeBoxSlot, properties {horizontalAlignment=HAlign_Center}}
  }
  Text_PanelInfo {
   widgetClassPath=&CommonTextBlock
   parent=Overlay_GiftInfo
   widget=Text_PanelInfo
   properties {
    justification=Center
    text="Personalize your gift!"
    textTransformPolicy=ToUpper
    font {
     bForceMonospaced=false
     bMaterialIsStencil=false
     fontMaterial=null
     fontObject=&Burbank
     letterSpacing=0
     monospacedWidth=1
     size=70
     skewAmount=0
     typefaceFontName=BurbankRegular-Black
     outlineSettings {
      bApplyOutlineToDropShadows=false
      bMiteredCorners=false
      bSeparateFillAlpha=false
      outlineMaterial=null
      outlineSize=0
      outlineColor=<rgba 0 0 0 1>
     }
    }
    renderTransform {angle=0, scale=<1 1>, shear=<-12 0>, translation=<0 0>}
    shadowOffset=<0 0>
   }
   slot {class=OverlaySlot, properties {horizontalAlignment=HAlign_Center, padding=<ltrb 0 40 0 0>}}
  }
  SB_Button {
   widgetClassPath=&SizeBox
   parent=Overlay_57
   widget=SB_Button
   properties {
    bOverride_HeightOverride=true
    bOverride_MinDesiredWidth=true
    heightOverride=128
    minDesiredWidth=416
    renderTransform {angle=0, scale=<1 1>, shear=<0 0>, translation=<0 64>}
   }
   slot {class=OverlaySlot, properties {horizontalAlignment=HAlign_Center, verticalAlignment=VAlign_Bottom, padding=<ltrb 20 20 10 0>}}
  }
  Button_ConfirmWrap {
   widgetClassPath=/FortniteCoreUI/Frontend/Spatial/Customization/Spatial_Specialbutton
   bInherited=true
   bIsVariable=true
   parent=SB_Button
   widget=Button_ConfirmWrap
   properties {
    auto wrap text=false
    button text=Confirm Wrapping
    fontSize=28
    fontSizeOverride=true
    justifiction=Center
    material Brush=/Game/Athena/UI/Frontend/Spatial/Art/M_Ui_Skewed_SpatialButton_MoreInfo_Blue
    onButtonBaseClicked=(&self:WidgetTree.Button_ConfirmWrap.HandleOnButtonClicked)
    onButtonBaseDoubleClicked=(&self:WidgetTree.Button_ConfirmWrap.HandleOnButtonDoubleClicked)
    onButtonBaseHovered=(&self:WidgetTree.Button_ConfirmWrap.HandleOnButtonHovered)
    onButtonBaseUnhovered=(&self:WidgetTree.Button_ConfirmWrap.HandleOnButtonUnhovered)
    onSelectedChangedBase=(&self:WidgetTree.Button_ConfirmWrap.HandleOnSelectedChanged)
    right Image=null
    shearText=true
    simpleHover=true
    hoveredSlateSoundOverride {resourceObject=/Game/Sounds/Fort_UI_Sounds/FrontEnd/DefaultSounds/PressedButtons/MSS_UI_Button_Hover_Default}
    pressedSlateSoundOverride {resourceObject=/Game/Sounds/Fort_UI_Sounds/FrontEnd/New_FrontEnd/Fortnite_UI_MainMenu_Generic_01_Cue}
    text Color {colorUseRule=UseColor_Specified, specifiedColor=<rgba 0.75 1 1 1>}
    triggeringInputAction {dataTable=/Gifting/GiftingInputActionDataTable, rowName=ConfirmWrapping}
   }
   slot {class=SizeBoxSlot}
  }
 ]
}
