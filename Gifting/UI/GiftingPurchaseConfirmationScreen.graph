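# Gift purchase confirmation screen built on AthenaGiftingConfirmationScreen, with wallet, player name, Wrap Options and Back over a blurred backdrop.
# OnPresentationModeChanged switches between purchase, wrap options, sending, confirmed and error panels, adding an intro animation and sound while sending.
# CreateRightOfWithdrawal rebinds OnROWConfirmed to the supplied delegate and, when required, pushes a modal right-of-withdrawal dialog.
# Accepting that dialog fires OnROWConfirmed.
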WidgetBlueprint /Gifting/UI/GiftingPurchaseConfirmationScreen
animations [GiftingProcess_Intro {
  binding_count=2
  length_seconds=0.2
  object=&self:GiftingProcess_Intro
  sequence {bindings [
    Text_SendingGift {
     id=#1
     possessed_class=CommonTextBlock
     tracks [
      {
       type=MovieSceneFloatTrack
       display_name=Render Opacity
       sections [{type=MovieSceneFloatSection, channels [None {type=MovieSceneScriptingFloatChannel, keys [[0,0],[3.0003,1]]}]}]
      }
      {
       type=MovieScene2DTransformTrack
       sections [{
         type=MovieScene2DTransformSection
         channels [
          Translation.X {type=MovieSceneScriptingFloatChannel, keys [[0,-130],[3,0]]}
          Translation.Y {type=MovieSceneScriptingFloatChannel, const [0,1]}
         ]
        }]
      }
     ]
    }
    SizeBox_0 {
     id=#2
     possessed_class=SizeBox
     tracks [{
       type=MovieSceneFloatTrack
       display_name=Render Opacity
       sections [{type=MovieSceneFloatSection, channels [None {type=MovieSceneScriptingFloatChannel, keys [[1,0],[4.0003,1]]}]}]
      }]
    }
   ]}
 }]
graphs [
 EventGraph {nodes [
   n1 = OnCpcDialogDismissed!() -> (OutputDelegate, then, Accepted:bool=~) {is=title K2Node_CustomEvent}
   > Branch(Condition=@n1.Accepted)
   > Call On ROWConfirmed!(execute:Exec, self:Gifting Purchase Confirmation Screen Object Reference) -> (then:Exec) {is=title K2Node_CallDelegate}
   n2 = Event CreateRightOfWithdrawal!() -> (OutputDelegate, then, ConfirmedDelegate:"Delegate (by ref)") {is=title K2Node_Event}
   > Unbind all Events from On ROWConfirmed!(execute:Exec, self:Gifting Purchase Confirmation Screen Object Reference) -> (then:Exec) {is=title K2Node_ClearDelegate}
   > Bind Event to On ROWConfirmed(self:Gifting Purchase Confirmation Screen Object Reference, Delegate=@n2.ConfirmedDelegate) {is=title K2Node_AddDelegate}
   > n3 = ShouldShowRightOfWithdrawalModal!(execute, Context:Widget Object Reference) -> (then, ReturnValue:bool)
   > Branch(Condition=@n3)
   > n4 = Construct Object from Class(Class=/FortniteCoreUI/Frontend/RightOfWithdrawal/DialogVM_RightOfWithdrawal.DialogVM_RightOfWithdrawal_C) -> (ReturnValue:Dialog VM Right Of Withdrawal Object Reference)
   > n5 = Create Widget!(
    execute:Exec
    Class:User Widget Class Reference=/FortniteCoreUI/Frontend/RightOfWithdrawal/WBP_RightOfWithdrawal_Dialog.WBP_RightOfWithdrawal_Dialog_C
    OwningPlayer:Player Controller Object Reference=GetOwningPlayer()
    DialogVM_RightOfWithdrawal:Dialog VM Right Of Withdrawal Object Reference=@n4
    BodyText:Text="LOCTABLE(\"/ItemShop/MultiProduct/Data/ST_ItemShop\", \"CPC_DefaultText\")"
   ) -> (then:Exec, ReturnValue:WBP Right Of Withdrawal Dialog Object Reference)
   > Unbind Event from On Dialog Dismissed(self:WBP Right Of Withdrawal Dialog Object Reference=@n5, Delegate=@n1) {is=title K2Node_RemoveDelegate}
   > Bind Event to On Dialog Dismissed(self:WBP Right Of Withdrawal Dialog Object Reference=@n5, Delegate=@n1) {is=title K2Node_AddDelegate}
   > TryPushWidgetAsModalPanel!(execute, PushWidget:User Widget Object Reference=@n5) -> (then)
  ]}
 OnPresentationModeChanged {nodes [
   OnPresentationModeChanged!() -> (then, NewMode:EGiftingPresentationMode Enum=ConfirmPurchase) {is=title K2Node_FunctionEntry}
   > SetActiveWidget(
    self:Widget Switcher Object Reference=Get Switcher_Panels() -> (Switcher_Panels:Common Activatable Widget Switcher Object Reference)
    Widget=Select!(
     ConfirmPurchase:Widget Object Reference=Get Panel_GiftingPurchase() -> (Panel_GiftingPurchase:Athena Gifting Purchase Panel Object Reference)
     WrapOptions:Widget Object Reference=Get Panel_WrapOptions() -> (Panel_WrapOptions:Athena Gifting Wrap Options Panel Object Reference)
     GiftingProcess:Widget Object Reference=Get Panel_GiftingProcess() -> (Panel_GiftingProcess:Overlay Object Reference)
     GiftingError:Widget Object Reference=Get Panel_GiftingErrors() -> (Panel_GiftingErrors:Athena Gifting Errors Panel Object Reference)
     GiftConfirmed:Widget Object Reference=Get Panel_GiftConfirmed() -> (Panel_GiftConfirmed:Athena Gift Confirmed Panel Object Reference)
     None:Widget Object Reference
     Index:EGiftingPresentationMode Enum=Get NewMode() -> (NewMode:EGiftingPresentationMode Enum=ConfirmPurchase)
    ) -> (ReturnValue:Widget Object Reference)
   )
   > Branch(Condition="Equal (Enum)"(A:EGiftingPresentationMode Enum=Get NewMode() -> (NewMode=ConfirmPurchase), B:EGiftingPresentationMode Enum=GiftingProcess))
   > PlayAnimation(InAnimation=Get GiftingProcess_Intro() -> (GiftingProcess_Intro:Widget Animation Object Reference))
   > PlaySound2D(Sound=/Game/Sounds/Fort_UI_Sounds/FrontEnd/RewardsAndGifts/MSS_UI_Gift_Sent_Processing_Sweetener)
  ]}
 OnROWConfirmed {nodes [OnROWConfirmed() {is=title K2Node_FunctionEntry}]}
]
variables [OnROWConfirmed {}]
widgets {
 info {inheritedWidgetCount=8, parentClass=/Script/GiftingUI.AthenaGiftingConfirmationScreen, rootWidgetClass=&Overlay, widgetCount=23}
 widgets [
  Overlay_0 {widgetClassPath=&Overlay, widget=Overlay_0}
  BackgroundDark {
   widgetClassPath=&Image
   bIsVariable=true
   parent=Overlay_0
   widget=BackgroundDark
   properties {
    visibility=HitTestInvisible
    brush {
     bIsDynamicallyLoaded=false
     drawAs=Image
     imageType=NoImage
     mirroring=NoMirror
     resourceName=None
     resourceObject=/Game/Athena/UI/Frontend/Art/M_UI_RadialGradientMaster_Dark
     tiling=NoTile
     imageSize=<32 32>
     margin=<ltrb 0 0 0 0>
     outlineSettings {
      bUseBrushTransparency=false
      roundingType=HalfHeightRadius
      width=0
      color {colorUseRule=UseColor_Specified, specifiedColor=<rgba 0 0 0 0>}
      cornerRadii=<0 0 0 0>
     }
     tintColor {colorUseRule=UseColor_Specified, specifiedColor=<rgba 1 1 1 1>}
     uVRegion {bIsValid=false, max=<0 0>, min=<0 0>}
    }
   }
   slot {class=OverlaySlot, properties {horizontalAlignment=HAlign_Fill, verticalAlignment=VAlign_Fill}}
  }
  BackgroundBlur_20 {
   widgetClassPath=&BackgroundBlur
   parent=Overlay_0
   widget=BackgroundBlur_20
   properties {
    blurStrength=12
    visibility=Visible
    lowQualityFallbackBrush {
     bIsDynamicallyLoaded=false
     drawAs=Image
     imageType=NoImage
     mirroring=NoMirror
     resourceName=None
     resourceObject=null
     tiling=NoTile
     imageSize=<0 0>
     margin=<ltrb 0 0 0 0>
     outlineSettings {
      bUseBrushTransparency=false
      roundingType=HalfHeightRadius
      width=0
      color {colorUseRule=UseColor_Specified, specifiedColor=<rgba 0 0 0 0>}
      cornerRadii=<0 0 0 0>
     }
     tintColor {colorUseRule=UseColor_Specified, specifiedColor=<rgba 0 0 0 0.75>}
     uVRegion {bIsValid=false, max=<0 0>, min=<0 0>}
    }
   }
   slot {class=OverlaySlot, properties {horizontalAlignment=HAlign_Fill, verticalAlignment=VAlign_Fill, padding=<ltrb -2 -2 -2 -2>}}
  }
  Switcher_Panels {
   widgetClassPath=&CommonActivatableWidgetSwitcher
   bIsVariable=true
   parent=Overlay_0
   widget=Switcher_Panels
   properties {transitionDuration=0}
   slot {class=OverlaySlot, properties {horizontalAlignment=HAlign_Fill, verticalAlignment=VAlign_Fill}}
  }
  Panel_GiftingPurchase {
   widgetClassPath=&/GiftingPurchasePanel
   bInherited=true
   parent=Switcher_Panels
   widget=Panel_GiftingPurchase
   slot {class=WidgetSwitcherSlot}
  }
  Panel_WrapOptions {
   widgetClassPath=&/GiftingWrapOptionsPanel
   bInherited=true
   parent=Switcher_Panels
   widget=Panel_WrapOptions
   slot {class=WidgetSwitcherSlot}
  }
  Panel_GiftingProcess {widgetClassPath=&Overlay, bIsVariable=true, parent=Switcher_Panels, widget=Panel_GiftingProcess, slot {class=WidgetSwitcherSlot}}
  VerticalBox_0 {
   widgetClassPath=&VerticalBox
   parent=Panel_GiftingProcess
   widget=VerticalBox_0
   slot {class=OverlaySlot, properties {horizontalAlignment=HAlign_Center, verticalAlignment=VAlign_Center}}
  }
  Text_SendingGift {
   widgetClassPath=&CommonTextBlock
   parent=VerticalBox_0
   widget=Text_SendingGift
   properties {
    justification=Center
    text=Sending Gift...
    textTransformPolicy=ToUpper
    font {
     bForceMonospaced=false
     bMaterialIsStencil=false
     fontMaterial=null
     fontObject=&BurbankSmall
     letterSpacing=0
     monospacedWidth=1
     size=37
     skewAmount=0
     typefaceFontName=Black
     outlineSettings {
      bApplyOutlineToDropShadows=false
      bMiteredCorners=false
      bSeparateFillAlpha=false
      outlineMaterial=null
      outlineSize=0
      outlineColor=<rgba 0 0 0 1>
     }
    }
    renderTransform {angle=0, scale=<1 1>, shear=<-12 0>, translation=<0 0>}
    shadowOffset=<0 0>
   }
   slot {class=VerticalBoxSlot}
  }
  SizeBox_0 {
   widgetClassPath=&SizeBox
   parent=VerticalBox_0
   widget=SizeBox_0
   properties {bOverride_MinDesiredHeight=true, bOverride_MinDesiredWidth=true, minDesiredHeight=200, minDesiredWidth=200}
   slot {class=VerticalBoxSlot}
  }
  LoadGuard_GiftingProcess {
   widgetClassPath=/Script/CommonUI.CommonLoadGuard
   parent=SizeBox_0
   widget=LoadGuard_GiftingProcess
   properties {bShowLoading=true, loadingText="", textStyle=&TextStyle-Base_C}
   slot {class=SizeBoxSlot}
  }
  Panel_GiftConfirmed {
   widgetClassPath=&/GiftConfirmedPanel
   bInherited=true
   parent=Switcher_Panels
   widget=Panel_GiftConfirmed
   slot {class=WidgetSwitcherSlot}
  }
  Panel_GiftingErrors {
   widgetClassPath=&/GiftingErrorsPanel
   bInherited=true
   parent=Switcher_Panels
   widget=Panel_GiftingErrors
   slot {class=WidgetSwitcherSlot}
  }
  SafeZone_TopRight {
   widgetClassPath=&SafeZone
   parent=Overlay_0
   widget=SafeZone_TopRight
   properties {padBottom=false, padLeft=false}
   slot {class=OverlaySlot, properties {horizontalAlignment=HAlign_Right, padding=<ltrb 0 24 24 0>}}
  }
  HorizontalBox_1 {widgetClassPath=&HorizontalBox, parent=SafeZone_TopRight, widget=HorizontalBox_1, slot {class=SafeZoneSlot}}
  VerticalBox_1 {
   widgetClassPath=&VerticalBox
   parent=HorizontalBox_1
   widget=VerticalBox_1
   slot {class=HorizontalBoxSlot, properties {horizontalAlignment=HAlign_Right}}
  }
  Button_MtxWallet {
   widgetClassPath=/Game/UI/Frontend/TopBar/MTXButton
   bInherited=true
   parent=VerticalBox_1
   widget=Button_MtxWallet
   properties {
    onButtonBaseClicked=(&self:WidgetTree.Button_MtxWallet.HandleOnButtonClicked)
    onButtonBaseDoubleClicked=(&self:WidgetTree.Button_MtxWallet.HandleOnButtonDoubleClicked)
    onButtonBaseHovered=(&self:WidgetTree.Button_MtxWallet.HandleOnButtonHovered)
    onButtonBaseUnhovered=(&self:WidgetTree.Button_MtxWallet.HandleOnButtonUnhovered)
    onSelectedChangedBase=(&self:WidgetTree.Button_MtxWallet.HandleOnSelectedChanged)
   }
   slot {class=VerticalBoxSlot, properties {horizontalAlignment=HAlign_Right}}
  }
  FortOwningPlayerSocialNameTextBlock {
   widgetClassPath=/Game/UI/Social/FortOwningPlayerSocialNameTextBlock
   parent=VerticalBox_1
   widget=FortOwningPlayerSocialNameTextBlock
   slot {class=VerticalBoxSlot, properties {horizontalAlignment=HAlign_Right, padding=<ltrb 0 4 0 0>}}
  }
  SafeZone_BottomRight {
   widgetClassPath=&SafeZone
   parent=Overlay_0
   widget=SafeZone_BottomRight
   properties {padLeft=false, padTop=false}
   slot {class=OverlaySlot, properties {horizontalAlignment=HAlign_Right, verticalAlignment=VAlign_Bottom, padding=<ltrb 0 0 24 24>}}
  }
  HorizontalBox_0 {widgetClassPath=&HorizontalBox, parent=SafeZone_BottomRight, widget=HorizontalBox_0, slot {class=SafeZoneSlot}}
  Button_WrapOptions {
   widgetClassPath=/FortUILibrary/Foundation/Buttons/SoloButton
   bInherited=true
   parent=HorizontalBox_0
   widget=Button_WrapOptions
   properties {
    button Description=Wrap Options
    fontSize=17
    fontSizeOveride=true
    minHeight=18
    minWidth=240
    onButtonBaseClicked=(&self:WidgetTree.Button_WrapOptions.HandleOnButtonClicked)
    onButtonBaseDoubleClicked=(&self:WidgetTree.Button_WrapOptions.HandleOnButtonDoubleClicked)
    onButtonBaseHovered=(&self:WidgetTree.Button_WrapOptions.HandleOnButtonHovered)
    onButtonBaseUnhovered=(&self:WidgetTree.Button_WrapOptions.HandleOnButtonUnhovered)
    onSelectedChangedBase=(&self:WidgetTree.Button_WrapOptions.HandleOnSelectedChanged)
    paddingOveride=true
    visibility=Visible
    hoveredSlateSoundOverride {resourceObject=/Game/Sounds/Fort_UI_Sounds/FrontEnd/DefaultSounds/PressedButtons/MSS_UI_Button_Hover_Default}
    padding Overide=<ltrb 16 12 16 12>
    pressedSlateSoundOverride {resourceObject=/Game/Sounds/Fort_UI_Sounds/FrontEnd/DefaultSounds/PressedButtons/MSS_UI_Button_Pressed_Default}
    triggeringInputAction {dataTable=/Gifting/GiftingInputActionDataTable, rowName=WrappingOptions}
   }
   slot {class=HorizontalBoxSlot}
  }
  Button_Back {
   widgetClassPath=&SoloButton
   bInherited=true
   parent=HorizontalBox_0
   widget=Button_Back
   properties {
    button Description=Back
    fontSize=17
    fontSizeOveride=true
    minHeight=18
    minWidth=192
    onButtonBaseClicked=(&self:WidgetTree.Button_Back.HandleOnButtonClicked)
    onButtonBaseDoubleClicked=(&self:WidgetTree.Button_Back.HandleOnButtonDoubleClicked)
    onButtonBaseHovered=(&self:WidgetTree.Button_Back.HandleOnButtonHovered)
    onButtonBaseUnhovered=(&self:WidgetTree.Button_Back.HandleOnButtonUnhovered)
    onSelectedChangedBase=(&self:WidgetTree.Button_Back.HandleOnSelectedChanged)
    paddingOveride=true
    visibility=Visible
    hoveredSlateSoundOverride {resourceObject=&MSS_UI_Button_Hover_Default}
    padding=<ltrb 16 0 0 0>
    padding Overide=<ltrb 16 12 16 12>
    pressedSlateSoundOverride {resourceObject=&MSS_UI_Button_Pressed_Default}
    triggeringInputAction {dataTable=&GiftingInputActionDataTable, rowName=Back}
   }
   slot {class=HorizontalBoxSlot}
  }
  Button_CloseTouch {widgetClassPath=&CommonButtonBase, bInherited=true, widget=null}
 ]
}
