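# GiftingScreen is the gifting screen widget over AthenaGiftingScreen, listing recipients and giftable items with a preview.
# Parent class events drive widget switchers for recipients, search warnings, join-party prompt and no-content state, plus a purchase blocker.
# Filter tabs get icons and labels by All or Party type.
# Slide-and-fade animations play on intro and selection changes, and item tiles resize once more than twelve are displayed.
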WidgetBlueprint /Gifting/UI/GiftingScreen
animations [
 Intro {
  binding_count=15
  length_seconds=0.6501
  object=&self:Intro
  sequence {bindings [
    HB_HeaderTexts {
     id=#1
     display_name=Text_OfferName
     possessed_class=Overlay
     tracks [
      {
       type=MovieScene2DTransformTrack
       sections [{
         type=MovieScene2DTransformSection
         end_frame=9
         start_frame=0
         channels [
          Translation.X {type=MovieSceneScriptingFloatChannel, keys [[4.001,-130],[7.0013,0]]}
          Translation.Y {type=MovieSceneScriptingFloatChannel, const [0,1]}
         ]
        }]
      }
      {
       type=MovieSceneFloatTrack
       display_name=Render Opacity
       sections [{
         type=MovieSceneFloatSection
         end_frame=9
         start_frame=0
         channels [None {type=MovieSceneScriptingFloatChannel, keys [[0,0],[4.001,0],[7.0013,1]]}]
        }]
      }
     ]
    }
    Overlay_MainContent {
     id=#2
     possessed_class=Overlay
     tracks [
      {
       type=MovieScene2DTransformTrack
       sections [{
         type=MovieScene2DTransformSection
         end_frame=3
         start_frame=0
         channels [
          Translation.X {type=MovieSceneScriptingFloatChannel, keys [[0,-150],[3,0]]}
          Translation.Y {type=MovieSceneScriptingFloatChannel, const [0,1]}
          Angle {type=MovieSceneScriptingFloatChannel, const [0,1]}
          Scale.X {type=MovieSceneScriptingFloatChannel, const [1,1]}
          Scale.Y {type=MovieSceneScriptingFloatChannel, const [1,1]}
          Shear.X {type=MovieSceneScriptingFloatChannel, const [0,1]}
          Shear.Y {type=MovieSceneScriptingFloatChannel, const [0,1]}
         ]
        }]
      }
      {
       type=MovieSceneFloatTrack
       display_name=Render Opacity
       sections [{type=MovieSceneFloatSection, end_frame=3, start_frame=0, channels [None {type=MovieSceneScriptingFloatChannel, keys [[0,0],[3.0003,1]]}]}]
      }
     ]
    }
    HB_HeaderTexts {
     id=#3
     display_name=HorizontalBox_3
     possessed_class=Overlay
     tracks [
      {
       type=MovieScene2DTransformTrack
       sections [{
         type=MovieScene2DTransformSection
         end_frame=10
         start_frame=0
         channels [
          Translation.X {type=MovieSceneScriptingFloatChannel, keys [[6.0007,-130],[9.001,0]]}
          Translation.Y {type=MovieSceneScriptingFloatChannel, const [0,1]}
          Angle {type=MovieSceneScriptingFloatChannel, const [0,1]}
          Scale.X {type=MovieSceneScriptingFloatChannel, const [1,1]}
          Scale.Y {type=MovieSceneScriptingFloatChannel, const [1,1]}
          Shear.X {type=MovieSceneScriptingFloatChannel, const [0,1]}
          Shear.Y {type=MovieSceneScriptingFloatChannel, const [0,1]}
         ]
        }]
      }
      {
       type=MovieSceneFloatTrack
       display_name=Render Opacity
       sections [{
         type=MovieSceneFloatSection
         end_frame=10
         start_frame=0
         channels [None {type=MovieSceneScriptingFloatChannel, keys [[0,0],[6.0007,0],[9.0013,1]]}]
        }]
      }
     ]
    }
    HB_HeaderTexts {
     id=#4
     display_name=VerticalBox_1
     possessed_class=Overlay
     tracks [
      {
       type=MovieScene2DTransformTrack
       sections [{
         type=MovieScene2DTransformSection
         end_frame=13
         start_frame=0
         channels [
          Translation.X {type=MovieSceneScriptingFloatChannel, keys [[10,130],[13,0]]}
          Translation.Y {type=MovieSceneScriptingFloatChannel, const [0,1]}
          Angle {type=MovieSceneScriptingFloatChannel, const [0,1]}
          Scale.X {type=MovieSceneScriptingFloatChannel, const [1,1]}
          Scale.Y {type=MovieSceneScriptingFloatChannel, const [1,1]}
          Shear.X {type=MovieSceneScriptingFloatChannel, const [0,1]}
          Shear.Y {type=MovieSceneScriptingFloatChannel, const [0,1]}
         ]
        }]
      }
      {
       type=MovieSceneFloatTrack
       display_name=Render Opacity
       sections [{
         type=MovieSceneFloatSection
         end_frame=13
         start_frame=0
         channels [None {type=MovieSceneScriptingFloatChannel, keys [[0,0],[10.0017,0],[13.0023,1]]}]
        }]
      }
     ]
    }
    searchWidget_SocialSearchWidget {
     id=#5
     display_name=GiftRecipientSearchBar
     possessed_class=Overlay
     tracks [
      {
       type=MovieScene2DTransformTrack
       sections [{
         type=MovieScene2DTransformSection
         end_frame=10
         start_frame=0
         channels [
          Translation.X {type=MovieSceneScriptingFloatChannel, keys [[6.0013,-130],[9.0013,0]]}
          Translation.Y {type=MovieSceneScriptingFloatChannel, const [0,1]}
          Angle {type=MovieSceneScriptingFloatChannel, const [0,1]}
          Scale.X {type=MovieSceneScriptingFloatChannel, const [1,1]}
          Scale.Y {type=MovieSceneScriptingFloatChannel, const [1,1]}
          Shear.X {type=MovieSceneScriptingFloatChannel, const [0,1]}
          Shear.Y {type=MovieSceneScriptingFloatChannel, const [0,1]}
         ]
        }]
      }
      {
       type=MovieSceneFloatTrack
       display_name=Render Opacity
       sections [{
         type=MovieSceneFloatSection
         end_frame=10
         start_frame=0
         channels [None {type=MovieSceneScriptingFloatChannel, keys [[0,0],[6.0013,0],[9.0017,1]]}]
        }]
      }
     ]
    }
    HB_HeaderTexts {
     id=#6
     display_name=TileView_Items
     possessed_class=Overlay
     tracks [
      {
       type=MovieScene2DTransformTrack
       sections [{
         type=MovieScene2DTransformSection
         end_frame=12
         start_frame=0
         channels [
          Translation.X {type=MovieSceneScriptingFloatChannel, keys [[8.0017,-130],[11.0023,0]]}
          Translation.Y {type=MovieSceneScriptingFloatChannel, const [0,1]}
          Angle {type=MovieSceneScriptingFloatChannel, const [0,1]}
          Scale.X {type=MovieSceneScriptingFloatChannel, const [1,1]}
          Scale.Y {type=MovieSceneScriptingFloatChannel, const [1,1]}
          Shear.X {type=MovieSceneScriptingFloatChannel, const [0,1]}
          Shear.Y {type=MovieSceneScriptingFloatChannel, const [0,1]}
         ]
        }]
      }
      {
       type=MovieSceneFloatTrack
       display_name=Render Opacity
       sections [{
         type=MovieSceneFloatSection
         end_frame=12
         start_frame=0
         channels [None {type=MovieSceneScriptingFloatChannel, keys [[0,0],[8.0017,0],[11.0023,1]]}]
        }]
      }
     ]
    }
    HB_HeaderTexts {
     id=#7
     display_name=RichText_OfferItemOwnedCount
     possessed_class=Overlay
     tracks [
      {
       type=MovieScene2DTransformTrack
       sections [{
         type=MovieScene2DTransformSection
         end_frame=9
         start_frame=0
         channels [
          Translation.X {type=MovieSceneScriptingFloatChannel, keys [[4.001,-130],[7.0013,0]]}
          Translation.Y {type=MovieSceneScriptingFloatChannel, const [0,1]}
         ]
        }]
      }
      {
       type=MovieSceneFloatTrack
       display_name=Render Opacity
       sections [{
         type=MovieSceneFloatSection
         end_frame=9
         start_frame=0
         channels [None {type=MovieSceneScriptingFloatChannel, keys [[0,0],[4.001,0],[7.0013,1]]}]
        }]
      }
     ]
    }
    HB_HeaderTexts {
     id=#8
     display_name=HB_FilterTabs
     possessed_class=Overlay
     tracks [
      {
       type=MovieScene2DTransformTrack
       sections [{
         type=MovieScene2DTransformSection
         end_frame=9
         start_frame=0
         channels [
          Translation.X {type=MovieSceneScriptingFloatChannel, keys [[4.0013,-130],[7.0013,0]]}
          Translation.Y {type=MovieSceneScriptingFloatChannel, const [0,1]}
          Angle {type=MovieSceneScriptingFloatChannel, const [0,1]}
          Scale.X {type=MovieSceneScriptingFloatChannel, const [1,1]}
          Scale.Y {type=MovieSceneScriptingFloatChannel, const [1,1]}
          Shear.X {type=MovieSceneScriptingFloatChannel, const [0,1]}
          Shear.Y {type=MovieSceneScriptingFloatChannel, const [0,1]}
         ]
        }]
      }
      {
       type=MovieSceneFloatTrack
       display_name=Render Opacity
       sections [{
         type=MovieSceneFloatSection
         end_frame=8
         start_frame=0
         channels [None {type=MovieSceneScriptingFloatChannel, keys [[0,0],[4.0013,0],[7.0017,1]]}]
        }]
      }
     ]
    }
    HB_HeaderTexts {
     id=#9
     display_name=ListView_Recipients
     possessed_class=Overlay
     tracks [
      {
       type=MovieScene2DTransformTrack
       sections [{
         type=MovieScene2DTransformSection
         end_frame=12
         start_frame=0
         channels [
          Translation.X {type=MovieSceneScriptingFloatChannel, keys [[8.002,-130],[11.002,0]]}
          Translation.Y {type=MovieSceneScriptingFloatChannel, const [0,1]}
          Angle {type=MovieSceneScriptingFloatChannel, const [0,1]}
          Scale.X {type=MovieSceneScriptingFloatChannel, const [1,1]}
          Scale.Y {type=MovieSceneScriptingFloatChannel, const [1,1]}
          Shear.X {type=MovieSceneScriptingFloatChannel, const [0,1]}
          Shear.Y {type=MovieSceneScriptingFloatChannel, const [0,1]}
         ]
        }]
      }
      {
       type=MovieSceneFloatTrack
       display_name=Render Opacity
       sections [{
         type=MovieSceneFloatSection
         end_frame=12
         start_frame=0
         channels [None {type=MovieSceneScriptingFloatChannel, keys [[0,0],[8.002,0],[11.0023,1]]}]
        }]
      }
     ]
    }
    HB_HeaderTexts {
     id=#10
     possessed_class=Overlay
     tracks [
      {
       type=MovieScene2DTransformTrack
       sections [{
         type=MovieScene2DTransformSection
         end_frame=6
         start_frame=0
         channels [
          Translation.X {type=MovieSceneScriptingFloatChannel, keys [[2.0017,-130],[5.0017,0]]}
          Translation.Y {type=MovieSceneScriptingFloatChannel, const [0,1]}
          Angle {type=MovieSceneScriptingFloatChannel, const [0,1]}
          Scale.X {type=MovieSceneScriptingFloatChannel, const [1,1]}
          Scale.Y {type=MovieSceneScriptingFloatChannel, const [1,1]}
          Shear.X {type=MovieSceneScriptingFloatChannel, const [0,1]}
          Shear.Y {type=MovieSceneScriptingFloatChannel, const [0,1]}
         ]
        }]
      }
      {
       type=MovieSceneFloatTrack
       display_name=Render Opacity
       sections [{
         type=MovieSceneFloatSection
         end_frame=6
         start_frame=0
         channels [None {type=MovieSceneScriptingFloatChannel, keys [[0,0],[2.0017,0],[5.002,1]]}]
        }]
      }
     ]
    }
   ]}
 }
 FilterCategoryChange {
  binding_count=1
  length_seconds=0.15
  object=&self:FilterCategoryChange
  sequence {bindings [FilterTab_CategoryName {
     id=#11
     possessed_class=CommonTextBlock
     tracks [
      {
       type=MovieSceneFloatTrack
       display_name=Render Opacity
       sections [{type=MovieSceneFloatSection, channels [None {type=MovieSceneScriptingFloatChannel, keys [[0,0],[3,1]]}]}]
      }
      {
       type=MovieScene2DTransformTrack
       sections [{type=MovieScene2DTransformSection, channels [Translation.X {type=MovieSceneScriptingFloatChannel, keys [[0,10],[3,0]]}]}]
      }
     ]
    }]}
 }
 SwitchItem {
  binding_count=1
  length_seconds=0.15
  object=&self:SwitchItem
  sequence {bindings [HB_HeaderTexts {
     id=#12
     display_name=VerticalBox_1
     possessed_class=Overlay
     tracks [
      {
       type=MovieScene2DTransformTrack
       sections [{
         type=MovieScene2DTransformSection
         end_frame=3
         start_frame=0
         channels [
          Translation.X {type=MovieSceneScriptingFloatChannel, keys [[0,130],[3,0]]}
          Translation.Y {type=MovieSceneScriptingFloatChannel, const [0,1]}
          Angle {type=MovieSceneScriptingFloatChannel, const [0,1]}
          Scale.X {type=MovieSceneScriptingFloatChannel, const [1,1]}
          Scale.Y {type=MovieSceneScriptingFloatChannel, const [1,1]}
          Shear.X {type=MovieSceneScriptingFloatChannel, const [0,1]}
          Shear.Y {type=MovieSceneScriptingFloatChannel, const [0,1]}
         ]
        }]
      }
      {
       type=MovieSceneFloatTrack
       display_name=Render Opacity
       sections [{type=MovieSceneFloatSection, end_frame=3, start_frame=0, channels [None {type=MovieSceneScriptingFloatChannel, keys [[0,0],[3,1]]}]}]
      }
     ]
    }]}
 }
 SwitchFriendSelection {
  binding_count=10
  length_seconds=0.4501
  object=&self:SwitchFriendSelection
  sequence {bindings [
    HB_HeaderTexts {
     id=#1
     display_name=Text_OfferName
     possessed_class=Overlay
     tracks [
      {
       type=MovieScene2DTransformTrack
       sections [{
         type=MovieScene2DTransformSection
         channels [
          Translation.X {type=MovieSceneScriptingFloatChannel, keys [[0.0023,-130],[3.0027,0]]}
          Translation.Y {type=MovieSceneScriptingFloatChannel, const [0,1]}
          Angle {type=MovieSceneScriptingFloatChannel, const [0,1]}
          Scale.X {type=MovieSceneScriptingFloatChannel, const [1,1]}
          Scale.Y {type=MovieSceneScriptingFloatChannel, const [1,1]}
          Shear.X {type=MovieSceneScriptingFloatChannel, const [0,1]}
          Shear.Y {type=MovieSceneScriptingFloatChannel, const [0,1]}
         ]
        }]
      }
      {
       type=MovieSceneFloatTrack
       display_name=Render Opacity
       sections [{type=MovieSceneFloatSection, channels [None {type=MovieSceneScriptingFloatChannel, keys [[0.0023,0],[3.0027,1]]}]}]
      }
     ]
    }
    HB_HeaderTexts {
     id=#3
     display_name=HorizontalBox_3
     possessed_class=Overlay
     tracks [
      {
       type=MovieScene2DTransformTrack
       sections [{
         type=MovieScene2DTransformSection
         channels [
          Translation.X {type=MovieSceneScriptingFloatChannel, keys [[2.002,-130],[5.0023,0]]}
          Translation.Y {type=MovieSceneScriptingFloatChannel, const [0,1]}
          Angle {type=MovieSceneScriptingFloatChannel, const [0,1]}
          Scale.X {type=MovieSceneScriptingFloatChannel, const [1,1]}
          Scale.Y {type=MovieSceneScriptingFloatChannel, const [1,1]}
          Shear.X {type=MovieSceneScriptingFloatChannel, const [0,1]}
          Shear.Y {type=MovieSceneScriptingFloatChannel, const [0,1]}
         ]
        }]
      }
      {
       type=MovieSceneFloatTrack
       display_name=Render Opacity
       sections [{type=MovieSceneFloatSection, channels [None {type=MovieSceneScriptingFloatChannel, keys [[2.002,0],[5.0027,1]]}]}]
      }
     ]
    }
    HB_HeaderTexts {
     id=#4
     display_name=VerticalBox_1
     possessed_class=Overlay
     tracks [
      {
       type=MovieScene2DTransformTrack
       sections [{
         type=MovieScene2DTransformSection
         channels [
          Translation.X {type=MovieSceneScriptingFloatChannel, keys [[6.0013,130],[9.0013,0]]}
          Translation.Y {type=MovieSceneScriptingFloatChannel, const [0,1]}
          Angle {type=MovieSceneScriptingFloatChannel, const [0,1]}
          Scale.X {type=MovieSceneScriptingFloatChannel, const [1,1]}
          Scale.Y {type=MovieSceneScriptingFloatChannel, const [1,1]}
          Shear.X {type=MovieSceneScriptingFloatChannel, const [0,1]}
          Shear.Y {type=MovieSceneScriptingFloatChannel, const [0,1]}
         ]
        }]
      }
      {
       type=MovieSceneFloatTrack
       display_name=Render Opacity
       sections [{type=MovieSceneFloatSection, channels [None {type=MovieSceneScriptingFloatChannel, keys [[6.003,0],[9.0037,1]]}]}]
      }
     ]
    }
    HB_HeaderTexts {
     id=#6
     display_name=TileView_Items
     possessed_class=Overlay
     tracks [
      {
       type=MovieScene2DTransformTrack
       sections [{
         type=MovieScene2DTransformSection
         channels [
          Translation.X {type=MovieSceneScriptingFloatChannel, keys [[4.003,-130],[7.0037,0]]}
          Translation.Y {type=MovieSceneScriptingFloatChannel, const [0,1]}
          Angle {type=MovieSceneScriptingFloatChannel, const [0,1]}
          Scale.X {type=MovieSceneScriptingFloatChannel, const [1,1]}
          Scale.Y {type=MovieSceneScriptingFloatChannel, const [1,1]}
          Shear.X {type=MovieSceneScriptingFloatChannel, const [0,1]}
          Shear.Y {type=MovieSceneScriptingFloatChannel, const [0,1]}
         ]
        }]
      }
      {
       type=MovieSceneFloatTrack
       display_name=Render Opacity
       sections [{type=MovieSceneFloatSection, channels [None {type=MovieSceneScriptingFloatChannel, keys [[4.003,0],[7.0037,1]]}]}]
      }
     ]
    }
    HB_HeaderTexts {
     id=#7
     display_name=RichText_OfferItemOwnedCount
     possessed_class=Overlay
     tracks [
      {
       type=MovieScene2DTransformTrack
       sections [{
         type=MovieScene2DTransformSection
         channels [
          Translation.X {type=MovieSceneScriptingFloatChannel, keys [[0.0023,-130],[3.0027,0]]}
          Translation.Y {type=MovieSceneScriptingFloatChannel, const [0,1]}
          Angle {type=MovieSceneScriptingFloatChannel, const [0,1]}
          Scale.X {type=MovieSceneScriptingFloatChannel, const [1,1]}
          Scale.Y {type=MovieSceneScriptingFloatChannel, const [1,1]}
          Shear.X {type=MovieSceneScriptingFloatChannel, const [0,1]}
          Shear.Y {type=MovieSceneScriptingFloatChannel, const [0,1]}
         ]
        }]
      }
      {
       type=MovieSceneFloatTrack
       display_name=Render Opacity
       sections [{type=MovieSceneFloatSection, channels [None {type=MovieSceneScriptingFloatChannel, keys [[0.0023,0],[3.0027,1]]}]}]
      }
     ]
    }
   ]}
 }
 SwitchBackToAllFilter {
  binding_count=2
  length_seconds=0.15
  object=&self:SwitchBackToAllFilter
  sequence {bindings [Overlay_MainContent {
     id=#13
     display_name=HB_GiftRecipients
     possessed_class=Overlay
     tracks [
      {
       type=MovieScene2DTransformTrack
       sections [{
         type=MovieScene2DTransformSection
         end_frame=3
         start_frame=0
         channels [
          Translation.X {type=MovieSceneScriptingFloatChannel, keys [[0,-150],[3,0]]}
          Translation.Y {type=MovieSceneScriptingFloatChannel, const [0,1]}
          Angle {type=MovieSceneScriptingFloatChannel, const [0,1]}
          Scale.X {type=MovieSceneScriptingFloatChannel, const [1,1]}
          Scale.Y {type=MovieSceneScriptingFloatChannel, const [1,1]}
          Shear.X {type=MovieSceneScriptingFloatChannel, const [0,1]}
          Shear.Y {type=MovieSceneScriptingFloatChannel, const [0,1]}
         ]
        }]
      }
      {
       type=MovieSceneFloatTrack
       display_name=Render Opacity
       sections [{type=MovieSceneFloatSection, end_frame=3, start_frame=0, channels [None {type=MovieSceneScriptingFloatChannel, keys [[0,0],[3.0003,1]]}]}]
      }
     ]
    }]}
 }
 JoinAParty {
  binding_count=1
  length_seconds=0.15
  object=&self:JoinAParty
  sequence {bindings [Text_JoinParty {
     id=#14
     display_name=JoinAParty_Text
     possessed_class=Overlay
     tracks [
      {
       type=MovieScene2DTransformTrack
       sections [{
         type=MovieScene2DTransformSection
         end_frame=3
         start_frame=0
         channels [Translation.X {type=MovieSceneScriptingFloatChannel, keys [[0,-130],[3,0]]}]
        }]
      }
      {
       type=MovieSceneFloatTrack
       display_name=Render Opacity
       sections [{type=MovieSceneFloatSection, end_frame=3, start_frame=0, channels [None {type=MovieSceneScriptingFloatChannel, keys [[0,0],[3.0003,1]]}]}]
      }
     ]
    }]}
 }
 SearchError {
  binding_count=1
  length_seconds=0.15
  object=&self:SearchError
  sequence {bindings [Border_SearchError {
     id=#14
     display_name=Border_0
     possessed_class=Overlay
     tracks [
      {
       type=MovieScene2DTransformTrack
       sections [{
         type=MovieScene2DTransformSection
         end_frame=3
         start_frame=0
         channels [Translation.X {type=MovieSceneScriptingFloatChannel, keys [[0,-130],[3,0]]}]
        }]
      }
      {
       type=MovieSceneFloatTrack
       display_name=Render Opacity
       sections [{type=MovieSceneFloatSection, end_frame=3, start_frame=0, channels [None {type=MovieSceneScriptingFloatChannel, keys [[0,0],[3.0003,1]]}]}]
      }
     ]
    }]}
 }
]
graphs [
 ContentBlocked {nodes [
   n1 = ContentBlocked!() -> (then, IsContentBlocked:bool=~, ContentBlockedText:Text) {is=title K2Node_FunctionEntry}
   > Set Data!(
    execute
    self:WBP Content Purchase Blocker Object Reference=Get WBP_ContentPurchaseBlocker() -> (WBP_ContentPurchaseBlocker:WBP Content Purchase Blocker Object Reference)
    InText:Text=@n1.ContentBlockedText
   ) -> (then) {is=title K2Node_CallFunction}
   > SetVisibility(self=Get WBP_ContentPurchaseBlocker(), InVisibility=Select(Option 0=Collapsed, Option 1=SelfHitTestInvisible, Index=@n1))
   > SetVisibility(
    self=Get VB_MainContent() -> (VB_MainContent:Vertical Box Object Reference)
    InVisibility=Select(Option 0=SelfHitTestInvisible, Option 1=Collapsed, Index=@n1)
   )
  ]}
 EventGraph {nodes [
   n1 = TilesCreated!() -> (OutputDelegate, then, Item:Object Reference, Widget:User Widget Object Reference) {is=title K2Node_CustomEvent}
   > UpdateTileSize()
   n2 = Event OnShowSearchWarningText!() -> (OutputDelegate, then, bShow:bool) {is=title K2Node_Event}
   > SetActiveWidget(
    self:Widget Switcher Object Reference=Get Switcher_ListError() -> (Switcher_ListError:Widget Switcher Object Reference)
    Widget=Select(
     Option 0:Widget Object Reference=Get ListView_Recipients() -> (ListView_Recipients:Fort Gifting Social User List View Object Reference)
     Option 1:Widget Object Reference=Get Border_SearchError() -> (Border_SearchError:Border Object Reference)
     Index=@n2.bShow
    ) -> (ReturnValue:Widget Object Reference)
   )
   n3 = Event OnPreviewActionButtonUpdated!() -> (OutputDelegate, then, RowDisplayName:"Text (by ref)", Button:Common Button Base Object Reference) {is=title K2Node_Event}
   > n4 = Cast To WBP_UIKit_ButtonCTA_Base!(execute, Object=Get Button_PreviewAction() -> (Button_PreviewAction:Common Button Base Object Reference)) -> (then, CastFailed, AsWBP UIKit Button CTA Base:WBP UIKit Button CTA Base Object Reference) {is=title K2Node_DynamicCast}
   > SetText!(execute, self:Fort CTAButton Object Reference=@n4, ButtonText:"Text (by ref)"=@n3.RowDisplayName) -> (then)
   > SetVisibility(
    self=@n4
    InVisibility=Select(Option 0:ESlateVisibility Enum=Visible, Option 1:ESlateVisibility Enum=Collapsed, Index=TextIsEmpty(InText=@n3.RowDisplayName)) -> (ReturnValue:ESlateVisibility Enum)
   )
   n5 = Event BlockScreenContent!() -> (OutputDelegate, then, bBlockScreen:bool, ContentBlockedText:"Text (by ref)") {is=title K2Node_Event}
   > ContentBlocked!(execute, self, IsContentBlocked:bool=@n5.bBlockScreen, ContentBlockedText:Text=@n5.ContentBlockedText) -> (then) {is=title K2Node_CallFunction}
   Event OnInitialized()
   > Bind Event to On Entry Initialized(self:List View Object Reference=Get TileView_Items() -> (TileView_Items:Tile View Object Reference), Delegate=@n1) {is=title K2Node_AddDelegate}
   Event Destruct()
   > Set FortItemShopOfferVM(FortItemShopOfferVM:Fort Item Shop Offer VM Object Reference) -> (Output_Get:Fort Item Shop Offer VM Object Reference)
   n6 = Event OnItemDownloadCompleted!() -> (OutputDelegate, then, Result:bool) {is=title K2Node_Event}
   > Branch(Condition=@n6.Result)
   > On3dItemDownloaded(self:WBP Item Preview 2d Preview Object Reference=Get WBP_ItemPreview_2dPreview() -> (WBP_ItemPreview_2dPreview:WBP Item Preview 2d Preview Object Reference))
   n7 = Event OnHandlePreviewItem!() -> (OutputDelegate, then, bUseWidePreview:bool) {is=title K2Node_Event}
   > SetPadding!(
    execute
    self:Overlay Slot Object Reference=SlotAsOverlaySlot(Widget=Get Border_CameraFrame() -> (Border_CameraFrame:Border Object Reference))
    InPadding:Margin Structure=Make Margin(Left=SelectFloat(A=-150, B=0, bPickA=@n7.bUseWidePreview), Top=100, Right=SelectFloat(A=-150, B=0, bPickA=@n7.bUseWidePreview), Bottom=100)
   ) -> (then)
   Event Construct()
   > PlayAnimation(InAnimation=Get Intro())
   > PlaySound2D(Sound=Get IntroSound() -> (IntroSound:Sound Base Object Reference))
   Event On Activated()
   n8 = Get WBP_ItemPreview_2dPreview()
   > Set Is2dPreviewEnabled!(
    execute
    Is2dPreviewEnabled:bool=IsCosmetic2dPreviewEnabled!(self:Fort Item Preview Screen Object Reference) -> (ReturnValue:bool)
    self:WBP Item Preview 2d Preview Object Reference=@n8
   ) -> (then, Output_Get:bool)
   > ActivateWidget(self=@n8)
   Event On Deactivated()
   > DeactivateWidget(self=Get WBP_ItemPreview_2dPreview())
  ]}
 OnFilterChanged {nodes [
   n1 = OnFilterChanged!() -> (then, FilterType:EFilterType Enum=All) {is=title K2Node_FunctionEntry}
   > "SetText (Text)"(
    self=Get FilterTab_CategoryName() -> (FilterTab_CategoryName:Common Text Object Reference)
    InText=Select!(All:Text="LOCTEXT(\"All\")", Party:Text="LOCTEXT(\"Party\")", Index:EFilterType Enum=@n1) -> (ReturnValue:Text)
   )
   > PlayAnimationForward(InAnimation=Get FilterCategoryChange() -> (FilterCategoryChange:Widget Animation Object Reference))
   > Branch(Condition="Equal (Enum)"(A:EFilterType Enum=Get FilterType() -> (FilterType:EFilterType Enum=All), B:EFilterType Enum=All))
   > SetActiveWidget(
    self=Get Switcher_GiftRecipients_PartyJoin() -> (Switcher_GiftRecipients_PartyJoin:Widget Switcher Object Reference)
    Widget=Get VBox_List() -> (VBox_List:Vertical Box Object Reference)
   )
   > RequestRefreshFocus(self:Common Activatable Widget Object Reference)
  ]}
 OnFriendSelectionChanged {nodes [
   OnFriendSelectionChanged() {is=title K2Node_FunctionEntry}
   > PlayAnimation(InAnimation=Get SwitchFriendSelection() -> (SwitchFriendSelection:Widget Animation Object Reference))
  ]}
 OnPartyListUpdated {nodes [
   n1 = OnPartyListUpdated!() -> (then, bEmpty:bool) {is=title K2Node_FunctionEntry}
   > SetActiveWidget(
    self=Get Switcher_GiftRecipients_PartyJoin()
    Widget=Select!(
     Option 0:Widget Object Reference=Get VBox_List()
     Option 1:Widget Object Reference=Get Text_JoinParty() -> (Text_JoinParty:Common Text Object Reference)
     Index:bool=@n1
    ) -> (ReturnValue:Widget Object Reference)
   )
   > Branch(Condition=Get bEmpty() -> (bEmpty:bool))
   > PlayAnimationForward(InAnimation=Get JoinAParty() -> (JoinAParty:Widget Animation Object Reference))
  ]}
 OnPresentationModeChanged {nodes [
   n1 = OnPresentationModeChanged!() -> (then, Mode:EGiftingScreenPresentationMode Enum=ItemList) {is=title K2Node_FunctionEntry}
   > SetActiveWidgetIndex(
    self=Get Switcher_GiftContent() -> (Switcher_GiftContent:Widget Switcher Object Reference)
    Index=Select!(ItemList:int=0, NoContent:int=1, Index:EGiftingScreenPresentationMode Enum=@n1) -> (ReturnValue:int)
   )
  ]}
 OnSetupFilterTabButton {nodes [
   n1 = OnSetupFilterTabButton!() -> (then, Button:Common Button Base Object Reference, FilterType:EFilterType Enum=All) {is=title K2Node_FunctionEntry}
   > n2 = Cast To IconTabButton_Legacy!(execute, Object=@n1) -> (then, CastFailed, AsIcon Tab Button Legacy:Icon Tab Button Legacy Object Reference) {is=title K2Node_DynamicCast}
   > n3 = Set TabButton(TabButton:Icon Tab Button Legacy Object Reference=@n2) -> (Output_Get:Icon Tab Button Legacy Object Reference)
   > Set Icon!(
    execute
    self:Icon Tab Button Legacy Object Reference=@n3
    IconBrush:Slate Brush Structure=Select!(
     All:Slate Brush Structure=Get All() -> (All:Slate Brush Structure)
     Party:Slate Brush Structure=Get Party() -> (Party:Slate Brush Structure)
     Index:EFilterType Enum=Get FilterType() -> (FilterType=All)
    ) -> (ReturnValue:Slate Brush Structure)
   ) -> (then) {is=title K2Node_CallFunction}
   > SetStyle!(
    execute
    self:Common Button Base Object Reference=Get TabButton() -> (TabButton:Icon Tab Button Legacy Object Reference)
    InStyle:Common Button Style Class Reference=Get In Style() -> (In Style:Common Button Style Class Reference)
   ) -> (then)
  ]}
 OnShownItemChanged {nodes [
   OnShownItemChanged() {is=title K2Node_FunctionEntry}
   > PlayAnimation(InAnimation=Get SwitchItem() -> (SwitchItem:Widget Animation Object Reference))
  ]}
 UpdateTileSize {nodes [
   UpdateTileSize() {is=title K2Node_FunctionEntry}
   n1 = Get TileView_Items()
   n2 = "integer > integer"(
    A=Length(TargetArray:Array of User Widget Object References=GetDisplayedEntryWidgets!(self:List View Base Object Reference=@n1) -> (ReturnValue:Array of User Widget Object References))
    B=12
   )
   > SetEntryHeight!(execute, self:Tile View Object Reference=@n1, NewHeight:float=SelectFloat(A=110, B=150, bPickA=@n2)) -> (then)
   > SetEntryWidth!(execute, self:Tile View Object Reference=@n1, NewWidth:float=SelectFloat(A=88, B=120, bPickA=@n2)) -> (then)
   > SetPadding!(
    execute
    self:Horizontal Box Slot Object Reference=SlotAsHorizontalBoxSlot!(Widget:Widget Object Reference=Get CenterColumn_SizeBox() -> (CenterColumn_SizeBox:Size Box Object Reference)) -> (ReturnValue:Horizontal Box Slot Object Reference)
    InPadding_Left:float=0
    InPadding_Top:float=SelectFloat(A=100, B=175, bPickA=@n2)
    InPadding_Right:float=0
    InPadding_Bottom:float=0
   ) -> (then)
  ]}
]
variables [
 All {
  type=SlateBrush
  default {
   bIsDynamicallyLoaded=false
   drawAs=Image
   imageType=NoImage
   mirroring=NoMirror
   resourceName=None
   resourceObject=/Game/UI/Foundation/Textures/Icons/Manage/T-Icon-Manage-All-64
   tiling=NoTile
   imageSize=<46 46>
   margin=<ltrb 0 0 0 0>
   outlineSettings {
    bUseBrushTransparency=false
    roundingType=HalfHeightRadius
    width=0
    color {colorUseRule=UseColor_Specified, specifiedColor=<rgba 0 0 0 0>}
    cornerRadii=<0 0 0 1>
   }
   tintColor {colorUseRule=UseColor_Specified, specifiedColor=<rgba 1 1 1 1>}
   uVRegion {bIsValid=false, max=<0 0>, min=<0 0>}
  }
 }
 In Style {type=Class, default=/FortUILibrary/Foundation/Buttons/ButtonStyles/Tab/Athena_ButtonStyle_LockerTab.Athena_ButtonStyle_LockerTab_C}
 IntroSound {type=SoundBase, default=/Game/Sounds/Fort_UI_Sounds/FrontEnd/RewardsAndGifts/UI_FrontEnd_Gift_Screen_Intro_Cue}
 Party {
  type=SlateBrush
  default {
   bIsDynamicallyLoaded=false
   drawAs=Image
   imageType=NoImage
   mirroring=NoMirror
   resourceName=None
   resourceObject=/Game/UI/Foundation/Textures/Icons/Social/T-Icon-Chat-Party-64
   tiling=NoTile
   imageSize=<46 46>
   margin=<ltrb 0 0 0 0>
   outlineSettings {
    bUseBrushTransparency=false
    roundingType=HalfHeightRadius
    width=0
    color {colorUseRule=UseColor_Specified, specifiedColor=<rgba 0 0 0 0>}
    cornerRadii=<0 0 0 1>
   }
   tintColor {colorUseRule=UseColor_Specified, specifiedColor=<rgba 1 1 1 1>}
   uVRegion {bIsValid=false, max=<0 0>, min=<0 0>}
  }
 }
]
widgets {
 info {
  inheritedWidgetCount=26
  namedSlotCount=1
  parentClass=/Script/GiftingUI.AthenaGiftingScreen
  rootWidgetClass=&Overlay
  widgetCount=66
 }
 widgets [
  Overlay_1 {widgetClassPath=&Overlay, widget=Overlay_1}
  SafeZone_Mobile {
   widgetClassPath=&SafeZone
   bIsVariable=true
   parent=Overlay_1
   widget=SafeZone_Mobile
   slot {class=OverlaySlot, properties {horizontalAlignment=HAlign_Fill, verticalAlignment=VAlign_Fill}}
  }
  Overlay_2 {widgetClassPath=&Overlay, parent=SafeZone_Mobile, widget=Overlay_2, slot {class=SafeZoneSlot}}
  HB_MainContent {
   widgetClassPath=&HorizontalBox
   parent=Overlay_2
   widget=HB_MainContent
   slot {class=OverlaySlot, properties {horizontalAlignment=HAlign_Fill, verticalAlignment=VAlign_Fill, padding=<ltrb 24 24 24 24>}}
  }
  Overlay_MainContent {
   widgetClassPath=&Overlay
   parent=HB_MainContent
   widget=Overlay_MainContent
   properties {renderTransformPivot=<0 0.5>}
   slot {class=HorizontalBoxSlot, properties {horizontalAlignment=HAlign_Left}}
  }
  SizeBox_Backplate {
   widgetClassPath=&SizeBox
   parent=Overlay_MainContent
   widget=SizeBox_Backplate
   properties {bOverride_WidthOverride=true, widthOverride=750}
   slot {class=OverlaySlot, properties {horizontalAlignment=HAlign_Fill, verticalAlignment=VAlign_Fill}}
  }
  Backplate {
   widgetClassPath=&Image
   parent=SizeBox_Backplate
   widget=Backplate
   properties {
    visibility=SelfHitTestInvisible
    brush {
     bIsDynamicallyLoaded=false
     drawAs=Image
     imageType=NoImage
     mirroring=NoMirror
     resourceName=None
     resourceObject=/Gifting/UI/Art/M_UI_Gifting_Background
     tiling=NoTile
     imageSize=<1024 1024>
     margin=<ltrb 0 0 0 0>
     outlineSettings {
      bUseBrushTransparency=false
      roundingType=HalfHeightRadius
      width=0
      color {colorUseRule=UseColor_Specified, specifiedColor=<rgba 0 0 0 0>}
      cornerRadii=<0 0 0 0>
     }
     tintColor {colorUseRule=UseColor_Specified, specifiedColor=<rgba 1 1 1 1>}
     uVRegion {bIsValid=false, max=<0 0>, min=<0 0>}
    }
   }
   slot {class=SizeBoxSlot, properties {padding=<ltrb -150 -130 0 -130>}}
  }
  VB_MainContent {
   widgetClassPath=&VerticalBox
   bIsVariable=true
   parent=Overlay_MainContent
   widget=VB_MainContent
   properties {flowDirectionPreference=Culture, renderTransformPivot=<0 0.5>}
   slot {class=OverlaySlot, properties {horizontalAlignment=HAlign_Fill, verticalAlignment=VAlign_Fill, padding=<ltrb 30 40 80 0>}}
  }
  HB_HeaderTexts {
   widgetClassPath=&VerticalBox
   parent=VB_MainContent
   widget=HB_HeaderTexts
   slot {class=VerticalBoxSlot, properties {horizontalAlignment=HAlign_Left}}
  }
  Text_Title {
   widgetClassPath=&CommonTextBlock
   parent=HB_HeaderTexts
   widget=Text_Title
   properties {
    style=/FortUILibrary/Styles/Typography/_TextStyle_PrimaryH2._TextStyle_PrimaryH2_C
    text=Gifting
    textTransformPolicy=ToUpper
    font {
     bForceMonospaced=false
     bMaterialIsStencil=false
     fontMaterial=null
     fontObject=&Burbank
     letterSpacing=0
     monospacedWidth=1
     size=48
     skewAmount=0
     typefaceFontName=BurbankRegular-Black
     outlineSettings {
      bApplyOutlineToDropShadows=false
      bMiteredCorners=false
      bSeparateFillAlpha=false
      outlineMaterial=null
      outlineSize=0
      outlineColor=<rgba 0 0 0 1>
     }
    }
    shadowOffset=<0 0>
    strikeBrush {
     bIsDynamicallyLoaded=false
     drawAs=NoDrawType
     imageType=NoImage
     mirroring=NoMirror
     resourceName=None
     resourceObject=null
     tiling=NoTile
     imageSize=<32 32>
     margin=<ltrb 0 0 0 0>
     outlineSettings {
      bUseBrushTransparency=false
      roundingType=HalfHeightRadius
      width=0
      color {colorUseRule=UseColor_Specified, specifiedColor=<rgba 0 0 0 0>}
      cornerRadii=<0 0 0 0>
     }
     tintColor {colorUseRule=UseColor_Specified, specifiedColor=<rgba 1 1 1 1>}
     uVRegion {bIsValid=false, max=<0 0>, min=<0 0>}
    }
   }
   slot {class=VerticalBoxSlot, properties {horizontalAlignment=HAlign_Left}}
  }
  RichText_GiftCount {
   widgetClassPath=&CommonRichTextBlock
   bInherited=true
   bIsVariable=true
   parent=HB_HeaderTexts
   widget=RichText_GiftCount
   properties {
    bOverrideDefaultStyle=true
    defaultTextStyleOverrideClass=/FortUILibrary/Styles/Typography/_TextStyle_SubH4_LightBlue._TextStyle_SubH4_LightBlue_C
    justification=Right
    mobileTextBlockScale=1.1
    text="<text color=\"fff\">5</> DAILY GIFTS REMAINING"
    textTransformPolicy=ToUpper
    defaultTextStyleOverride {
     overflowPolicy=Clip
     transformPolicy=None
     colorAndOpacity {colorUseRule=UseColor_Specified, specifiedColor=<rgba 1 1 1 1>}
     font {
      bForceMonospaced=false
      bMaterialIsStencil=false
      fontMaterial=null
      fontObject=&BurbankSmall
      letterSpacing=0
      monospacedWidth=1
      size=24
      skewAmount=0
      typefaceFontName=Black
      outlineSettings {
       bApplyOutlineToDropShadows=false
       bMiteredCorners=false
       bSeparateFillAlpha=false
       outlineMaterial=null
       outlineSize=0
       outlineColor=<rgba 0 0 0 1>
      }
     }
     highlightColor {colorUseRule=UseColor_Specified, specifiedColor=<rgba 0 0 0 1>}
     highlightShape {
      bIsDynamicallyLoaded=false
      drawAs=Image
      imageType=NoImage
      mirroring=NoMirror
      resourceName=None
      resourceObject=null
      tiling=NoTile
      imageSize=<32 32>
      margin=<ltrb 0 0 0 0>
      outlineSettings {
       bUseBrushTransparency=false
       roundingType=HalfHeightRadius
       width=0
       color {colorUseRule=UseColor_Specified, specifiedColor=<rgba 0 0 0 0>}
       cornerRadii=<0 0 0 0>
      }
      tintColor {colorUseRule=UseColor_Specified, specifiedColor=<rgba 1 1 1 1>}
      uVRegion {bIsValid=false, max=<0 0>, min=<0 0>}
     }
     selectedBackgroundColor {colorUseRule=UseColor_Foreground, specifiedColor=<rgba 1 0 1 1>}
     shadowColorAndOpacity=<rgba 0 0 0 1>
     shadowOffset=<0 0>
     strikeBrush {
      bIsDynamicallyLoaded=false
      drawAs=NoDrawType
      imageType=NoImage
      mirroring=NoMirror
      resourceName=None
      resourceObject=null
      tiling=NoTile
      imageSize=<32 32>
      margin=<ltrb 0 0 0 0>
      outlineSettings {
       bUseBrushTransparency=false
       roundingType=HalfHeightRadius
       width=0
       color {colorUseRule=UseColor_Specified, specifiedColor=<rgba 0 0 0 0>}
       cornerRadii=<0 0 0 0>
      }
      tintColor {colorUseRule=UseColor_Specified, specifiedColor=<rgba 1 1 1 1>}
      uVRegion {bIsValid=false, max=<0 0>, min=<0 0>}
     }
     underlineBrush {
      bIsDynamicallyLoaded=false
      drawAs=Image
      imageType=NoImage
      mirroring=NoMirror
      resourceName=None
      resourceObject=null
      tiling=NoTile
      imageSize=<32 32>
      margin=<ltrb 0 0 0 0>
      outlineSettings {
       bUseBrushTransparency=false
       roundingType=HalfHeightRadius
       width=0
       color {colorUseRule=UseColor_Specified, specifiedColor=<rgba 0 0 0 0>}
       cornerRadii=<0 0 0 0>
      }
      tintColor {colorUseRule=UseColor_Specified, specifiedColor=<rgba 1 1 1 1>}
      uVRegion {bIsValid=false, max=<0 0>, min=<0 0>}
     }
    }
   }
   slot {class=VerticalBoxSlot}
  }
  HB_FilterTabs {
   widgetClassPath=&HorizontalBox
   parent=VB_MainContent
   widget=HB_FilterTabs
   slot {class=VerticalBoxSlot, properties {horizontalAlignment=HAlign_Left, verticalAlignment=VAlign_Top, padding=<ltrb 0 10 0 0>}}
  }
  ActionWidget_FilterTabsPrevious {
   widgetClassPath=&CommonActionWidget
   bInherited=true
   bIsVariable=true
   parent=HB_FilterTabs
   widget=ActionWidget_FilterTabsPrevious
   properties {inputActions [{dataTable=null, rowName=None}]}
   slot {class=HorizontalBoxSlot}
  }
  EntryBox_FilterTabs {
   widgetClassPath=&DynamicEntryBox
   bInherited=true
   bIsVariable=true
   parent=HB_FilterTabs
   widget=EntryBox_FilterTabs
   properties {entryWidgetClass=/FortUILibrary/Foundation/Buttons/IconTabButton_Legacy.IconTabButton_Legacy_C, numDesignerPreviewEntries=2}
   slot {class=HorizontalBoxSlot}
  }
  ActionWidget_FilterTabsNext {
   widgetClassPath=&CommonActionWidget
   bInherited=true
   bIsVariable=true
   parent=HB_FilterTabs
   widget=ActionWidget_FilterTabsNext
   properties {inputActions [{dataTable=null, rowName=None}]}
   slot {class=HorizontalBoxSlot}
  }
  FilterTab_CategoryName {
   widgetClassPath=&CommonTextBlock
   bIsVariable=true
   parent=HB_FilterTabs
   widget=FilterTab_CategoryName
   properties {
    text="{CATEGORY}"
    textTransformPolicy=ToUpper
    font {
     bForceMonospaced=false
     bMaterialIsStencil=false
     fontMaterial=null
     fontObject=&Burbank
     letterSpacing=0
     monospacedWidth=1
     size=30
     skewAmount=0
     typefaceFontName=BurbankRegular-Black
     outlineSettings {
      bApplyOutlineToDropShadows=false
      bMiteredCorners=false
      bSeparateFillAlpha=false
      outlineMaterial=null
      outlineSize=0
      outlineColor=<rgba 0 0 0 1>
     }
    }
    renderTransform {angle=0, scale=<1 1>, shear=<-12 0>, translation=<0 0>}
    shadowOffset=<0 0>
   }
   slot {class=HorizontalBoxSlot, properties {verticalAlignment=VAlign_Center, padding=<ltrb 35 0 0 0>}}
  }
  HB_GiftRecipients {
   widgetClassPath=&HorizontalBox
   parent=VB_MainContent
   widget=HB_GiftRecipients
   slot {class=VerticalBoxSlot, properties {horizontalAlignment=HAlign_Left, padding=<ltrb 0 20 0 0>, size {sizeRule=Fill, value=1}}}
  }
  SizeBox_2 {
   widgetClassPath=&SizeBox
   parent=HB_GiftRecipients
   widget=SizeBox_2
   properties {bOverride_WidthOverride=true, widthOverride=640}
   slot {class=HorizontalBoxSlot}
  }
  Switcher_GiftRecipients_PartyJoin {
   widgetClassPath=&WidgetSwitcher
   bIsVariable=true
   parent=SizeBox_2
   widget=Switcher_GiftRecipients_PartyJoin
   slot {class=SizeBoxSlot}
  }
  VBox_List {
   widgetClassPath=&VerticalBox
   bIsVariable=true
   parent=Switcher_GiftRecipients_PartyJoin
   widget=VBox_List
   properties {navigation=&self:WidgetTree.VBox_List.WidgetNavigation_0}
   slot {class=WidgetSwitcherSlot}
  }
  searchWidget_SocialSearchWidget {
   widgetClassPath=&/GiftRecipientSearchBar
   bInherited=true
   bIsVariable=true
   parent=VBox_List
   widget=searchWidget_SocialSearchWidget
   properties {flowDirectionPreference=LeftToRight}
   slot {class=VerticalBoxSlot}
  }
  Switcher_ListError {
   widgetClassPath=&WidgetSwitcher
   bIsVariable=true
   parent=VBox_List
   widget=Switcher_ListError
   slot {class=VerticalBoxSlot, properties {padding=<ltrb 0 20 0 0>, size {sizeRule=Fill, value=1}}}
  }
  ListView_Recipients {
   widgetClassPath=/Script/GiftingUI.FortGiftingSocialUserListView
   bInherited=true
   bIsVariable=true
   parent=Switcher_ListError
   widget=ListView_Recipients
   properties {entrySpacing=10, entryWidgetClass=&/GiftRecipientEntry.GiftRecipientEntry_C, numDesignerPreviewEntries=8}
   slot {class=WidgetSwitcherSlot}
  }
  Border_SearchError {
   widgetClassPath=&Border
   bIsVariable=true
   parent=Switcher_ListError
   widget=Border_SearchError
   properties {
    background {
     bIsDynamicallyLoaded=false
     drawAs=NoDrawType
     imageType=NoImage
     mirroring=NoMirror
     resourceName=None
     resourceObject=/FortniteCoreUI/Frontend/StoreMain/M_UI_SkewBoxBorder_VBucks
     tiling=NoTile
     imageSize=<32 32>
     margin=<ltrb 1 1 1 1>
     outlineSettings {
      bUseBrushTransparency=false
      roundingType=HalfHeightRadius
      width=0
      color {colorUseRule=UseColor_Specified, specifiedColor=<rgba 0 0 0 0>}
      cornerRadii=<0 0 0 0>
     }
     tintColor {colorUseRule=UseColor_Specified, specifiedColor=<rgba 1 1 1 1>}
     uVRegion {bIsValid=false, max=<0 0>, min=<0 0>}
    }
    brushColor=<rgba 1 0.1714 0.1329 1>
    padding=<ltrb 2 2 2 2>
    renderTransform {angle=0, scale=<1 1>, shear=<-12 0>, translation=<0 0>}
   }
   slot {class=WidgetSwitcherSlot, properties {horizontalAlignment=HAlign_Center, verticalAlignment=VAlign_Center, padding=<ltrb 0 0 0 135>}}
  }
  Text_SearchError {
   widgetClassPath=&CommonTextBlock
   parent=Border_SearchError
   widget=Text_SearchError
   properties {
    justification=Center
    text=You must use at least 3 characters to search your friend list.
    textTransformPolicy=ToUpper
    wrapTextAt=400
    colorAndOpacity {colorUseRule=UseColor_Specified, specifiedColor=<rgba 1 0.1714 0.1329 1>}
    font {
     bForceMonospaced=false
     bMaterialIsStencil=false
     fontMaterial=null
     fontObject=&BurbankSmall
     letterSpacing=0
     monospacedWidth=1
     size=20
     skewAmount=0
     typefaceFontName=Bold
     outlineSettings {
      bApplyOutlineToDropShadows=false
      bMiteredCorners=false
      bSeparateFillAlpha=false
      outlineMaterial=null
      outlineSize=0
      outlineColor=<rgba 0 0 0 1>
     }
    }
    shadowOffset=<0 0>
   }
   slot {class=BorderSlot, properties {padding=<ltrb 2 2 2 2>}}
  }
  ScrollBox_Offers {
   widgetClassPath=/Script/FortniteUI.AthenaScrollBox
   parent=VBox_List
   widget=ScrollBox_Offers
   properties {
    bAllowRightClickDragScrolling=false
    bAnimateWheelScrolling=false
    bEnableClickThrough=true
    consumeMouseWheel=Never
    navigationScrollPadding=100
    scrollWhenFocusChanges=AnimatedScroll
    visibility=Collapsed
    scrollbarPadding=<ltrb 30 0 0 0>
    scrollbarThickness=<10 10>
    widgetBarStyle {
     thickness=16
     draggedThumbImage {
      bIsDynamicallyLoaded=false
      drawAs=Image
      imageType=FullColor
      mirroring=NoMirror
      resourceName=../../../Engine/Content/Slate/Common/Scrollbar_Thumb.png
      resourceObject=/FortUILibrary/Foundation/Art_ToolKit/AngledBoxes/M_UI_ItemShop_ScrollBar_Drag
      tiling=NoTile
      imageSize=<0 0>
      margin=<ltrb 0.25 0.25 0.25 0.25>
      outlineSettings {
       bUseBrushTransparency=false
       roundingType=HalfHeightRadius
       width=0
       color {colorUseRule=UseColor_Specified, specifiedColor=<rgba 0 0 0 0>}
       cornerRadii=<0 0 0 0>
      }
      tintColor {colorUseRule=UseColor_Specified, specifiedColor=<rgba 1 1 1 1>}
      uVRegion {bIsValid=false, max=<0 0>, min=<0 0>}
     }
     horizontalBackgroundImage {
      bIsDynamicallyLoaded=false
      drawAs=Image
      imageType=NoImage
      mirroring=NoMirror
      resourceName=None
      resourceObject=/FortUILibrary/Foundation/Art_ToolKit/AngledBoxes/M_UI_ItemShop_ScrollBar_Backing
      tiling=NoTile
      imageSize=<0 0>
      margin=<ltrb 0 0 0 0>
      outlineSettings {
       bUseBrushTransparency=false
       roundingType=HalfHeightRadius
       width=0
       color {colorUseRule=UseColor_Specified, specifiedColor=<rgba 0 0 0 0>}
       cornerRadii=<0 0 0 0>
      }
      tintColor {colorUseRule=UseColor_Specified, specifiedColor=<rgba 1 1 1 1>}
      uVRegion {bIsValid=false, max=<0 0>, min=<0 0>}
     }
     horizontalBottomSlotImage {
      bIsDynamicallyLoaded=false
      drawAs=NoDrawType
      imageType=FullColor
      mirroring=NoMirror
      resourceName=../../../Engine/Content/Slate/Common/Scrollbar_Background_Horizontal.png
      resourceObject=null
      tiling=NoTile
      imageSize=<8 8>
      margin=<ltrb 0 0 0 0>
      outlineSettings {
       bUseBrushTransparency=false
       roundingType=HalfHeightRadius
       width=0
       color {colorUseRule=UseColor_Specified, specifiedColor=<rgba 0 0 0 0>}
       cornerRadii=<0 0 0 0>
      }
      tintColor {colorUseRule=UseColor_Specified, specifiedColor=<rgba 1 1 1 1>}
      uVRegion {bIsValid=false, max=<0 0>, min=<0 0>}
     }
     horizontalTopSlotImage {
      bIsDynamicallyLoaded=false
      drawAs=NoDrawType
      imageType=FullColor
      mirroring=NoMirror
      resourceName=../../../Engine/Content/Slate/Common/Scrollbar_Background_Horizontal.png
      resourceObject=null
      tiling=NoTile
      imageSize=<8 8>
      margin=<ltrb 0 0 0 0>
      outlineSettings {
       bUseBrushTransparency=false
       roundingType=HalfHeightRadius
       width=0
       color {colorUseRule=UseColor_Specified, specifiedColor=<rgba 0 0 0 0>}
       cornerRadii=<0 0 0 0>
      }
      tintColor {colorUseRule=UseColor_Specified, specifiedColor=<rgba 1 1 1 1>}
      uVRegion {bIsValid=false, max=<0 0>, min=<0 0>}
     }
     hoveredThumbImage {
      bIsDynamicallyLoaded=false
      drawAs=Image
      imageType=FullColor
      mirroring=NoMirror
      resourceName=../../../Engine/Content/Slate/Common/Scrollbar_Thumb.png
      resourceObject=/FortUILibrary/Foundation/Art_ToolKit/AngledBoxes/M_UI_ItemShop_ScrollBar_Hover
      tiling=NoTile
      imageSize=<0 0>
      margin=<ltrb 0.25 0.25 0.25 0.25>
      outlineSettings {
       bUseBrushTransparency=false
       roundingType=HalfHeightRadius
       width=0
       color {colorUseRule=UseColor_Specified, specifiedColor=<rgba 0 0 0 0>}
       cornerRadii=<0 0 0 0>
      }
      tintColor {colorUseRule=UseColor_Specified, specifiedColor=<rgba 1 1 1 1>}
      uVRegion {bIsValid=false, max=<0 0>, min=<0 0>}
     }
     normalThumbImage {
      bIsDynamicallyLoaded=false
      drawAs=Image
      imageType=FullColor
      mirroring=NoMirror
      resourceName=../../../Engine/Content/Slate/Common/Scrollbar_Thumb.png
      resourceObject=/FortUILibrary/Foundation/Art_ToolKit/AngledBoxes/M_UI_ItemShop_ScrollBar_Base
      tiling=NoTile
      imageSize=<0 0>
      margin=<ltrb 0.25 0.25 0.25 0.25>
      outlineSettings {
       bUseBrushTransparency=false
       roundingType=HalfHeightRadius
       width=0
       color {colorUseRule=UseColor_Specified, specifiedColor=<rgba 0 0 0 0>}
       cornerRadii=<0 0 0 0>
      }
      tintColor {colorUseRule=UseColor_Specified, specifiedColor=<rgba 1 1 1 1>}
      uVRegion {bIsValid=false, max=<0 0>, min=<0 0>}
     }
     verticalBackgroundImage {
      bIsDynamicallyLoaded=false
      drawAs=Image
      imageType=NoImage
      mirroring=NoMirror
      resourceName=None
      resourceObject=&M_UI_ItemShop_ScrollBar_Backing
      tiling=NoTile
      imageSize=<0 0>
      margin=<ltrb 0 0 0 0>
      outlineSettings {
       bUseBrushTransparency=false
       roundingType=HalfHeightRadius
       width=0
       color {colorUseRule=UseColor_Specified, specifiedColor=<rgba 0 0 0 0>}
       cornerRadii=<0 0 0 0>
      }
      tintColor {colorUseRule=UseColor_Specified, specifiedColor=<rgba 1 1 1 1>}
      uVRegion {bIsValid=false, max=<0 0>, min=<0 0>}
     }
     verticalBottomSlotImage {
      bIsDynamicallyLoaded=false
      drawAs=Image
      imageType=FullColor
      mirroring=NoMirror
      resourceName=../../../Engine/Content/Slate/Common/Scrollbar_Background_Vertical.png
      resourceObject=null
      tiling=NoTile
      imageSize=<8 8>
      margin=<ltrb 0 0 0 0>
      outlineSettings {
       bUseBrushTransparency=false
       roundingType=HalfHeightRadius
       width=0
       color {colorUseRule=UseColor_Specified, specifiedColor=<rgba 0 0 0 0>}
       cornerRadii=<0 0 0 0>
      }
      tintColor {colorUseRule=UseColor_Specified, specifiedColor=<rgba 1 1 1 1>}
      uVRegion {bIsValid=false, max=<0 0>, min=<0 0>}
     }
     verticalTopSlotImage {
      bIsDynamicallyLoaded=false
      drawAs=Image
      imageType=FullColor
      mirroring=NoMirror
      resourceName=../../../Engine/Content/Slate/Common/Scrollbar_Background_Vertical.png
      resourceObject=null
      tiling=NoTile
      imageSize=<8 8>
      margin=<ltrb 0 0 0 0>
      outlineSettings {
       bUseBrushTransparency=false
       roundingType=HalfHeightRadius
       width=0
       color {colorUseRule=UseColor_Specified, specifiedColor=<rgba 0 0 0 0>}
       cornerRadii=<0 0 0 0>
      }
      tintColor {colorUseRule=UseColor_Specified, specifiedColor=<rgba 1 1 1 1>}
      uVRegion {bIsValid=false, max=<0 0>, min=<0 0>}
     }
    }
    widgetStyle {
     barThickness=9
     bottomShadowBrush {
      bIsDynamicallyLoaded=false
      drawAs=NoDrawType
      imageType=FullColor
      mirroring=NoMirror
      resourceName=../../../Engine/Content/Slate/Common/ScrollBoxShadowBottom.png
      resourceObject=null
      tiling=NoTile
      imageSize=<16 8>
      margin=<ltrb 0.5 0 0.5 1>
      outlineSettings {
       bUseBrushTransparency=false
       roundingType=HalfHeightRadius
       width=0
       color {colorUseRule=UseColor_Specified, specifiedColor=<rgba 0 0 0 0>}
       cornerRadii=<0 0 0 0>
      }
      tintColor {colorUseRule=UseColor_Specified, specifiedColor=<rgba 1 1 1 1>}
      uVRegion {bIsValid=false, max=<0 0>, min=<0 0>}
     }
     horizontalScrolledContentPadding=<ltrb 0 0 1 0>
     leftShadowBrush {
      bIsDynamicallyLoaded=false
      drawAs=NoDrawType
      imageType=FullColor
      mirroring=NoMirror
      resourceName=../../../Engine/Content/Slate/Common/ScrollBoxShadowLeft.png
      resourceObject=null
      tiling=NoTile
      imageSize=<8 16>
      margin=<ltrb 1 0.5 0 0.5>
      outlineSettings {
       bUseBrushTransparency=false
       roundingType=HalfHeightRadius
       width=0
       color {colorUseRule=UseColor_Specified, specifiedColor=<rgba 0 0 0 0>}
       cornerRadii=<0 0 0 0>
      }
      tintColor {colorUseRule=UseColor_Specified, specifiedColor=<rgba 1 1 1 1>}
      uVRegion {bIsValid=false, max=<0 0>, min=<0 0>}
     }
     rightShadowBrush {
      bIsDynamicallyLoaded=false
      drawAs=NoDrawType
      imageType=FullColor
      mirroring=NoMirror
      resourceName=../../../Engine/Content/Slate/Common/ScrollBoxShadowRight.png
      resourceObject=null
      tiling=NoTile
      imageSize=<8 16>
      margin=<ltrb 0 0.5 1 0.5>
      outlineSettings {
       bUseBrushTransparency=false
       roundingType=HalfHeightRadius
       width=0
       color {colorUseRule=UseColor_Specified, specifiedColor=<rgba 0 0 0 0>}
       cornerRadii=<0 0 0 0>
      }
      tintColor {colorUseRule=UseColor_Specified, specifiedColor=<rgba 1 1 1 1>}
      uVRegion {bIsValid=false, max=<0 0>, min=<0 0>}
     }
     topShadowBrush {
      bIsDynamicallyLoaded=false
      drawAs=NoDrawType
      imageType=FullColor
      mirroring=NoMirror
      resourceName=../../../Engine/Content/Slate/Common/ScrollBoxShadowTop.png
      resourceObject=null
      tiling=NoTile
      imageSize=<16 8>
      margin=<ltrb 0.5 1 0.5 0>
      outlineSettings {
       bUseBrushTransparency=false
       roundingType=HalfHeightRadius
       width=0
       color {colorUseRule=UseColor_Specified, specifiedColor=<rgba 0 0 0 0>}
       cornerRadii=<0 0 0 0>
      }
      tintColor {colorUseRule=UseColor_Specified, specifiedColor=<rgba 1 1 1 1>}
      uVRegion {bIsValid=false, max=<0 0>, min=<0 0>}
     }
     verticalScrolledContentPadding=<ltrb 0 0 0 1>
    }
   }
   slot {class=VerticalBoxSlot, properties {padding=<ltrb 0 20 0 0>, size {sizeRule=Fill, value=1}}}
  }
  EntryBox_OfferEntries {
   widgetClassPath=&DynamicEntryBox
   bIsVariable=true
   parent=ScrollBox_Offers
   widget=EntryBox_OfferEntries
   properties {
    entryBoxType=Vertical
    entryWidgetClass=&GiftRecipientEntry_C
    numDesignerPreviewEntries=10
    entrySpacing=<0 10>
   }
   slot {class=ScrollBoxSlot}
  }
  Text_JoinParty {
   widgetClassPath=&CommonTextBlock
   bIsVariable=true
   parent=Switcher_GiftRecipients_PartyJoin
   widget=Text_JoinParty
   properties {
    autoWrapText=true
    justification=Center
    text=Join a party
    textTransformPolicy=ToUpper
    wrapTextAt=330
    font {
     bForceMonospaced=false
     bMaterialIsStencil=false
     fontMaterial=null
     fontObject=&Burbank
     letterSpacing=0
     monospacedWidth=1
     size=24
     skewAmount=0
     typefaceFontName=BurbankRegular-Black
     outlineSettings {
      bApplyOutlineToDropShadows=false
      bMiteredCorners=false
      bSeparateFillAlpha=false
      outlineMaterial=null
      outlineSize=0
      outlineColor=<rgba 0 0 0 1>
     }
    }
    renderTransform {angle=0, scale=<1 1>, shear=<-12 0>, translation=<0 0>}
    shadowOffset=<0 0>
   }
   slot {class=WidgetSwitcherSlot, properties {verticalAlignment=VAlign_Center, padding=<ltrb 0 0 0 140>}}
  }
  WBP_ContentPurchaseBlocker {
   widgetClassPath=/FortniteCoreUI/UI/ContentControls/WBP_ContentPurchaseBlocker
   bIsVariable=true
   parent=Overlay_MainContent
   widget=WBP_ContentPurchaseBlocker
   properties {visibility=Collapsed}
   slot {class=OverlaySlot, properties {horizontalAlignment=HAlign_Center, verticalAlignment=VAlign_Center}}
  }
  VerticalBox_2 {
   widgetClassPath=&VerticalBox
   parent=HB_MainContent
   widget=VerticalBox_2
   slot {class=HorizontalBoxSlot, properties {padding=<ltrb 40 0 0 0>, size {sizeRule=Fill, value=0.5}}}
  }
  Switcher_GiftContent {
   widgetClassPath=&WidgetSwitcher
   bIsVariable=true
   parent=VerticalBox_2
   widget=Switcher_GiftContent
   slot {class=VerticalBoxSlot, properties {size {sizeRule=Fill, value=1}}}
  }
  HB_CentreAndCameraFraming {widgetClassPath=&HorizontalBox, parent=Switcher_GiftContent, widget=HB_CentreAndCameraFraming, slot {class=WidgetSwitcherSlot}}
  CenterColumn_SizeBox {
   widgetClassPath=&SizeBox
   bIsVariable=true
   parent=HB_CentreAndCameraFraming
   widget=CenterColumn_SizeBox
   properties {bOverride_WidthOverride=true, widthOverride=480}
   slot {class=HorizontalBoxSlot, properties {padding=<ltrb 0 175 0 0>}}
  }
  VerticalBox_0 {
   widgetClassPath=&VerticalBox
   parent=CenterColumn_SizeBox
   widget=VerticalBox_0
   properties {flowDirectionPreference=Culture}
   slot {class=SizeBoxSlot}
  }
  Text_OfferName {
   widgetClassPath=&CommonTextBlock
   bInherited=true
   bIsVariable=true
   parent=VerticalBox_0
   widget=Text_OfferName
   properties {
    mobileFontSizeMultiplier=1.2
    scrollStyle=&TextScrollStyle-Base_C
    style=/FortUILibrary/Styles/Typography/_TextStyle_SubH4._TextStyle_SubH4_C
    text="{ITEMNAME}"
    textTransformPolicy=ToUpper
    font {
     bForceMonospaced=false
     bMaterialIsStencil=false
     fontMaterial=null
     fontObject=&BurbankSmall
     letterSpacing=0
     monospacedWidth=1
     size=24
     skewAmount=0
     typefaceFontName=Black
     outlineSettings {
      bApplyOutlineToDropShadows=false
      bMiteredCorners=false
      bSeparateFillAlpha=false
      outlineMaterial=null
      outlineSize=0
      outlineColor=<rgba 0 0 0 1>
     }
    }
    shadowOffset=<0 0>
    strikeBrush {
     bIsDynamicallyLoaded=false
     drawAs=NoDrawType
     imageType=NoImage
     mirroring=NoMirror
     resourceName=None
     resourceObject=null
     tiling=NoTile
     imageSize=<32 32>
     margin=<ltrb 0 0 0 0>
     outlineSettings {
      bUseBrushTransparency=false
      roundingType=HalfHeightRadius
      width=0
      color {colorUseRule=UseColor_Specified, specifiedColor=<rgba 0 0 0 0>}
      cornerRadii=<0 0 0 0>
     }
     tintColor {colorUseRule=UseColor_Specified, specifiedColor=<rgba 1 1 1 1>}
     uVRegion {bIsValid=false, max=<0 0>, min=<0 0>}
    }
   }
   slot {class=VerticalBoxSlot, properties {horizontalAlignment=HAlign_Left, verticalAlignment=VAlign_Center}}
  }
  RichText_OfferItemOwnedCount {
   widgetClassPath=&CommonRichTextBlock
   bInherited=true
   parent=VerticalBox_0
   widget=RichText_OfferItemOwnedCount
   properties {
    bOverrideDefaultStyle=true
    defaultTextStyleOverrideClass=/FortUILibrary/Styles/Typography/_TextStyle_NormalTag_LightBlue._TextStyle_NormalTag_LightBlue_C
    mobileTextBlockScale=1.1
    text="<text color=\"fff\">0</> / 0 ITEMS OWNED"
    textTransformPolicy=ToUpper
    defaultTextStyleOverride {
     overflowPolicy=Clip
     transformPolicy=None
     colorAndOpacity {colorUseRule=UseColor_Specified, specifiedColor=<rgba 0.2159 0.6445 0.9473 1>}
     font {
      bForceMonospaced=false
      bMaterialIsStencil=false
      fontMaterial=null
      fontObject=&Burbank
      letterSpacing=0
      monospacedWidth=1
      size=16
      skewAmount=0
      typefaceFontName=BurbankRegular-Black
      outlineSettings {
       bApplyOutlineToDropShadows=false
       bMiteredCorners=false
       bSeparateFillAlpha=false
       outlineMaterial=null
       outlineSize=0
       outlineColor=<rgba 0 0 0 1>
      }
     }
     highlightColor {colorUseRule=UseColor_Specified, specifiedColor=<rgba 0 0 0 1>}
     highlightShape {
      bIsDynamicallyLoaded=false
      drawAs=Image
      imageType=NoImage
      mirroring=NoMirror
      resourceName=None
      resourceObject=null
      tiling=NoTile
      imageSize=<32 32>
      margin=<ltrb 0 0 0 0>
      outlineSettings {
       bUseBrushTransparency=false
       roundingType=HalfHeightRadius
       width=0
       color {colorUseRule=UseColor_Specified, specifiedColor=<rgba 0 0 0 0>}
       cornerRadii=<0 0 0 0>
      }
      tintColor {colorUseRule=UseColor_Specified, specifiedColor=<rgba 1 1 1 1>}
      uVRegion {bIsValid=false, max=<0 0>, min=<0 0>}
     }
     selectedBackgroundColor {colorUseRule=UseColor_Foreground, specifiedColor=<rgba 1 0 1 1>}
     shadowColorAndOpacity=<rgba 0 0 0 1>
     shadowOffset=<0 0>
     strikeBrush {
      bIsDynamicallyLoaded=false
      drawAs=Image
      imageType=NoImage
      mirroring=NoMirror
      resourceName=None
      resourceObject=null
      tiling=NoTile
      imageSize=<32 32>
      margin=<ltrb 0 0 0 0>
      outlineSettings {
       bUseBrushTransparency=false
       roundingType=HalfHeightRadius
       width=0
       color {colorUseRule=UseColor_Specified, specifiedColor=<rgba 0 0 0 0>}
       cornerRadii=<0 0 0 0>
      }
      tintColor {colorUseRule=UseColor_Specified, specifiedColor=<rgba 1 1 1 1>}
      uVRegion {bIsValid=false, max=<0 0>, min=<0 0>}
     }
     underlineBrush {
      bIsDynamicallyLoaded=false
      drawAs=Image
      imageType=NoImage
      mirroring=NoMirror
      resourceName=None
      resourceObject=null
      tiling=NoTile
      imageSize=<32 32>
      margin=<ltrb 0 0 0 0>
      outlineSettings {
       bUseBrushTransparency=false
       roundingType=HalfHeightRadius
       width=0
       color {colorUseRule=UseColor_Specified, specifiedColor=<rgba 0 0 0 0>}
       cornerRadii=<0 0 0 0>
      }
      tintColor {colorUseRule=UseColor_Specified, specifiedColor=<rgba 1 1 1 1>}
      uVRegion {bIsValid=false, max=<0 0>, min=<0 0>}
     }
    }
   }
   slot {class=VerticalBoxSlot, properties {horizontalAlignment=HAlign_Left, verticalAlignment=VAlign_Center}}
  }
  HorizontalBox_3 {
   widgetClassPath=&HorizontalBox
   parent=VerticalBox_0
   widget=HorizontalBox_3
   slot {class=VerticalBoxSlot, properties {horizontalAlignment=HAlign_Left, verticalAlignment=VAlign_Top, padding=<ltrb 0 8 0 0>}}
  }
  Container_VBucksOffViolator {
   widgetClassPath=&Overlay
   bInherited=true
   parent=HorizontalBox_3
   widget=Container_VBucksOffViolator
   properties {visibility=Visible}
   slot {class=HorizontalBoxSlot}
  }
  CommonBorder_0 {
   widgetClassPath=&CommonBorder
   parent=Container_VBucksOffViolator
   widget=CommonBorder_0
   properties {
    verticalAlignment=VAlign_Center
    background {
     bIsDynamicallyLoaded=false
     drawAs=Image
     imageType=NoImage
     mirroring=NoMirror
     resourceName=None
     resourceObject=null
     tiling=NoTile
     imageSize=<32 32>
     margin=<ltrb 0 0 0 0>
     outlineSettings {
      bUseBrushTransparency=false
      roundingType=HalfHeightRadius
      width=0
      color {colorUseRule=UseColor_Specified, specifiedColor=<rgba 0 0 0 0>}
      cornerRadii=<0 0 0 0>
     }
     tintColor {colorUseRule=UseColor_Specified, specifiedColor=<rgba 1 1 0 1>}
     uVRegion {bIsValid=false, max=<0 0>, min=<0 0>}
    }
    padding=<ltrb 12 0 12 0>
    renderTransform {angle=0, scale=<1 1>, shear=<0 0>, translation=<-12 0>}
   }
   slot {class=OverlaySlot, properties {horizontalAlignment=HAlign_Fill, verticalAlignment=VAlign_Fill, padding=<ltrb 0 4 0 4>}}
  }
  Text_VBucksOffCount {
   widgetClassPath=&CommonTextBlock
   bInherited=true
   parent=CommonBorder_0
   widget=Text_VBucksOffCount
   properties {
    text="{1000 VBUCKS OFF}"
    textTransformPolicy=ToUpper
    colorAndOpacity {colorUseRule=UseColor_Specified, specifiedColor=<rgba 0 0 0 1>}
    font {
     bForceMonospaced=false
     bMaterialIsStencil=false
     fontMaterial=null
     fontObject=&Burbank
     letterSpacing=0
     monospacedWidth=1
     size=18
     skewAmount=0
     typefaceFontName=BurbankRegular-Black
     outlineSettings {
      bApplyOutlineToDropShadows=false
      bMiteredCorners=false
      bSeparateFillAlpha=false
      outlineMaterial=null
      outlineSize=0
      outlineColor=<rgba 0 0 0 1>
     }
    }
    renderTransform {angle=0, scale=<1 1>, shear=<-12 0>, translation=<0 0>}
    shadowOffset=<0 0>
   }
   slot {class=BorderSlot, properties {verticalAlignment=VAlign_Center, padding=<ltrb 12 0 12 0>}}
  }
  Widget_Price {
   widgetClassPath=&/GiftingPrice
   bInherited=true
   bIsVariable=true
   parent=HorizontalBox_3
   widget=Widget_Price
   slot {class=HorizontalBoxSlot}
  }
  TileView_Items {
   widgetClassPath=/Script/UMG.TileView
   bInherited=true
   bIsVariable=true
   parent=VerticalBox_0
   widget=TileView_Items
   properties {
    bIsFocusable=false
    entryHeight=150
    entryWidgetClass=/FortniteCoreUI/Frontend/AthenaStoreReward.AthenaStoreReward_C
    entryWidth=120
    horizontalEntrySpacing=6
    navigation=&self:WidgetTree.TileView_Items.WidgetNavigation_0
    numDesignerPreviewEntries=8
    tileAlignment=LeftAligned
    verticalEntrySpacing=6
    visibility=HitTestInvisible
   }
   slot {class=VerticalBoxSlot, properties {padding=<ltrb -10 44 0 0>}}
  }
  Overlay_CameraFraming {
   widgetClassPath=&Overlay
   bIsVariable=true
   parent=HB_CentreAndCameraFraming
   widget=Overlay_CameraFraming
   slot {class=HorizontalBoxSlot, properties {size {sizeRule=Fill, value=0.5}}}
  }
  WBP_ItemPreview_2dPreview {
   widgetClassPath=/FortniteCoreUI/Frontend/ItemPreview/WBP_ItemPreview_2dPreview
   bIsVariable=true
   parent=Overlay_CameraFraming
   widget=WBP_ItemPreview_2dPreview
   slot {class=OverlaySlot, properties {horizontalAlignment=HAlign_Fill, verticalAlignment=VAlign_Fill}}
  }
  Border_CameraFrame {
   widgetClassPath=&Border
   bIsVariable=true
   parent=Overlay_CameraFraming
   widget=Border_CameraFrame
   properties {
    visibility=HitTestInvisible
    background {
     bIsDynamicallyLoaded=false
     drawAs=NoDrawType
     imageType=NoImage
     mirroring=NoMirror
     resourceName=None
     resourceObject=null
     tiling=NoTile
     imageSize=<32 32>
     margin=<ltrb 0 0 0 0>
     outlineSettings {
      bUseBrushTransparency=false
      roundingType=HalfHeightRadius
      width=0
      color {colorUseRule=UseColor_Specified, specifiedColor=<rgba 0 0 0 0>}
      cornerRadii=<0 0 0 0>
     }
     tintColor {colorUseRule=UseColor_Specified, specifiedColor=<rgba 1 1 1 1>}
     uVRegion {bIsValid=false, max=<0 0>, min=<0 0>}
    }
    padding=<ltrb 0 0 0 0>
   }
   slot {class=OverlaySlot, properties {horizontalAlignment=HAlign_Fill, verticalAlignment=VAlign_Fill}}
  }
  Widget_CameraFrame {
   widgetClassPath=&SizeBox
   bInherited=true
   bIsVariable=true
   parent=Border_CameraFrame
   widget=Widget_CameraFrame
   slot {class=BorderSlot, properties {padding=<ltrb 0 0 0 0>}}
  }
  ItemDownloadIndicator {
   widgetClassPath=/Game/UI/ContentOnDemand/ItemDownloadIndicator/ItemDownloadIndicator
   bInherited=true
   bIsVariable=true
   parent=Widget_CameraFrame
   widget=ItemDownloadIndicator
   slot {class=SizeBoxSlot}
  }
  VerticalBox_1 {
   widgetClassPath=&VerticalBox
   parent=Overlay_CameraFraming
   widget=VerticalBox_1
   slot {class=OverlaySlot, properties {horizontalAlignment=HAlign_Right, verticalAlignment=VAlign_Bottom, padding=<ltrb 0 0 0 20>}}
  }
  Text_ShownItemIndex {
   widgetClassPath=&CommonTextBlock
   bInherited=true
   bIsVariable=true
   parent=VerticalBox_1
   widget=Text_ShownItemIndex
   properties {
    text="## of ####"
    font {
     bForceMonospaced=false
     bMaterialIsStencil=false
     fontMaterial=null
     fontObject=&BurbankSmall
     letterSpacing=0
     monospacedWidth=1
     size=16
     skewAmount=0
     typefaceFontName=Bold
     outlineSettings {
      bApplyOutlineToDropShadows=false
      bMiteredCorners=false
      bSeparateFillAlpha=false
      outlineMaterial=null
      outlineSize=0
      outlineColor=<rgba 0 0 0 1>
     }
    }
    shadowOffset=<0 0>
   }
   slot {
    class=VerticalBoxSlot
    properties {horizontalAlignment=HAlign_Right, verticalAlignment=VAlign_Bottom, padding=<ltrb 0 0 40 10>, size {sizeRule=Fill, value=1}}
   }
  }
  Widget_ItemInfo {
   widgetClassPath=/FortniteCoreUI/Customization/AthenaLocker_ItemInfo
   bInherited=true
   bIsVariable=true
   parent=VerticalBox_1
   widget=Widget_ItemInfo
   properties {wrap Text=true}
   slot {class=VerticalBoxSlot}
  }
  VB_NoContentMessage {
   widgetClassPath=&Overlay
   parent=Switcher_GiftContent
   widget=VB_NoContentMessage
   slot {class=WidgetSwitcherSlot, properties {horizontalAlignment=HAlign_Center, verticalAlignment=VAlign_Center}}
  }
  SizeBox_0 {
   widgetClassPath=&SizeBox
   parent=VB_NoContentMessage
   widget=SizeBox_0
   properties {bOverride_WidthOverride=true, widthOverride=900}
   slot {class=OverlaySlot, properties {horizontalAlignment=HAlign_Fill, verticalAlignment=VAlign_Fill}}
  }
  Text_NoContent {
   widgetClassPath=&CommonRichTextBlock
   bInherited=true
   bIsVariable=true
   parent=SizeBox_0
   widget=Text_NoContent
   properties {
    autoWrapText=true
    bOverrideDefaultStyle=true
    defaultTextStyleOverrideClass=&_TextStyle_SubH4_C
    justification=Center
    text="YOU HAVE NO DAILY GIFTS REMAINING.\r\nCOME BACK TOMORROW"
    textTransformPolicy=ToUpper
    defaultTextStyleOverride {
     overflowPolicy=Clip
     transformPolicy=None
     colorAndOpacity {colorUseRule=UseColor_Specified, specifiedColor=<rgba 0.2086 0.9301 1 1>}
     font {
      bForceMonospaced=false
      bMaterialIsStencil=false
      fontMaterial=null
      fontObject=&BurbankSmall
      letterSpacing=0
      monospacedWidth=1
      size=23
      skewAmount=0
      typefaceFontName=Black
      outlineSettings {
       bApplyOutlineToDropShadows=false
       bMiteredCorners=false
       bSeparateFillAlpha=false
       outlineMaterial=null
       outlineSize=0
       outlineColor=<rgba 0 0 0 1>
      }
     }
     highlightColor {colorUseRule=UseColor_Specified, specifiedColor=<rgba 0 0 0 1>}
     highlightShape {
      bIsDynamicallyLoaded=false
      drawAs=Image
      imageType=NoImage
      mirroring=NoMirror
      resourceName=None
      resourceObject=null
      tiling=NoTile
      imageSize=<32 32>
      margin=<ltrb 0 0 0 0>
      outlineSettings {
       bUseBrushTransparency=false
       roundingType=HalfHeightRadius
       width=0
       color {colorUseRule=UseColor_Specified, specifiedColor=<rgba 0 0 0 0>}
       cornerRadii=<0 0 0 0>
      }
      tintColor {colorUseRule=UseColor_Specified, specifiedColor=<rgba 1 1 1 1>}
      uVRegion {bIsValid=false, max=<0 0>, min=<0 0>}
     }
     selectedBackgroundColor {colorUseRule=UseColor_Foreground, specifiedColor=<rgba 1 0 1 1>}
     shadowColorAndOpacity=<rgba 0 0 0 1>
     shadowOffset=<0 0>
     strikeBrush {
      bIsDynamicallyLoaded=false
      drawAs=Image
      imageType=NoImage
      mirroring=NoMirror
      resourceName=None
      resourceObject=null
      tiling=NoTile
      imageSize=<32 32>
      margin=<ltrb 0 0 0 0>
      outlineSettings {
       bUseBrushTransparency=false
       roundingType=HalfHeightRadius
       width=0
       color {colorUseRule=UseColor_Specified, specifiedColor=<rgba 0 0 0 0>}
       cornerRadii=<0 0 0 0>
      }
      tintColor {colorUseRule=UseColor_Specified, specifiedColor=<rgba 1 1 1 1>}
      uVRegion {bIsValid=false, max=<0 0>, min=<0 0>}
     }
     underlineBrush {
      bIsDynamicallyLoaded=false
      drawAs=Image
      imageType=NoImage
      mirroring=NoMirror
      resourceName=None
      resourceObject=null
      tiling=NoTile
      imageSize=<32 32>
      margin=<ltrb 0 0 0 0>
      outlineSettings {
       bUseBrushTransparency=false
       roundingType=HalfHeightRadius
       width=0
       color {colorUseRule=UseColor_Specified, specifiedColor=<rgba 0 0 0 0>}
       cornerRadii=<0 0 0 0>
      }
      tintColor {colorUseRule=UseColor_Specified, specifiedColor=<rgba 1 1 1 1>}
      uVRegion {bIsValid=false, max=<0 0>, min=<0 0>}
     }
    }
   }
   slot {class=SizeBoxSlot}
  }
  WBP_ActionBarContainer {
   widgetClassPath=/Game/UI/Frontend/BottomBar/WBP_ActionBarContainer
   parent=VerticalBox_2
   widget=WBP_ActionBarContainer
   slot {class=VerticalBoxSlot, properties {horizontalAlignment=HAlign_Right, verticalAlignment=VAlign_Bottom}}
  }
  HB_BottomRightButtons {widgetClassPath=&HorizontalBox, namedSlotHost=WBP_ActionBarContainer, widget=HB_BottomRightButtons}
  Button_CameraControl {
   widgetClassPath=/FortUILibrary/UIKit/Buttons/WBP_UIKit_Button_Quiet
   bInherited=true
   bIsVariable=true
   parent=HB_BottomRightButtons
   widget=Button_CameraControl
   properties {
    text=Camera
    hoveredSlateSoundOverride {resourceObject=/Game/Sounds/Fort_UI_Sounds/FrontEnd/BattlePass/S20_Store_JoinFNCrew_Hover_Cue}
    lockedHoveredSlateSoundOverride {resourceObject=&S20_Store_JoinFNCrew_Hover_Cue}
    lockedPressedSlateSoundOverride {resourceObject=/Game/Sounds/Fort_UI_Sounds/FrontEnd/BattlePass/BattlePass_Button_Confirm_Generic_B_Cue}
    pressedSlateSoundOverride {resourceObject=&BattlePass_Button_Confirm_Generic_B_Cue}
    selectedHoveredSlateSoundOverride {resourceObject=&S20_Store_JoinFNCrew_Hover_Cue}
    selectedPressedSlateSoundOverride {resourceObject=&BattlePass_Button_Confirm_Generic_B_Cue}
    triggeringInputAction {dataTable=/FortUILibrary/Input/ItemCameraInputActionDataTable, rowName=CanControlCamera}
   }
   slot {class=HorizontalBoxSlot}
  }
  Button_PreviewAction {
   widgetClassPath=&WBP_UIKit_Button_Quiet
   bInherited=true
   bIsVariable=true
   parent=HB_BottomRightButtons
   widget=Button_PreviewAction
   properties {
    text=Preview Action
    visibility=Collapsed
    hoveredSlateSoundOverride {resourceObject=&S20_Store_JoinFNCrew_Hover_Cue}
    lockedHoveredSlateSoundOverride {resourceObject=&S20_Store_JoinFNCrew_Hover_Cue}
    lockedPressedSlateSoundOverride {resourceObject=&BattlePass_Button_Confirm_Generic_B_Cue}
    pressedSlateSoundOverride {resourceObject=&BattlePass_Button_Confirm_Generic_B_Cue}
    selectedHoveredSlateSoundOverride {resourceObject=&S20_Store_JoinFNCrew_Hover_Cue}
    selectedPressedSlateSoundOverride {resourceObject=&BattlePass_Button_Confirm_Generic_B_Cue}
    triggeringInputAction {dataTable=&InputActionDataTable, rowName=GenericLeftStickClick}
   }
   slot {class=HorizontalBoxSlot, properties {padding=<ltrb 24 0 0 0>}}
  }
  Button_GiftingPolicy {
   widgetClassPath=&WBP_UIKit_Button_Quiet
   bInherited=true
   bIsVariable=true
   parent=HB_BottomRightButtons
   widget=Button_GiftingPolicy
   properties {
    text=Gifting Policy
    hoveredSlateSoundOverride {resourceObject=&S20_Store_JoinFNCrew_Hover_Cue}
    lockedHoveredSlateSoundOverride {resourceObject=&S20_Store_JoinFNCrew_Hover_Cue}
    lockedPressedSlateSoundOverride {resourceObject=&BattlePass_Button_Confirm_Generic_B_Cue}
    pressedSlateSoundOverride {resourceObject=&BattlePass_Button_Confirm_Generic_B_Cue}
    selectedHoveredSlateSoundOverride {resourceObject=&S20_Store_JoinFNCrew_Hover_Cue}
    selectedPressedSlateSoundOverride {resourceObject=&BattlePass_Button_Confirm_Generic_B_Cue}
    triggeringInputAction {dataTable=/Gifting/GiftingInputActionDataTable, rowName=GiftingPolicy}
   }
   slot {class=HorizontalBoxSlot, properties {padding=<ltrb 24 0 0 0>}}
  }
  Button_Back {
   widgetClassPath=&WBP_UIKit_Button_Quiet
   bInherited=true
   bIsVariable=true
   parent=HB_BottomRightButtons
   widget=Button_Back
   properties {
    bEnableOnClickedAnalytics=true
    holdPressedSound=&BattlePass_Button_Confirm_Generic_B_Cue
    text=Back
    hoveredSlateSoundOverride {resourceObject=&S20_Store_JoinFNCrew_Hover_Cue}
    lockedHoveredSlateSoundOverride {resourceObject=&S20_Store_JoinFNCrew_Hover_Cue}
    lockedPressedSlateSoundOverride {resourceObject=&BattlePass_Button_Confirm_Generic_B_Cue}
    pressedSlateSoundOverride {resourceObject=&BattlePass_Button_Confirm_Generic_B_Cue}
    selectedHoveredSlateSoundOverride {resourceObject=&S20_Store_JoinFNCrew_Hover_Cue}
    selectedPressedSlateSoundOverride {resourceObject=&BattlePass_Button_Confirm_Generic_B_Cue}
    triggeringInputAction {dataTable=&GiftingInputActionDataTable, rowName=Back}
   }
   slot {class=HorizontalBoxSlot, properties {padding=<ltrb 24 0 0 0>}}
  }
  MTXButtonBox {
   widgetClassPath=&SizeBox
   parent=Overlay_2
   widget=MTXButtonBox
   slot {class=OverlaySlot, properties {horizontalAlignment=HAlign_Right, padding=<ltrb 0 24 24 0>}}
  }
  MTXButton {
   widgetClassPath=/Game/UI/Frontend/TopBar/MTXButton
   parent=MTXButtonBox
   widget=MTXButton
   properties {
    onButtonBaseClicked=(&self:WidgetTree.MTXButton.HandleOnButtonClicked)
    onButtonBaseDoubleClicked=(&self:WidgetTree.MTXButton.HandleOnButtonDoubleClicked)
    onButtonBaseHovered=(&self:WidgetTree.MTXButton.HandleOnButtonHovered)
    onButtonBaseUnhovered=(&self:WidgetTree.MTXButton.HandleOnButtonUnhovered)
    onSelectedChangedBase=(&self:WidgetTree.MTXButton.HandleOnSelectedChanged)
   }
   slot {class=SizeBoxSlot}
  }
  ActivatableWidget_GiftingConfirmation {
   widgetClassPath=&/GiftingPurchaseConfirmationScreen
   bInherited=true
   bIsVariable=true
   parent=Overlay_1
   widget=ActivatableWidget_GiftingConfirmation
   properties {activatedVisibility=Visible, bSetVisibilityOnActivated=true, bSetVisibilityOnDeactivated=true, visibility=Collapsed}
   slot {class=OverlaySlot, properties {horizontalAlignment=HAlign_Fill, verticalAlignment=VAlign_Fill}}
  }
  Button_CloseTouch {widgetClassPath=&CommonButtonBase, bInherited=true, widget=null}
  Button_PreviewAction1 {widgetClassPath=&CommonButtonBase, bInherited=true, widget=null}
  Button_Camera {widgetClassPath=&CommonButtonBase, bInherited=true, widget=null}
  Text_DebugPreviewDetails {widgetClassPath=&CommonTextBlock, bInherited=true, widget=null}
 ]
}
